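# Niagara system for the balloon pop in the ShyTurkey emote, made of confetti and balloon explosion emitters that fire instant bursts.
# Confetti flies out from a sphere, colored blue, dark red or white, under drag, curl noise and downward forces.
# Explosion emitters use spawn groups to vary each mesh shard's size, speed and gravity, adding aerodynamic drag and time-of-day tinting.
# Several emitters are disabled, while Balloon_Confetti003's mesh confetti is active.
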NiagaraSystem /BRCosmetics/Animation/Game/MainPlayer/Emotes/ShyTurkey/FX/NS_Emote_ShyTurkey_BalloonPop
dependencies {
 usedDataInterfaces [
  &NiagaraDataInterfaceVector2DCurve
  &NiagaraDataInterfaceCurve
  &NiagaraDataInterfaceArrayInt32
  &NiagaraDataInterfaceColorCurve
  &NiagaraDataInterfaceArrayFloat2
  &NiagaraDataInterfaceSpriteRendererInfo
  &NiagaraDataInterfaceArrayColor
  &NiagaraDataInterfaceArrayFloat3
 ]
 usedDynamicInputs []
 usedModules [&SystemState]
}
emitters {
 Balloon_Confetti {
  bEnabled=false
  emitterUpdateScript [
   &EmitterState()
   &SpawnBurst_Instantaneous(Spawn Count=&RandomRangeInt(Minimum=10, Maximum=16), Use Spawn Probability=false)
  ]
  particleSpawnScript [
   /SetVariables(Particles.VisibilityTag=&SpawnGroup()) {isSetParameters=true, module=SetVariables_#1}
   &InitializeParticle(
    Color Mode=::NewEnumerator0
    Lifetime Max=1.3
    Lifetime Min=0.8
    Lifetime Mode=::NewEnumerator1
    Sprite Rotation Mode=::NewEnumerator2
    Sprite Size=&MergeHelper_Vector2Switch(
     0=&SelectVector2DFromArray(
      Array Sampling Mode=::NewEnumerator0
      Direct Array Index=?
      Vector 2D Selection Array=DataInterface() {propertyValues {FloatData [<10 8>, <8 5>, <2 3>, <3 2>, <11 13>, <13 11>, <8 8>, <6 4>]}}
      Random Seed=0
     )
     1=&SelectVector2DFromArray(
      Array Sampling Mode=::NewEnumerator0
      Direct Array Index=?
      Vector 2D Selection Array=DataInterface() {propertyValues {FloatData [<10 8>, <8 5>, <2 3>, <3 2>, <11 13>, <13 11>, <4 6>, <6 4>]}}
      Random Seed=0
     )
     2=&RandomRangeVector2D(Minimum=<5 5>, Maximum=<50 50>)
     3=?
     4=?
     Index=$Particles.VisibilityTag
     Num=::NewEnumerator1
    )
    Sprite Size Mode=::NewEnumerator4
    UsePositionOffset=false
    Write Material Random=true
   )
   /Niagara/Modules/Spawn/Location/SphereLocation(MaskSpawnGroup=false, Module.Override Randomness=false, Sphere Radius=20, Surface Only=true)
   &AddVelocityFromPoint(
    Module.Use Falloff Distance=false
    Origin Offset Coordinate Space=::Simulation
    Velocity Strength=&RandomRangeFloat(Minimum=65, Maximum=300)
   )
   &AddVelocity(Velocity=&RandomRangeVector(Minimum=<0 0 10>, Maximum=<1 1 50>))
   &SubUVAnimation(
    End Frame Range Override=23
    Output Alternate Renderer Binding=true
    "Sprite Renderer "=DataInterface() {propertyValues {SpriteRenderer=&self:Balloon_Confetti_0.NiagaraSpriteRendererProperties_2}}
    Start Frame Range Override=21
    SubUV Animation Mode:ENiagara_SubUVLookupModeV2=::NewEnumerator2
    UseEndFrame=true
    UseStartFrame=true
   ) {module=SubUVAnimation001}
   &SubUVAnimation(
    End Frame Range Override=22
    Output Alternate Renderer Binding=true
    "Sprite Renderer "=DataInterface() {propertyValues {SpriteRenderer=&self:Balloon_Confetti_0.NiagaraSpriteRendererProperties_1}}
    Start Frame Range Override=11
    SubUV Animation Mode:ENiagara_SubUVLookupModeV2=::NewEnumerator2
    UseEndFrame=true
    UseStartFrame=true
   ) {module=SubUVAnimation002}
  ]
  particleUpdateScript [
   &ParticleState()
   &Color(Color=&SelectLinearColorFromArray(
     Array Sampling Mode=::NewEnumerator0
     Direct Array Index=?
     Color Selection Array=DataInterface() {propertyValues {ColorData [<rgba 0.0307 0.1144 0.8796 1>, <rgba 0.227 0.0513 0.0529 1>, <rgba 1 1 1 1>]}}
     Random Seed=0
    ))
   &Drag(
    Drag=&FloatFromCurve(
     FloatCurve=DataInterface() {propertyValues {Curve {keys [<at 0.8759 1>,<at 1 0>]}}}
     Scale Curve=&RandomRangeFloat(Minimum=2, Maximum=5)
    )
    Use Linear Drag=true
    Use Rotational Drag=false
   )
   &AccelerationForce(Acceleration=&Multiply_VectorByFloat(
     Vector=&RandomRangeVector(Minimum=<0 0 -50>, Maximum=<0 0 -250>)
     Float=&FloatFromCurve(FloatCurve=DataInterface() {propertyValues {Curve {keys [{arriveTangent=10.9029, interpMode=RCIM_Cubic, leaveTangent=10.9029, tangentMode=RCTM_User, time=0, value=0}, <at 0.25 1>]}}})
    ))
   &CurlNoiseForce(Noise Frequency=5, Noise Strength=100, Pan Noise=false)
   &SolveForcesAndVelocity(Clamp Velocity=false, Limit Acceleration=false)
   &ScaleSpriteSize(Scale Factor=&Vector2DFromFloat(Value=&FloatFromCurve(FloatCurve=DataInterface() {propertyValues {Curve {keys [
          {arriveTangent=2.3096, interpMode=RCIM_Cubic, leaveTangent=2.3096, tangentMode=RCTM_User, time=0, value=0}
          {arriveTangent=-0.0952, interpMode=RCIM_Cubic, leaveTangent=-0.0952, tangentMode=RCTM_User, time=0.431, value=0.3973}
          <at 0.9969 0.0123 RCIM_Cubic>
         ]}}})))
   &SubUVAnimation(End Frame=36, Sub UV Play Rate=0.8, SubUV Animation Mode:ENiagara_SubUVLookupMode=::NewEnumerator3) {enabled=false}
  ]
  rendererClasses [&NiagaraSpriteRendererProperties]
  renderers [
   {rendererClass=&NiagaraSpriteRendererProperties}
   {rendererClass=&NiagaraSpriteRendererProperties, rendererIndex=1}
  ]
 }
 Balloon_Confetti001 {
  bEnabled=false
  emitterUpdateScript [
   &EmitterState()
   &SpawnBurst_Instantaneous(Spawn Count=&RandomRangeInt(Minimum=5, Maximum=7), Use Spawn Probability=false)
  ]
  particleSpawnScript [
   /SetVariables(Particles.VisibilityTag=&SpawnGroup()) {isSetParameters=true, module=SetVariables_#1}
   &InitializeParticle(
    Color Mode=::NewEnumerator0
    Lifetime Max=1.05
    Lifetime Min=0.5
    Lifetime Mode=::NewEnumerator1
    Sprite Rotation Mode=::NewEnumerator2
    Sprite Size=&RandomRangeVector2D(Minimum=<2 2>, Maximum=<5 5>)
    Sprite Size Mode=::NewEnumerator4
    UsePositionOffset=false
    Write Material Random=true
   )
   &SphereLocation(
    MaskSpawnGroup=false
    Module.Override Randomness=false
    Sphere Radius=5
    Surface Expansion Mode:ENiagaraExpansionMode=::NewEnumerator2
    Surface Only=true
   )
   &AddVelocityFromPoint(
    Module.Use Falloff Distance=false
    Origin Offset Coordinate Space=::Simulation
    Velocity Strength=&RandomRangeFloat(Minimum=150, Maximum=250)
   )
   &SubUVAnimation(
    End Frame Range Override=23
    Output Alternate Renderer Binding=true
    "Sprite Renderer "=DataInterface() {propertyValues {SpriteRenderer=&self:Balloon_Confetti001_0.NiagaraSpriteRendererProperties_2}}
    Start Frame Range Override=21
    SubUV Animation Mode:ENiagara_SubUVLookupModeV2=::NewEnumerator2
    UseEndFrame=true
    UseStartFrame=true
   ) {module=SubUVAnimation001}
  ]
  particleUpdateScript [
   &ParticleState()
   &Color(Color=&SelectLinearColorFromArray(
     Array Sampling Mode=::NewEnumerator0
     Direct Array Index=?
     Color Selection Array=DataInterface() {propertyValues {ColorData [<rgba 0.0307 0.1144 0.8796 1>, <rgba 0.227 0.0513 0.0529 1>, <rgba 1 1 1 1>]}}
     Random Seed=0
    ))
   &ScaleColor(
    Scale Alpha=&FloatFromCurve(FloatCurve=DataInterface() {propertyValues {Curve {
        defaultValue=1
        keys [{interpMode=RCIM_Cubic, tangentMode=RCTM_User, time=0, value=1}, {interpMode=RCIM_Cubic, tangentMode=RCTM_User, time=1, value=0}]
       }}})
    Scale RGB=&MakeVectorFromLinearColorRGB(Color=&SelectLinearColorFromArray(
      Array Sampling Mode=::NewEnumerator0
      Direct Array Index=?
      Color Selection Array=DataInterface() {propertyValues {ColorData [<rgba 0.0578 0.2159 0.7305 1>, <rgba 0.227 0.0513 0.0529 1>, <rgba 0.5089 0.521 0.5395 1>]}}
      Random Seed=0
     ))
    ScaleA=true
    ScaleRGB=true
   )
   &Drag(
    Drag=&FloatFromCurve(
     FloatCurve=DataInterface() {propertyValues {Curve {keys [<at 0.8759 1>,<at 0.954 0.0471>]}}}
     Scale Curve=&RandomRangeFloat(Minimum=2, Maximum=5)
    )
    Use Linear Drag=true
    Use Rotational Drag=false
   )
   &AccelerationForce(Acceleration=&Multiply_VectorByFloat(
     Vector=&RandomRangeVector(Minimum=<0 0 -50>, Maximum=<0 0 -250>)
     Float=&FloatFromCurve(FloatCurve=DataInterface() {propertyValues {Curve {keys [{arriveTangent=10.9029, interpMode=RCIM_Cubic, leaveTangent=10.9029, tangentMode=RCTM_User, time=0, value=0}, <at 0.25 1>]}}})
    ))
   &CurlNoiseForce(Noise Frequency=5, Noise Strength=100, Pan Noise=false)
   &GravityForce(Gravity=&RandomRangeVector(Minimum=<0 0 -50>, Maximum=<1 1 -150>))
   &SolveForcesAndVelocity(Clamp Velocity=false, Limit Acceleration=false)
   &ScaleSpriteSize(Scale Factor=&Vector2DFromFloat(Value=&FloatFromCurve(FloatCurve=DataInterface() {propertyValues {Curve {keys [
          {arriveTangent=2.3096, interpMode=RCIM_Cubic, leaveTangent=2.3096, tangentMode=RCTM_User, time=0, value=0}
          {arriveTangent=-0.0952, interpMode=RCIM_Cubic, leaveTangent=-0.0952, tangentMode=RCTM_User, time=0.431, value=1}
          <at 1 1 RCIM_Cubic>
         ]}}})))
   &SubUVAnimation(End Frame=36, Sub UV Play Rate=0.8, SubUV Animation Mode:ENiagara_SubUVLookupMode=::NewEnumerator3) {enabled=false}
  ]
  rendererClasses [&NiagaraSpriteRendererProperties]
  renderers [{rendererClass=&NiagaraSpriteRendererProperties}]
 }
 Balloon_Confetti002 {
  bEnabled=false
  emitterUpdateScript [
   &EmitterState()
   &SpawnBurst_Instantaneous(Spawn Count=&RandomRangeInt(Minimum=5, Maximum=10), Spawn Time=0.02, Use Spawn Probability=false)
  ]
  particleSpawnScript [
   /SetVariables(Particles.VisibilityTag=&SpawnGroup()) {isSetParameters=true, module=SetVariables_#1}
   &InitializeParticle(
    Color Mode=::NewEnumerator0
    Lifetime Max=1.05
    Lifetime Min=0.5
    Lifetime Mode=::NewEnumerator1
    Mesh Scale Mode=::NewEnumerator3
    Mesh Uniform Scale=&RandomRangeFloat(Minimum=0.05, Maximum=0.15)
    Sprite Rotation Mode=::NewEnumerator2
    Sprite Size=&RandomRangeVector2D(Minimum=<2 2>, Maximum=<5 5>)
    Sprite Size Mode=::NewEnumerator4
    UsePositionOffset=false
    Write Material Random=true
   )
   &SphereLocation(
    MaskSpawnGroup=false
    Module.Override Randomness=false
    Sphere Radius=5
    Surface Expansion Mode:ENiagaraExpansionMode=::NewEnumerator2
    Surface Only=true
   )
   &AddVelocityFromPoint(
    Module.Use Falloff Distance=false
    Origin Offset Coordinate Space=::Simulation
    Velocity Strength=&RandomRangeFloat(Minimum=150, Maximum=555)
   )
   &SubUVAnimation(
    End Frame Range Override=23
    Output Alternate Renderer Binding=true
    "Sprite Renderer "=DataInterface() {propertyValues {SpriteRenderer=&self:Balloon_Confetti002_0.NiagaraSpriteRendererProperties_2}}
    Start Frame Range Override=21
    SubUV Animation Mode:ENiagara_SubUVLookupModeV2=::NewEnumerator2
    UseEndFrame=true
    UseStartFrame=true
   ) {enabled=false, module=SubUVAnimation001}
   &InitialMeshOrientation(Use Rotation Vector=false)
  ]
  particleUpdateScript [
   &ParticleState()
   &ScaleMeshSize(Scale Factor=&VectorFromFloat(Value=&FloatFromCurve(FloatCurve=DataInterface() {propertyValues {Curve {keys [<at 0 1>,<at 0.5773 0.0094>]}}})))
   &Color(Color=&SelectLinearColorFromArray(
     Array Sampling Mode=::NewEnumerator0
     Direct Array Index=?
     Color Selection Array=DataInterface() {propertyValues {ColorData [<rgba 0.0307 0.1144 0.8796 1>, <rgba 0.227 0.0513 0.0529 1>, <rgba 1 1 1 1>]}}
     Random Seed=0
    ))
   &ScaleColor(
    Scale Alpha=&FloatFromCurve(FloatCurve=DataInterface() {propertyValues {Curve {
        defaultValue=1
        keys [{interpMode=RCIM_Cubic, tangentMode=RCTM_User, time=0, value=1}, {interpMode=RCIM_Cubic, tangentMode=RCTM_User, time=1, value=0}]
       }}})
    Scale RGB=&MakeVectorFromLinearColorRGB(Color=&SelectLinearColorFromArray(
      Array Sampling Mode=::NewEnumerator0
      Direct Array Index=?
      Color Selection Array=DataInterface() {propertyValues {ColorData [<rgba 0.0578 0.2159 0.7305 1>, <rgba 0.227 0.0513 0.0529 1>, <rgba 0.5089 0.521 0.5395 1>]}}
      Random Seed=0
     ))
    ScaleA=true
    ScaleRGB=true
   )
   &Drag(
    Drag=&FloatFromCurve(
     FloatCurve=DataInterface() {propertyValues {Curve {defaultValue=1, keys [{tangentMode=RCTM_User, time=0, value=1}, {tangentMode=RCTM_User, time=1, value=0}]}}}
     Scale Curve=&RandomRangeFloat(Minimum=1, Maximum=2)
    )
    Use Linear Drag=true
    Use Rotational Drag=false
   )
   &AccelerationForce(Acceleration=&Multiply_VectorByFloat(
     Vector=&RandomRangeVector(Minimum=<0 0 -50>, Maximum=<0 0 -250>)
     Float=&FloatFromCurve(FloatCurve=DataInterface() {propertyValues {Curve {keys [{arriveTangent=10.9029, interpMode=RCIM_Cubic, leaveTangent=10.9029, tangentMode=RCTM_User, time=0, value=0}, <at 0.25 1>]}}})
    ))
   &CurlNoiseForce(Noise Frequency=5, Noise Strength=100, Pan Noise=false)
   &GravityForce(Gravity=&RandomRangeVector(Minimum=<0 0 -50>, Maximum=<1 1 -150>))
   &AddRotationalVelocity(Rotation Rate=<0.1 0.1 0.1>)
   &SolveForcesAndVelocity(Clamp Velocity=false, Limit Acceleration=false)
   &ScaleSpriteSize(Scale Factor=&Vector2DFromFloat(Value=&FloatFromCurve(FloatCurve=DataInterface() {propertyValues {Curve {keys [
          {arriveTangent=2.3096, interpMode=RCIM_Cubic, leaveTangent=2.3096, tangentMode=RCTM_User, time=0, value=0}
          {arriveTangent=-0.0952, interpMode=RCIM_Cubic, leaveTangent=-0.0952, tangentMode=RCTM_User, time=0.431, value=1}
          <at 1 1 RCIM_Cubic>
         ]}}}))) {enabled=false}
   &SubUVAnimation(End Frame=36, Sub UV Play Rate=0.8, SubUV Animation Mode:ENiagara_SubUVLookupMode=::NewEnumerator3) {enabled=false}
  ]
  rendererClasses [&NiagaraSpriteRendererProperties, &NiagaraMeshRendererProperties]
  renderers [
   {rendererClass=&NiagaraSpriteRendererProperties}
   {rendererClass=&NiagaraMeshRendererProperties, rendererIndex=1}
  ]
 }
 Balloon_Confetti003 {
  emitterUpdateScript [
   &EmitterState()
   &SpawnBurst_Instantaneous(Spawn Count=&RandomRangeInt(Minimum=5, Maximum=7), Spawn Time=0.02, Use Spawn Probability=false)
  ]
  particleSpawnScript [
   /SetVariables(Particles.VisibilityTag=&SpawnGroup()) {isSetParameters=true, module=SetVariables_#1}
   &InitializeParticle(
    Color Mode=::NewEnumerator0
    Lifetime Max=1.05
    Lifetime Min=0.5
    Lifetime Mode=::NewEnumerator1
    Mesh Scale Mode=::NewEnumerator3
    Mesh Uniform Scale=&RandomRangeFloat(Minimum=0.03, Maximum=0.1)
    Sprite Rotation Mode=::NewEnumerator2
    Sprite Size=&RandomRangeVector2D(Minimum=<2 2>, Maximum=<5 5>)
    Sprite Size Mode=::NewEnumerator4
    UsePositionOffset=false
    Write Material Random=true
   )
   &SphereLocation(
    MaskSpawnGroup=false
    Module.Override Randomness=false
    Sphere Radius=15
    Surface Expansion Mode:ENiagaraExpansionMode=::NewEnumerator2
    Surface Only=true
   )
   &AddVelocityFromPoint(
    Module.Use Falloff Distance=false
    Origin Offset Coordinate Space=::Simulation
    Velocity Strength=&RandomRangeFloat(Minimum=150, Maximum=555)
   )
   &InitialMeshOrientation(Use Rotation Vector=false)
  ]
  particleUpdateScript [
   &ParticleState()
   &ScaleMeshSize(Scale Factor=&VectorFromFloat(Value=&FloatFromCurve(FloatCurve=DataInterface() {propertyValues {Curve {keys [<at 0 1>,<at 0.5773 0.0094>]}}})))
   &Color(Color=&SelectLinearColorFromArray(
     Array Sampling Mode=::NewEnumerator0
     Direct Array Index=?
     Color Selection Array=DataInterface() {propertyValues {ColorData [<rgba 0.0307 0.1144 0.8796 1>, <rgba 0.227 0.0513 0.0529 1>, <rgba 1 1 1 1>]}}
     Random Seed=0
    ))
   &ScaleColor(
    Scale Alpha=&FloatFromCurve(FloatCurve=DataInterface() {propertyValues {Curve {
        defaultValue=1
        keys [{interpMode=RCIM_Cubic, tangentMode=RCTM_User, time=0, value=1}, {interpMode=RCIM_Cubic, tangentMode=RCTM_User, time=1, value=0}]
       }}})
    Scale RGB=&MakeVectorFromLinearColorRGB(Color=&SelectLinearColorFromArray(
      Array Sampling Mode=::NewEnumerator0
      Direct Array Index=?
      Color Selection Array=DataInterface() {propertyValues {ColorData [<rgba 0.0578 0.2159 0.7305 1>, <rgba 0.227 0.0513 0.0529 1>, <rgba 0.5089 0.521 0.5395 1>]}}
      Random Seed=0
     ))
    ScaleA=true
    ScaleRGB=true
   )
   &Drag(
    Drag=&FloatFromCurve(
     FloatCurve=DataInterface() {propertyValues {Curve {
        defaultValue=1
        keys [{interpMode=RCIM_Cubic, tangentMode=RCTM_Break, time=0, value=0.5}, {interpMode=RCIM_Cubic, tangentMode=RCTM_Break, time=1, value=1}]
       }}}
     Scale Curve=&RandomRangeFloat(Minimum=1, Maximum=2)
    )
    Use Linear Drag=true
    Use Rotational Drag=false
   )
   &AccelerationForce(Acceleration=&Multiply_VectorByFloat(
     Vector=&RandomRangeVector(Minimum=<0 0 -50>, Maximum=<0 0 -250>)
     Float=&FloatFromCurve(FloatCurve=DataInterface() {propertyValues {Curve {keys [{arriveTangent=10.9029, interpMode=RCIM_Cubic, leaveTangent=10.9029, tangentMode=RCTM_User, time=0, value=0}, <at 0.25 1>]}}})
    ))
   &CurlNoiseForce(Noise Frequency=5, Noise Strength=100, Pan Noise=false)
   &GravityForce(Gravity=&RandomRangeVector(Minimum=<0 0 -50>, Maximum=<1 1 -150>))
   &AddRotationalVelocity(Rotation Rate=<0.1 0.1 0.1>)
   &SolveForcesAndVelocity(Clamp Velocity=false, Limit Acceleration=false)
  ]
  rendererClasses [&NiagaraMeshRendererProperties]
  renderers [{rendererClass=&NiagaraMeshRendererProperties, rendererIndex=0}]
 }
 Balloon_explosion_Blue {
  bEnabled=false
  emitterUpdateScript [
   &EmitterState()
   &SpawnBurst_Instantaneous(Spawn Count=4, Spawn Group=2, Use Spawn Probability=false, UseLoopCountLimit=false) {module=SpawnBurst_Instantaneous002}
   &SpawnBurst_Instantaneous(Spawn Count=5, Spawn Group=1, Use Spawn Probability=false, UseLoopCountLimit=false) {module=SpawnBurst_Instantaneous001}
   &SpawnBurst_Instantaneous(Spawn Count=7, Use Spawn Probability=false, UseLoopCountLimit=false)
  ]
  particleSpawnScript [
   /SetVariables(Particles.VisibilityTag=&SpawnGroup()) {isSetParameters=true, module=SetVariables_#2}
   /SetVariables(Particles.ExecutionIndex:NiagaraFloat=&ReturnNormalizedExecIndex()) {isSetParameters=true, module=SetVariables_#3}
   &InitializeParticle(
    Color=&MergeHelper_ColorSwitch(
     0=<rgba 0.0685 0.1946 0.5711 1>
     1=<rgba 0.0307 0.1144 0.8796 1>
     2=<rgba 0.0578 0.2159 0.7305 1>
     3=?
     Index=$Particles.VisibilityTag
     Num=::NewEnumerator1
    )
    Lifetime=&RandomRangeFloat(Minimum=0.65, Maximum=2.3)
    Mesh Scale Mode=::NewEnumerator3
    Mesh Uniform Scale=&MergeHelper_FloatSwitch(
     0=&RandomRangeFloat(Minimum=0.5, Maximum=1)
     1=&RandomRangeFloat(Minimum=0.1, Maximum=0.3)
     2=&RandomRangeFloat(Minimum=0.05, Maximum=0.09)
     3=&MakeFloatFromVector2d(Vector 2D=<5 10>, Channel=::NewEnumerator1)
     Index=$Particles.VisibilityTag
     Num=::NewEnumerator2
    )
    UsePositionOffset=false
    Write Material Random=true
   )
   &ShapeLocation(
    MaskSpawnGroup=false
    Override Seed=false
    Sphere Radius=&MergeHelper_FloatSwitch(0=10, 1=7, 2=5, 3=?, Index=$Particles.VisibilityTag, Num=::NewEnumerator1)
    Sphere Surface Distribution=0.981
    UseOwnerScale=true
   )
   &AddVelocity(
    Module.Use Falloff Distance=false
    Override Seed=false
    Velocity Mode=::NewEnumerator1
    Velocity Speed=&MergeHelper_FloatSwitch(
     0=&RandomRangeFloat(Minimum=444, Maximum=888)
     1=&RandomRangeFloat(Minimum=250, Maximum=350)
     2=&RandomRangeFloat(Minimum=50, Maximum=450)
     3=?
     Index=$Particles.VisibilityTag
     Num=::NewEnumerator1
    )
   )
   &InitialMeshOrientation(
    Mesh Orientation Mode=::NewEnumerator2
    Orientation Coordinate Space=::Local
    Orientation Vector=$Particles.Velocity
    Rotation=<0.125 0 0>
    Use Rotation Vector=true
   )
   /SetVariables(Particles.MeshIndex=&SelectIntFromArray(
     Array Sampling Mode=::NewEnumerator1
     Direct Array Index=?
     Int Selection Array=DataInterface() {propertyValues {IntData [0,1,2,3]}}
     Random Seed=0
    )) {isSetParameters=true, module=SetVariables_#4}
   &SkeletalMeshLocation(
    MaskSpawnGroup=false
    Mesh Sampling Type=::NewEnumerator5
    Orientation Sampling=::NewEnumerator1
    Skeletal Mesh=$System.Skeletal Mesh
    Socket Sampling Mode=::NewEnumerator5
   )
  ]
  particleUpdateScript [
   &ParticleState()
   &ScaleMeshSize(Scale Factor=&VectorFromFloat(Value=&RandomRangeFloat(
      Minimum=&FloatFromCurve(FloatCurve=DataInterface() {propertyValues {Curve {keys [<at 0.35 1>,<at 1 0>]}}})
      Maximum=&FloatFromCurve(FloatCurve=DataInterface() {propertyValues {Curve {keys [<at 0.5 1>,<at 1 0>]}}})
     )))
   &AerodynamicDrag(
    Aerodynamic Drag:NiagaraFloat=&RandomRangeFloat(Minimum=0, Maximum=0.05)
    Aerodynamic Rotational Drag:NiagaraFloat=&RandomRangeFloat(Minimum=0, Maximum=50)
    Drag Shape Curve=DataInterface() {propertyValues {Curve {keys [
        {interpMode=RCIM_Cubic, tangentMode=RCTM_User, time=0, value=0.1}
        {arriveTangent=2.0799, interpMode=RCIM_Cubic, leaveTangent=2.0799, tangentMode=RCTM_User, time=1, value=1}
       ]}}}
    Lift Contribution=0.5
    Lift Shape Curve=DataInterface() {propertyValues {Curve {keys [
        {interpMode=RCIM_Cubic, tangentMode=RCTM_User, time=0, value=0}
        {interpMode=RCIM_Cubic, tangentMode=RCTM_User, time=0.5, value=1}
        {tangentMode=RCTM_User, time=1, value=0}
       ]}}}
    Pivot Offset Maximum:Vector3f=<5 0.5 1>
    Pivot Offset Minimum:Vector3f=<-5 -0.5 -1>
    Rotation Strength:NiagaraFloat=&RandomRangeFloat(Minimum=0.7, Maximum=1)
    Use Linear Drag=true
    Use Rotational Drag=true
   )
   &GravityForce(Gravity=&MakeVector(
     X=0
     Y=0
     Z=&MergeHelper_FloatSwitch(
      0=&FloatFromCurve(
       FloatCurve=DataInterface() {propertyValues {Curve {
          defaultValue=1
          keys [
           {interpMode=RCIM_Cubic, tangentMode=RCTM_User, time=0, value=0}
           {arriveTangent=0.2683, interpMode=RCIM_Cubic, leaveTangent=0.2683, time=0.7, value=0.7652}
           {arriveTangent=0.143, interpMode=RCIM_Cubic, leaveTangent=0.143, time=1.5434, value=0.414}
           {arriveTangent=-0.3012, interpMode=RCIM_Cubic, leaveTangent=-0.3012, tangentMode=RCTM_User, time=2.0131, value=0.9529}
          ]
         }}}
       Scale Curve=&RandomRangeFloat(Minimum=-450, Maximum=-850)
      )
      1=&FloatFromCurve(
       FloatCurve=DataInterface() {propertyValues {Curve {
          defaultValue=1
          keys [
           {interpMode=RCIM_Cubic, tangentMode=RCTM_User, time=0, value=0}
           {interpMode=RCIM_Cubic, tangentMode=RCTM_User, time=0.7, value=0.9843}
           {arriveTangent=0.1579, interpMode=RCIM_Cubic, leaveTangent=0.1579, time=0.8212, value=0.0431}
           {arriveTangent=-0.1169, interpMode=RCIM_Cubic, leaveTangent=-0.1169, time=0.8987, value=1.0157}
           <at 1.0192 0.02 RCIM_Cubic>
           <at 1.0938 1.0157 RCIM_Cubic>
          ]
         }}}
       Scale Curve=&RandomRangeFloat(Minimum=-400, Maximum=-600)
      )
      2=&FloatFromCurve(
       FloatCurve=DataInterface() {propertyValues {Curve {
          defaultValue=1
          keys [
           {interpMode=RCIM_Cubic, tangentMode=RCTM_User, time=0, value=0}
           {interpMode=RCIM_Cubic, tangentMode=RCTM_User, time=0.4, value=0.7}
           {arriveTangent=1.524, interpMode=RCIM_Cubic, leaveTangent=1.524, time=0.573, value=0.1768}
           {arriveTangent=0.7086, interpMode=RCIM_Cubic, leaveTangent=0.7086, time=0.7599, value=1.2485}
           <at 1.121 0.5651 RCIM_Cubic>
          ]
         }}}
       Scale Curve=-150
      )
      3=?
      Index=$Particles.VisibilityTag
      Num=::NewEnumerator1
     )
    ))
   &Drag(
    Drag=&MergeHelper_FloatSwitch(
     0=&FloatFromCurve(
      FloatCurve=DataInterface() {propertyValues {Curve {
         defaultValue=1
         keys [{interpMode=RCIM_Cubic, tangentMode=RCTM_User, time=0, value=1}, {interpMode=RCIM_Cubic, tangentMode=RCTM_User, time=1, value=0}]
        }}}
      Scale Curve=&RandomRangeFloat(Minimum=8, Maximum=10)
     )
     1=&FloatFromCurve(FloatCurve=DataInterface() {propertyValues {Curve {keys [<at 0.8759 1>,<at 1 0>]}}}, Scale Curve=7)
     2=&FloatFromCurve(
      FloatCurve=DataInterface() {propertyValues {Curve {
         defaultValue=1
         keys [{interpMode=RCIM_Cubic, tangentMode=RCTM_User, time=0, value=1}, {interpMode=RCIM_Cubic, tangentMode=RCTM_User, time=1, value=0}]
        }}}
      Scale Curve=5
     )
     3=&FloatFromCurve(
      FloatCurve=DataInterface() {propertyValues {Curve {
         defaultValue=1
         keys [{interpMode=RCIM_Cubic, tangentMode=RCTM_User, time=0, value=1}, {interpMode=RCIM_Cubic, tangentMode=RCTM_User, time=1, value=0}]
        }}}
      Scale Curve=3
     )
     Index=$Particles.VisibilityTag
     Num=::NewEnumerator2
    )
    Use Linear Drag=true
    Use Rotational Drag=false
   )
   &VectorNoiseForce(Force Amount=300, Module.Use Random Noise Time Interval=true)
   &SolveForcesAndVelocity(Clamp Velocity=false, Limit Acceleration=false)
   &ScaleColor(
    Scale Alpha=0
    Scale RGB=&VectorFromFloat(Value=&FNTimeOfDayScalar(Night Scale=0.3, Morning Scale=0.333, Day Scale=0.666, Evening Scale=0.333))
    ScaleA=false
    ScaleRGB=true
   )
   &AlignSpriteToMeshOrientation(SetAlignment=true, SetFacing=true)
   &DynamicMaterialParameters(
    Index 0 Param 1=&RandomRangeFloat(
     Minimum=&FloatFromCurve(
      FloatCurve=DataInterface() {propertyValues {Curve {
         defaultValue=1
         keys [{interpMode=RCIM_Cubic, tangentMode=RCTM_User, time=0, value=1}, {interpMode=RCIM_Cubic, tangentMode=RCTM_User, time=1, value=0}]
        }}}
      Scale Curve=0.5
     )
     Maximum=&FloatFromCurve(FloatCurve=DataInterface() {propertyValues {Curve {
         defaultValue=1
         keys [{interpMode=RCIM_Cubic, tangentMode=RCTM_User, time=0, value=1}, {interpMode=RCIM_Cubic, tangentMode=RCTM_User, time=1, value=0}]
        }}})
    )
    Module.Index0Float0Write=true
    Module.Index0Float1Write=true
    Module.Index0Float2Write=true
    Module.Index0Float3Write=true
   )
  ]
  rendererClasses [&NiagaraMeshRendererProperties]
  renderers [
   {rendererClass=&NiagaraMeshRendererProperties, rendererIndex=0}
   {rendererClass=&NiagaraMeshRendererProperties, rendererIndex=1}
   {rendererClass=&NiagaraMeshRendererProperties, rendererIndex=2}
  ]
 }
 Balloon_explosion_Blue001 {
  bEnabled=false
  emitterUpdateScript [
   &EmitterState()
   &SpawnBurst_Instantaneous(Spawn Count=4, Spawn Group=2, Use Spawn Probability=false, UseLoopCountLimit=false) {module=SpawnBurst_Instantaneous002}
   &SpawnBurst_Instantaneous(Spawn Count=5, Spawn Group=1, Use Spawn Probability=false, UseLoopCountLimit=false) {module=SpawnBurst_Instantaneous001}
   &SpawnBurst_Instantaneous(Spawn Count=7, Use Spawn Probability=false, UseLoopCountLimit=false)
  ]
  particleSpawnScript [
   /SetVariables(Particles.VisibilityTag=&SpawnGroup()) {isSetParameters=true, module=SetVariables_#2}
   /SetVariables(Particles.ExecutionIndex:NiagaraFloat=&ReturnNormalizedExecIndex()) {isSetParameters=true, module=SetVariables_#3}
   &InitializeParticle(
    Color=&MergeHelper_ColorSwitch(0=<rgba 0.227 0.0513 0.0529 1>, 1=<rgba 1 1 1 1>, 2=<rgba 0 0 0 1>, 3=?, Index=$Particles.VisibilityTag, Num=::NewEnumerator1)
    Lifetime=&MergeHelper_FloatSwitch(
     0=&RandomRangeFloat(Minimum=0.65, Maximum=2.3)
     1=&RandomRangeFloat(Minimum=0.65, Maximum=1.5)
     2=&RandomRangeFloat(Minimum=0.65, Maximum=2.3)
     3=?
     Index=$Particles.VisibilityTag
     Num=::NewEnumerator1
    )
    Mesh Scale Mode=::NewEnumerator3
    Mesh Uniform Scale=&MergeHelper_FloatSwitch(
     0=&RandomRangeFloat(Minimum=0.5, Maximum=1)
     1=&RandomRangeFloat(Minimum=0.1, Maximum=0.3)
     2=&RandomRangeFloat(Minimum=0.05, Maximum=0.09)
     3=&MakeFloatFromVector2d(Vector 2D=<5 10>, Channel=::NewEnumerator1)
     Index=$Particles.VisibilityTag
     Num=::NewEnumerator2
    )
    UsePositionOffset=false
    Write Material Random=true
   )
   &AddVelocity(
    Cone Angle=90
    Cone Axis=<1 0 0.5>
    Inner Cone Angle=50
    Module.UseConeFalloff=true
    Module.UseVelDistribution=true
    Override Seed=false
    Speed Falloff From Cone Axis:NiagaraFloat=&RandomRangeFloat(Minimum=0.3, Maximum=0.7)
    Velocity Mode=::NewEnumerator2
    Velocity Speed=&MergeHelper_FloatSwitch(
     0=&RandomRangeFloat(Minimum=350, Maximum=888)
     1=&RandomRangeFloat(Minimum=250, Maximum=350)
     2=&RandomRangeFloat(Minimum=50, Maximum=450)
     3=?
     Index=$Particles.VisibilityTag
     Num=::NewEnumerator1
    )
   )
   &ShapeLocation(
    MaskSpawnGroup=false
    Override Seed=false
    Scale Mode:ENiagara_ScaleMode=::NewEnumerator1
    Sphere Radius=&MergeHelper_FloatSwitch(0=10, 1=7, 2=5, 3=?, Index=$Particles.VisibilityTag, Num=::NewEnumerator1)
    Sphere Surface Distribution=0.981
   )
   &InitialMeshOrientation(
    Mesh Orientation Mode=::NewEnumerator2
    Orientation Coordinate Space=::Local
    Orientation Vector=$Particles.Velocity
    Rotation=<0.125 0 0>
    Use Rotation Vector=true
   )
   /SetVariables(Particles.MeshIndex=&SelectIntFromArray(
     Array Sampling Mode=::NewEnumerator1
     Direct Array Index=?
     Int Selection Array=DataInterface() {propertyValues {IntData [0,1,2,3]}}
     Random Seed=0
    )) {isSetParameters=true, module=SetVariables_#4}
   &SkeletalMeshLocation(
    MaskSpawnGroup=false
    Mesh Sampling Type=::NewEnumerator5
    Orientation Sampling=::NewEnumerator1
    Skeletal Mesh=$System.Skeletal Mesh
    Socket Sampling Mode=::NewEnumerator5
   )
  ]
  particleUpdateScript [
   &ParticleState()
   &ScaleMeshSize(Scale Factor=&VectorFromFloat(Value=&RandomRangeFloat(
      Minimum=&FloatFromCurve(FloatCurve=DataInterface() {propertyValues {Curve {keys [<at 0.35 1>,<at 1 0>]}}})
      Maximum=&FloatFromCurve(FloatCurve=DataInterface() {propertyValues {Curve {keys [<at 0.5 1>,<at 1 0>]}}})
     )))
   &AerodynamicDrag(
    Aerodynamic Drag:NiagaraFloat=&RandomRangeFloat(Minimum=0, Maximum=0.05)
    Aerodynamic Rotational Drag:NiagaraFloat=&RandomRangeFloat(Minimum=0, Maximum=25)
    Drag Shape Curve=DataInterface() {propertyValues {Curve {keys [
        {interpMode=RCIM_Cubic, tangentMode=RCTM_User, time=0, value=0.1}
        {arriveTangent=2.0799, interpMode=RCIM_Cubic, leaveTangent=2.0799, tangentMode=RCTM_User, time=1, value=1}
       ]}}}
    Lift Contribution=0.5
    Lift Shape Curve=DataInterface() {propertyValues {Curve {keys [
        {interpMode=RCIM_Cubic, tangentMode=RCTM_User, time=0, value=0}
        {interpMode=RCIM_Cubic, tangentMode=RCTM_User, time=0.5, value=1}
        {tangentMode=RCTM_User, time=1, value=0}
       ]}}}
    Pivot Offset Maximum:Vector3f=<5 0.5 1>
    Pivot Offset Minimum:Vector3f=<-5 -0.5 -1>
    Rotation Strength:NiagaraFloat=&RandomRangeFloat(Minimum=0.7, Maximum=1)
    Use Linear Drag=true
    Use Rotational Drag=true
   )
   &GravityForce(Gravity=&MakeVector(
     X=0
     Y=0
     Z=&MergeHelper_FloatSwitch(
      0=&FloatFromCurve(
       FloatCurve=DataInterface() {propertyValues {Curve {
          defaultValue=1
          keys [
           {interpMode=RCIM_Cubic, tangentMode=RCTM_User, time=0, value=0}
           {arriveTangent=0.2683, interpMode=RCIM_Cubic, leaveTangent=0.2683, time=0.75, value=0.7652}
           {arriveTangent=0.1878, interpMode=RCIM_Cubic, leaveTangent=0.1878, time=1.5434, value=0.414}
           {arriveTangent=-0.3012, interpMode=RCIM_Cubic, leaveTangent=-0.3012, tangentMode=RCTM_User, time=2, value=1}
          ]
         }}}
       Scale Curve=-500
      )
      1=&FloatFromCurve(
       FloatCurve=DataInterface() {propertyValues {Curve {
          defaultValue=1
          keys [
           {interpMode=RCIM_Cubic, tangentMode=RCTM_User, time=0, value=0}
           {interpMode=RCIM_Cubic, tangentMode=RCTM_User, time=0.1234, value=0.8792}
           {arriveTangent=0.1761, interpMode=RCIM_Cubic, leaveTangent=0.1761, time=0.8212, value=0.0431}
           {arriveTangent=0.8774, interpMode=RCIM_Cubic, leaveTangent=0.8774, time=0.8987, value=1.0157}
           <at 1 0.2 RCIM_Cubic>
           <at 1.0938 1.0157 RCIM_Cubic>
          ]
         }}}
       Scale Curve=&RandomRangeFloat(Minimum=-500, Maximum=-980)
      )
      2=&FloatFromCurve(
       FloatCurve=DataInterface() {propertyValues {Curve {
          defaultValue=1
          keys [
           {interpMode=RCIM_Cubic, tangentMode=RCTM_User, time=0, value=0}
           {interpMode=RCIM_Cubic, tangentMode=RCTM_User, time=0.4, value=0.7}
           {arriveTangent=1.524, interpMode=RCIM_Cubic, leaveTangent=1.524, time=0.573, value=0.1768}
           {arriveTangent=0.7086, interpMode=RCIM_Cubic, leaveTangent=0.7086, time=0.7599, value=1.2485}
           <at 1.121 0.5651 RCIM_Cubic>
          ]
         }}}
       Scale Curve=-150
      )
      3=?
      Index=$Particles.VisibilityTag
      Num=::NewEnumerator1
     )
    ))
   &Drag(
    Drag=&MergeHelper_FloatSwitch(
     0=&FloatFromCurve(
      FloatCurve=DataInterface() {propertyValues {Curve {
         defaultValue=1
         keys [{interpMode=RCIM_Cubic, tangentMode=RCTM_User, time=0, value=1}, {interpMode=RCIM_Cubic, tangentMode=RCTM_User, time=1, value=0}]
        }}}
      Scale Curve=&RandomRangeFloat(Minimum=8, Maximum=10)
     )
     1=&FloatFromCurve(FloatCurve=DataInterface() {propertyValues {Curve {keys [<at 0.8759 1>,<at 1 0>]}}}, Scale Curve=7)
     2=&FloatFromCurve(
      FloatCurve=DataInterface() {propertyValues {Curve {
         defaultValue=1
         keys [{interpMode=RCIM_Cubic, tangentMode=RCTM_User, time=0, value=1}, {interpMode=RCIM_Cubic, tangentMode=RCTM_User, time=1, value=0}]
        }}}
      Scale Curve=5
     )
     3=&FloatFromCurve(
      FloatCurve=DataInterface() {propertyValues {Curve {
         defaultValue=1
         keys [{interpMode=RCIM_Cubic, tangentMode=RCTM_User, time=0, value=1}, {interpMode=RCIM_Cubic, tangentMode=RCTM_User, time=1, value=0}]
        }}}
      Scale Curve=3
     )
     Index=$Particles.VisibilityTag
     Num=::NewEnumerator2
    )
    Use Linear Drag=true
    Use Rotational Drag=false
   )
   &VectorNoiseForce(Force Amount=300, Module.Use Random Noise Time Interval=true)
   &SolveForcesAndVelocity(Clamp Velocity=false, Limit Acceleration=false)
   &ScaleColor(
    Scale Alpha=0
    Scale RGB=&VectorFromFloat(Value=&FNTimeOfDayScalar(Night Scale=0.3, Morning Scale=0.333, Day Scale=0.666, Evening Scale=0.333))
    ScaleA=false
    ScaleRGB=true
   )
   &AlignSpriteToMeshOrientation(SetAlignment=true, SetFacing=true)
   &DynamicMaterialParameters(
    Index 0 Param 1=&RandomRangeFloat(
     Minimum=&FloatFromCurve(
      FloatCurve=DataInterface() {propertyValues {Curve {
         defaultValue=1
         keys [{interpMode=RCIM_Cubic, tangentMode=RCTM_User, time=0, value=1}, {interpMode=RCIM_Cubic, tangentMode=RCTM_User, time=1, value=0}]
        }}}
      Scale Curve=0.5
     )
     Maximum=&FloatFromCurve(FloatCurve=DataInterface() {propertyValues {Curve {
         defaultValue=1
         keys [{interpMode=RCIM_Cubic, tangentMode=RCTM_User, time=0, value=1}, {interpMode=RCIM_Cubic, tangentMode=RCTM_User, time=1, value=0}]
        }}})
    )
    Module.Index0Float0Write=true
    Module.Index0Float1Write=true
    Module.Index0Float2Write=true
    Module.Index0Float3Write=true
   )
  ]
  rendererClasses [&NiagaraMeshRendererProperties]
  renderers [
   {rendererClass=&NiagaraMeshRendererProperties, rendererIndex=0}
   {rendererClass=&NiagaraMeshRendererProperties, rendererIndex=1}
   {rendererClass=&NiagaraMeshRendererProperties, rendererIndex=2}
  ]
 }
 Balloon_explosion_Blue002 {
  bEnabled=false
  emitterUpdateScript [
   &EmitterState()
   &SpawnBurst_Instantaneous(Spawn Count=4, Spawn Group=2, Spawn Time=0.02, Use Spawn Probability=false, UseLoopCountLimit=false) {module=SpawnBurst_Instantaneous002}
   &SpawnBurst_Instantaneous(Spawn Count=7, Spawn Group=1, Spawn Time=0.02, Use Spawn Probability=false, UseLoopCountLimit=false) {module=SpawnBurst_Instantaneous001}
   &SpawnBurst_Instantaneous(Spawn Count=10, Spawn Time=0.02, Use Spawn Probability=false, UseLoopCountLimit=false)
  ]
  particleSpawnScript [
   /SetVariables(Particles.VisibilityTag=&SpawnGroup()) {isSetParameters=true, module=SetVariables_#2}
   /SetVariables(Particles.ExecutionIndex:NiagaraFloat=&ReturnNormalizedExecIndex()) {isSetParameters=true, module=SetVariables_#3}
   &InitializeParticle(
    Color=&MergeHelper_ColorSwitch(
     0=&FNTimeOfDayColor(
      Night Color=<rgba 0.0943 0.268 0.7865 1>
      Morning Color=<rgba 0.0685 0.1946 0.5711 1>
      Day Color=<rgba 0.0685 0.1946 0.5711 1>
      Evening Color=<rgba 0.0685 0.1946 0.5711 1>
     )
     1=&FNTimeOfDayColor(
      Night Color=<rgba 0.0227 0.0844 0.649 1>
      Morning Color=<rgba 0.0307 0.1144 0.8796 1>
      Day Color=<rgba 0.0307 0.1144 0.8796 1>
      Evening Color=<rgba 0.0307 0.1144 0.8796 1>
     )
     2=&FNTimeOfDayColor(
      Night Color=<rgba 0.0328 0.1224 0.9406 1>
      Morning Color=<rgba 0.0685 0.1946 0.5711 1>
      Day Color=<rgba 0.0685 0.1946 0.5711 1>
      Evening Color=<rgba 0.0685 0.1946 0.5711 1>
     )
     3=?
     Index=&SpawnGroup()
     Num=::NewEnumerator1
    )
    Lifetime=&MergeHelper_FloatSwitch(
     0=&RandomRangeFloat(Minimum=0.45, Maximum=0.9)
     1=&RandomRangeFloat(Minimum=0.35, Maximum=0.7)
     2=&RandomRangeFloat(Minimum=0.5, Maximum=0.65)
     3=&RandomRangeFloat(Minimum=0, Maximum=1)
     4=&RandomRangeFloat(Minimum=0, Maximum=1)
     Index=&SpawnGroup()
     Num=::NewEnumerator1
    )
    Mesh Scale Mode=::NewEnumerator3
    Mesh Uniform Scale=&MergeHelper_FloatSwitch(
     0=&RandomRangeFloat(Minimum=0.2, Maximum=0.5)
     1=&RandomRangeFloat(Minimum=0.1, Maximum=0.3)
     2=&RandomRangeFloat(Minimum=0.05, Maximum=0.09)
     3=&MakeFloatFromVector2d(Vector 2D=<5 10>, Channel=::NewEnumerator1)
     Index=$Particles.VisibilityTag
     Num=::NewEnumerator2
    )
    UsePositionOffset=false
    Write Material Random=true
   )
   &ShapeLocation(
    MaskSpawnGroup=false
    Override Seed=false
    Sphere Radius=&MergeHelper_FloatSwitch(0=10, 1=8, 2=5, 3=?, Index=$Particles.VisibilityTag, Num=::NewEnumerator1)
    Sphere Surface Distribution=0.981
    UseOwnerScale=true
   )
   &AddVelocity(
    Module.Use Falloff Distance=false
    Override Seed=false
    Velocity Mode=::NewEnumerator1
    Velocity Speed=&MergeHelper_FloatSwitch(
     0=&RandomRangeFloat(Minimum=666, Maximum=999)
     1=&RandomRangeFloat(Minimum=250, Maximum=350)
     2=&RandomRangeFloat(Minimum=50, Maximum=600)
     3=?
     Index=$Particles.VisibilityTag
     Num=::NewEnumerator1
    )
   )
   &InitialMeshOrientation(
    Mesh Orientation Mode=::NewEnumerator2
    Orientation Coordinate Space=::Local
    Orientation Vector=$Particles.Velocity
    Rotation=<0.125 0 0>
    Use Rotation Vector=true
   )
   /SetVariables(Particles.MeshIndex=&SelectIntFromArray(
     Array Sampling Mode=::NewEnumerator1
     Direct Array Index=?
     Int Selection Array=DataInterface() {propertyValues {IntData [0,1,2,3]}}
     Random Seed=0
    )) {isSetParameters=true, module=SetVariables_#4}
   &SkeletalMeshLocation(
    MaskSpawnGroup=false
    Mesh Sampling Type=::NewEnumerator5
    Orientation Sampling=::NewEnumerator1
    Skeletal Mesh=$System.Skeletal Mesh
    Socket Sampling Mode=::NewEnumerator5
   )
   &CameraOffset(Camera Offset Amount=&RandomRangeFloat(Minimum=3, Maximum=7))
  ]
  particleUpdateScript [
   &ParticleState()
   &ScaleMeshSize(Scale Factor=&VectorFromFloat(Value=&RandomRangeFloat(
      Minimum=&FloatFromCurve(FloatCurve=DataInterface() {propertyValues {Curve {keys [<at 0.35 1>,<at 0.65 0>]}}})
      Maximum=&FloatFromCurve(FloatCurve=DataInterface() {propertyValues {Curve {keys [<at 0.5 1>,<at 0.8 0>]}}})
     )))
   &AerodynamicDrag(
    Aerodynamic Drag:NiagaraFloat=&RandomRangeFloat(Minimum=0, Maximum=0.05)
    Aerodynamic Rotational Drag:NiagaraFloat=&RandomRangeFloat(Minimum=20, Maximum=50)
    Drag Shape Curve=DataInterface() {propertyValues {Curve {keys [
        {interpMode=RCIM_Cubic, tangentMode=RCTM_User, time=0, value=0.1}
        {arriveTangent=2.0799, interpMode=RCIM_Cubic, leaveTangent=2.0799, tangentMode=RCTM_User, time=1, value=1}
       ]}}}
    Lift Contribution=0.5
    Lift Shape Curve=DataInterface() {propertyValues {Curve {keys [
        {interpMode=RCIM_Cubic, tangentMode=RCTM_User, time=0, value=0}
        {interpMode=RCIM_Cubic, tangentMode=RCTM_User, time=0.5, value=1}
        {tangentMode=RCTM_User, time=1, value=0}
       ]}}}
    Pivot Offset Maximum:Vector3f=<5 0.5 1>
    Pivot Offset Minimum:Vector3f=<-5 -0.5 -1>
    Rotation Strength:NiagaraFloat=&RandomRangeFloat(Minimum=0.7, Maximum=3)
    Use Linear Drag=true
    Use Rotational Drag=true
   )
   &GravityForce(Gravity=&MakeVector(
     X=0
     Y=0
     Z=&MergeHelper_FloatSwitch(
      0=&FloatFromCurve(
       FloatCurve=DataInterface() {propertyValues {Curve {
          defaultValue=1
          keys [
           {interpMode=RCIM_Cubic, tangentMode=RCTM_User, time=0, value=0}
           {arriveTangent=0.2683, interpMode=RCIM_Cubic, leaveTangent=0.2683, time=0.7, value=0.7652}
           {arriveTangent=0.143, interpMode=RCIM_Cubic, leaveTangent=0.143, time=1.5434, value=0.414}
           {arriveTangent=-0.3012, interpMode=RCIM_Cubic, leaveTangent=-0.3012, tangentMode=RCTM_User, time=2.0131, value=0.9529}
          ]
         }}}
       Scale Curve=&RandomRangeFloat(Minimum=-450, Maximum=-850)
      )
      1=&FloatFromCurve(
       FloatCurve=DataInterface() {propertyValues {Curve {
          defaultValue=1
          keys [
           {interpMode=RCIM_Cubic, tangentMode=RCTM_User, time=0, value=0}
           {interpMode=RCIM_Cubic, tangentMode=RCTM_User, time=0.7, value=0.9843}
           {arriveTangent=0.1579, interpMode=RCIM_Cubic, leaveTangent=0.1579, time=0.8212, value=0.0431}
           {arriveTangent=-0.1169, interpMode=RCIM_Cubic, leaveTangent=-0.1169, time=0.8987, value=1.0157}
           <at 1.0192 0.02 RCIM_Cubic>
           <at 1.0938 1.0157 RCIM_Cubic>
          ]
         }}}
       Scale Curve=&RandomRangeFloat(Minimum=-400, Maximum=-600)
      )
      2=&FloatFromCurve(
       FloatCurve=DataInterface() {propertyValues {Curve {
          defaultValue=1
          keys [
           {interpMode=RCIM_Cubic, tangentMode=RCTM_User, time=0, value=0}
           {interpMode=RCIM_Cubic, tangentMode=RCTM_User, time=0.4, value=0.7}
           {arriveTangent=1.524, interpMode=RCIM_Cubic, leaveTangent=1.524, time=0.573, value=0.1768}
           {arriveTangent=0.7086, interpMode=RCIM_Cubic, leaveTangent=0.7086, time=0.7599, value=1.2485}
           <at 1.121 0.5651 RCIM_Cubic>
          ]
         }}}
       Scale Curve=-150
      )
      3=?
      Index=$Particles.VisibilityTag
      Num=::NewEnumerator1
     )
    ))
   &Drag(
    Drag=&MergeHelper_FloatSwitch(
     0=&FloatFromCurve(
      FloatCurve=DataInterface() {propertyValues {Curve {
         defaultValue=1
         keys [{interpMode=RCIM_Cubic, tangentMode=RCTM_User, time=0, value=1}, {interpMode=RCIM_Cubic, tangentMode=RCTM_User, time=1, value=0}]
        }}}
      Scale Curve=&RandomRangeFloat(Minimum=6, Maximum=8)
     )
     1=&FloatFromCurve(FloatCurve=DataInterface() {propertyValues {Curve {keys [<at 0.8759 1>,<at 1 0>]}}}, Scale Curve=3)
     2=&FloatFromCurve(
      FloatCurve=DataInterface() {propertyValues {Curve {
         defaultValue=1
         keys [{interpMode=RCIM_Cubic, tangentMode=RCTM_User, time=0, value=1}, {interpMode=RCIM_Cubic, tangentMode=RCTM_User, time=1, value=0}]
        }}}
      Scale Curve=5
     )
     3=&FloatFromCurve(
      FloatCurve=DataInterface() {propertyValues {Curve {
         defaultValue=1
         keys [{interpMode=RCIM_Cubic, tangentMode=RCTM_User, time=0, value=1}, {interpMode=RCIM_Cubic, tangentMode=RCTM_User, time=1, value=0}]
        }}}
      Scale Curve=3
     )
     Index=$Particles.VisibilityTag
     Num=::NewEnumerator2
    )
    Use Linear Drag=true
    Use Rotational Drag=false
   )
   &VectorNoiseForce(Force Amount=300, Module.Use Random Noise Time Interval=true)
   &AddRotationalVelocity(Rotation Rate=&RandomRangeVector(Minimum=<0.02 0.02 0.02>, Maximum=<0.04 0.04 0.04>))
   &SolveForcesAndVelocity(Clamp Velocity=false, Limit Acceleration=false)
   &ScaleColor(
    Scale Alpha=0
    Scale RGB=&VectorFromFloat(Value=&FNTimeOfDayScalar(Night Scale=0.3, Morning Scale=0.333, Day Scale=0.666, Evening Scale=0.333))
    ScaleA=false
    ScaleRGB=true
   )
   &AlignSpriteToMeshOrientation(SetAlignment=true, SetFacing=true)
   &DynamicMaterialParameters(
    Index 0 Param 1=&RandomRangeFloat(
     Minimum=&FloatFromCurve(
      FloatCurve=DataInterface() {propertyValues {Curve {
         defaultValue=1
         keys [{interpMode=RCIM_Cubic, tangentMode=RCTM_User, time=0, value=1}, {interpMode=RCIM_Cubic, tangentMode=RCTM_User, time=1, value=0}]
        }}}
      Scale Curve=0.5
     )
     Maximum=&FloatFromCurve(FloatCurve=DataInterface() {propertyValues {Curve {
         defaultValue=1
         keys [{interpMode=RCIM_Cubic, tangentMode=RCTM_User, time=0, value=1}, {interpMode=RCIM_Cubic, tangentMode=RCTM_User, time=1, value=0}]
        }}})
    )
    Module.Index0Float0Write=true
    Module.Index0Float1Write=true
    Module.Index0Float2Write=true
    Module.Index0Float3Write=true
   )
  ]
  rendererClasses [&NiagaraMeshRendererProperties]
  renderers [
   {rendererClass=&NiagaraMeshRendererProperties, rendererIndex=0}
   {rendererClass=&NiagaraMeshRendererProperties, rendererIndex=1}
   {rendererClass=&NiagaraMeshRendererProperties, rendererIndex=2}
  ]
 }
 Balloon_explosion_Blue003 {
  bEnabled=false
  emitterUpdateScript [
   &EmitterState()
   &SpawnBurst_Instantaneous(Spawn Count=4, Spawn Group=2, Spawn Time=0.05, Use Spawn Probability=false, UseLoopCountLimit=false) {module=SpawnBurst_Instantaneous002}
   &SpawnBurst_Instantaneous(Spawn Count=3, Spawn Group=1, Spawn Time=0.05, Use Spawn Probability=false, UseLoopCountLimit=false) {module=SpawnBurst_Instantaneous001}
   &SpawnBurst_Instantaneous(Spawn Count=7, Spawn Time=0.02, Use Spawn Probability=false, UseLoopCountLimit=false)
  ]
  particleSpawnScript [
   /SetVariables(Particles.VisibilityTag=&SpawnGroup()) {isSetParameters=true, module=SetVariables_#2}
   /SetVariables(Particles.ExecutionIndex:NiagaraFloat=&ReturnNormalizedExecIndex()) {isSetParameters=true, module=SetVariables_#3}
   &InitializeParticle(
    Color=&MergeHelper_ColorSwitch(0=<rgba 0.227 0.0513 0.0529 1>, 1=<rgba 1 1 1 1>, 2=<rgba 0 0 0 1>, 3=?, Index=$Particles.VisibilityTag, Num=::NewEnumerator1)
    Lifetime=&MergeHelper_FloatSwitch(
     0=&RandomRangeFloat(Minimum=0.35, Maximum=0.7)
     1=&RandomRangeFloat(Minimum=0.65, Maximum=0.85)
     2=&RandomRangeFloat(Minimum=0.25, Maximum=0.6)
     3=?
     Index=$Particles.VisibilityTag
     Num=::NewEnumerator1
    )
    Mesh Scale Mode=::NewEnumerator3
    Mesh Uniform Scale=&MergeHelper_FloatSwitch(
     0=&RandomRangeFloat(Minimum=0.35, Maximum=0.5)
     1=&RandomRangeFloat(Minimum=0.2, Maximum=0.3)
     2=&RandomRangeFloat(Minimum=0.08, Maximum=0.15)
     3=&MakeFloatFromVector2d(Vector 2D=<5 10>, Channel=::NewEnumerator1)
     Index=$Particles.VisibilityTag
     Num=::NewEnumerator1
    )
    UsePositionOffset=false
    Write Material Random=true
   )
   &AddVelocity(
    Cone Angle=90
    Cone Axis=<1 0 0.5>
    Inner Cone Angle=50
    Module.UseConeFalloff=true
    Module.UseVelDistribution=true
    Override Seed=false
    Speed Falloff From Cone Axis:NiagaraFloat=&RandomRangeFloat(Minimum=0.1, Maximum=0.7)
    Velocity Mode=::NewEnumerator2
    Velocity Speed=&MergeHelper_FloatSwitch(
     0=&RandomRangeFloat(Minimum=555, Maximum=888)
     1=&RandomRangeFloat(Minimum=250, Maximum=550)
     2=&RandomRangeFloat(Minimum=50, Maximum=450)
     3=?
     Index=$Particles.VisibilityTag
     Num=::NewEnumerator1
    )
   )
   &ShapeLocation(
    MaskSpawnGroup=false
    Override Seed=false
    Scale Mode:ENiagara_ScaleMode=::NewEnumerator1
    Sphere Radius=&MergeHelper_FloatSwitch(0=3, 1=7, 2=5, 3=?, Index=$Particles.VisibilityTag, Num=::NewEnumerator1)
    Sphere Surface Distribution=0.981
   )
   &InitialMeshOrientation(
    Mesh Orientation Mode=::NewEnumerator2
    Orientation Coordinate Space=::Local
    Orientation Vector=$Particles.Velocity
    Rotation=<0.125 0 0>
    Use Rotation Vector=true
   )
   /SetVariables(Particles.MeshIndex=&SelectIntFromArray(
     Array Sampling Mode=::NewEnumerator1
     Direct Array Index=?
     Int Selection Array=DataInterface() {propertyValues {IntData [0,1,2,3]}}
     Random Seed=0
    )) {isSetParameters=true, module=SetVariables_#4}
   &SkeletalMeshLocation(
    MaskSpawnGroup=false
    Mesh Sampling Type=::NewEnumerator5
    Orientation Sampling=::NewEnumerator1
    Sampled Position Offset=<0 0 15>
    Skeletal Mesh=$System.Skeletal Mesh
    Socket Sampling Mode=::NewEnumerator5
   )
  ]
  particleUpdateScript [
   &ParticleState()
   /SetVariables(Particles.Update scale:Vector3f=&MergeHelper_VectorSwitch(
     0=&VectorFromFloat(Value=&RandomRangeFloat(
       Minimum=&FloatFromCurve(FloatCurve=DataInterface() {propertyValues {Curve {keys [<at 0.15 1>,<at 0.45 0>]}}}, Scale Curve=0.3)
       Maximum=&FloatFromCurve(FloatCurve=DataInterface() {propertyValues {Curve {keys [<at 0.65 1>,<at 0.8 0>]}}}, Scale Curve=0.6)
      ))
     1=&VectorFromFloat(Value=&RandomRangeFloat(
       Minimum=&FloatFromCurve(FloatCurve=DataInterface() {propertyValues {Curve {keys [<at 0.15 1>,<at 0.35 0>]}}}, Scale Curve=0.55)
       Maximum=&FloatFromCurve(FloatCurve=DataInterface() {propertyValues {Curve {keys [<at 0.5 1>,<at 0.85 0>]}}}, Scale Curve=0.75)
      ))
     2=&VectorFromFloat(Value=&RandomRangeFloat(
       Minimum=&FloatFromCurve(FloatCurve=DataInterface() {propertyValues {Curve {keys [<at 0.45 1>,<at 0.88 0>]}}}, Scale Curve=0.2)
       Maximum=&FloatFromCurve(FloatCurve=DataInterface() {propertyValues {Curve {keys [<at 0.5 1>,<at 0.95 0>]}}}, Scale Curve=0.4)
      ))
     3=&VectorFromFloat(Value=&RandomRangeFloat(
       Minimum=&FloatFromCurve(FloatCurve=DataInterface() {propertyValues {Curve {keys [<at 0.35 1>,<at 1 0>]}}})
       Maximum=&FloatFromCurve(FloatCurve=DataInterface() {propertyValues {Curve {keys [<at 0.5 1>,<at 1 0>]}}})
      ))
     4=&VectorFromFloat(Value=&RandomRangeFloat(
       Minimum=&FloatFromCurve(FloatCurve=DataInterface() {propertyValues {Curve {keys [<at 0.35 1>,<at 1 0>]}}})
       Maximum=&FloatFromCurve(FloatCurve=DataInterface() {propertyValues {Curve {keys [<at 0.5 1>,<at 1 0>]}}})
      ))
     5=&VectorFromFloat(Value=&RandomRangeFloat(
       Minimum=&FloatFromCurve(FloatCurve=DataInterface() {propertyValues {Curve {keys [<at 0.35 1>,<at 1 0>]}}})
       Maximum=&FloatFromCurve(FloatCurve=DataInterface() {propertyValues {Curve {keys [<at 0.5 1>,<at 1 0>]}}})
      ))
     Index=&SpawnGroup()
     Num=::NewEnumerator1
    )) {isSetParameters=true, module=SetVariables_#5}
   &ScaleMeshSize(Scale Factor=&VectorFromFloat(Value=&RandomRangeFloat(
      Minimum=&FloatFromCurve(FloatCurve=DataInterface() {propertyValues {Curve {keys [<at 0.3 1>,<at 1 0>]}}})
      Maximum=&FloatFromCurve(FloatCurve=DataInterface() {propertyValues {Curve {keys [<at 0.5 1>,<at 1 0>]}}})
     ))) {enabled=false}
   &AerodynamicDrag(
    Aerodynamic Drag:NiagaraFloat=&RandomRangeFloat(Minimum=0, Maximum=0.05)
    Aerodynamic Rotational Drag:NiagaraFloat=&RandomRangeFloat(Minimum=0, Maximum=25)
    Drag Shape Curve=DataInterface() {propertyValues {Curve {keys [
        {interpMode=RCIM_Cubic, tangentMode=RCTM_User, time=0, value=0.1}
        {arriveTangent=2.0799, interpMode=RCIM_Cubic, leaveTangent=2.0799, tangentMode=RCTM_User, time=1, value=1}
       ]}}}
    Lift Contribution=0.5
    Lift Shape Curve=DataInterface() {propertyValues {Curve {keys [
        {interpMode=RCIM_Cubic, tangentMode=RCTM_User, time=0, value=0}
        {interpMode=RCIM_Cubic, tangentMode=RCTM_User, time=0.5, value=1}
        {tangentMode=RCTM_User, time=1, value=0}
       ]}}}
    Pivot Offset Maximum:Vector3f=<5 0.5 1>
    Pivot Offset Minimum:Vector3f=<-5 -0.5 -1>
    Rotation Strength:NiagaraFloat=&RandomRangeFloat(Minimum=0.7, Maximum=1)
    Use Linear Drag=true
    Use Rotational Drag=true
   )
   &GravityForce(Gravity=&MakeVector(
     X=0
     Y=0
     Z=&MergeHelper_FloatSwitch(
      0=&FloatFromCurve(
       FloatCurve=DataInterface() {propertyValues {Curve {
          defaultValue=1
          keys [
           {interpMode=RCIM_Cubic, tangentMode=RCTM_User, time=0, value=0}
           {arriveTangent=0.2683, interpMode=RCIM_Cubic, leaveTangent=0.2683, time=0.75, value=0.7652}
           {arriveTangent=0.1878, interpMode=RCIM_Cubic, leaveTangent=0.1878, time=1.5434, value=0.414}
           {arriveTangent=-0.3012, interpMode=RCIM_Cubic, leaveTangent=-0.3012, tangentMode=RCTM_User, time=2, value=1}
          ]
         }}}
       Scale Curve=-500
      )
      1=&FloatFromCurve(
       FloatCurve=DataInterface() {propertyValues {Curve {
          defaultValue=1
          keys [
           {interpMode=RCIM_Cubic, tangentMode=RCTM_User, time=0, value=0}
           {interpMode=RCIM_Cubic, tangentMode=RCTM_User, time=0.1234, value=0.8792}
           {arriveTangent=0.1761, interpMode=RCIM_Cubic, leaveTangent=0.1761, time=0.8212, value=0.0431}
           {arriveTangent=0.8774, interpMode=RCIM_Cubic, leaveTangent=0.8774, time=0.8987, value=1.0157}
           <at 1 0.2 RCIM_Cubic>
           <at 1.0938 1.0157 RCIM_Cubic>
          ]
         }}}
       Scale Curve=&RandomRangeFloat(Minimum=-500, Maximum=-980)
      )
      2=&FloatFromCurve(
       FloatCurve=DataInterface() {propertyValues {Curve {
          defaultValue=1
          keys [
           {interpMode=RCIM_Cubic, tangentMode=RCTM_User, time=0, value=0}
           {interpMode=RCIM_Cubic, tangentMode=RCTM_User, time=0.4, value=0.7}
           {arriveTangent=1.524, interpMode=RCIM_Cubic, leaveTangent=1.524, time=0.573, value=0.1768}
           {arriveTangent=0.7086, interpMode=RCIM_Cubic, leaveTangent=0.7086, time=0.7599, value=1.2485}
           <at 1.121 0.5651 RCIM_Cubic>
          ]
         }}}
       Scale Curve=-150
      )
      3=?
      Index=$Particles.VisibilityTag
      Num=::NewEnumerator1
     )
    ))
   &Drag(
    Drag=&MergeHelper_FloatSwitch(
     0=&FloatFromCurve(
      FloatCurve=DataInterface() {propertyValues {Curve {
         defaultValue=1
         keys [{interpMode=RCIM_Cubic, tangentMode=RCTM_User, time=0, value=1}, {interpMode=RCIM_Cubic, tangentMode=RCTM_User, time=1, value=0}]
        }}}
      Scale Curve=&RandomRangeFloat(Minimum=4, Maximum=6)
     )
     1=&FloatFromCurve(
      FloatCurve=DataInterface() {propertyValues {Curve {keys [<at 0.8759 1>,<at 1 0>]}}}
      Scale Curve=&RandomRangeFloat(Minimum=2, Maximum=5)
     )
     2=&FloatFromCurve(
      FloatCurve=DataInterface() {propertyValues {Curve {
         defaultValue=1
         keys [{interpMode=RCIM_Cubic, tangentMode=RCTM_User, time=0, value=1}, {interpMode=RCIM_Cubic, tangentMode=RCTM_User, time=1, value=0}]
        }}}
      Scale Curve=5
     )
     3=&FloatFromCurve(
      FloatCurve=DataInterface() {propertyValues {Curve {
         defaultValue=1
         keys [{interpMode=RCIM_Cubic, tangentMode=RCTM_User, time=0, value=1}, {interpMode=RCIM_Cubic, tangentMode=RCTM_User, time=1, value=0}]
        }}}
      Scale Curve=3
     )
     Index=$Particles.VisibilityTag
     Num=::NewEnumerator2
    )
    Use Linear Drag=true
    Use Rotational Drag=false
   )
   &VectorNoiseForce(Force Amount=300, Module.Use Random Noise Time Interval=true)
   &SolveForcesAndVelocity(Clamp Velocity=false, Limit Acceleration=false)
   &ScaleColor(
    Scale Alpha=0
    Scale RGB=&VectorFromFloat(Value=&FNTimeOfDayScalar(Night Scale=0.3, Morning Scale=0.333, Day Scale=0.666, Evening Scale=0.333))
    ScaleA=false
    ScaleRGB=true
   )
   &AlignSpriteToMeshOrientation(SetAlignment=true, SetFacing=true)
   &DynamicMaterialParameters(
    Index 0 Param 1=&RandomRangeFloat(
     Minimum=&FloatFromCurve(
      FloatCurve=DataInterface() {propertyValues {Curve {
         defaultValue=1
         keys [{interpMode=RCIM_Cubic, tangentMode=RCTM_User, time=0, value=1}, {interpMode=RCIM_Cubic, tangentMode=RCTM_User, time=1, value=0}]
        }}}
      Scale Curve=0.5
     )
     Maximum=&FloatFromCurve(FloatCurve=DataInterface() {propertyValues {Curve {
         defaultValue=1
         keys [{interpMode=RCIM_Cubic, tangentMode=RCTM_User, time=0, value=1}, {interpMode=RCIM_Cubic, tangentMode=RCTM_User, time=1, value=0}]
        }}})
    )
    Module.Index0Float0Write=true
    Module.Index0Float1Write=true
    Module.Index0Float2Write=true
    Module.Index0Float3Write=true
   )
  ]
  rendererClasses [&NiagaraMeshRendererProperties]
  renderers [
   {rendererClass=&NiagaraMeshRendererProperties, rendererIndex=0}
   {rendererClass=&NiagaraMeshRendererProperties, rendererIndex=1}
   {rendererClass=&NiagaraMeshRendererProperties, rendererIndex=2}
  ]
 }
 Balloon_explosion_Blue004 {
  emitterUpdateScript [
   &EmitterState()
   &SpawnBurst_Instantaneous(Spawn Group=3, Spawn Time=0.01, Use Spawn Probability=false, UseLoopCountLimit=false) {module=SpawnBurst_Instantaneous003}
   &SpawnBurst_Instantaneous(Spawn Count=3, Spawn Group=2, Spawn Time=0.01, Use Spawn Probability=false, UseLoopCountLimit=false) {module=SpawnBurst_Instantaneous002}
   &SpawnBurst_Instantaneous(Spawn Count=5, Spawn Group=1, Spawn Time=0.011, Use Spawn Probability=false, UseLoopCountLimit=false) {module=SpawnBurst_Instantaneous001}
   &SpawnBurst_Instantaneous(Spawn Count=7, Use Spawn Probability=false, UseLoopCountLimit=false)
  ]
  particleSpawnScript [
   /SetVariables(Particles.VisibilityTag=&SpawnGroup()) {isSetParameters=true, module=SetVariables_#2}
   /SetVariables(Particles.ExecutionIndex:NiagaraFloat=&ReturnNormalizedExecIndex()) {isSetParameters=true, module=SetVariables_#3}
   &InitializeParticle(
    Color=&MergeHelper_ColorSwitch(
     0=&FNTimeOfDayColor(
      Night Color=<rgba 0.0943 0.268 0.7865 1>
      Morning Color=<rgba 0.0685 0.1946 0.5711 1>
      Day Color=<rgba 0.0685 0.1946 0.5711 1>
      Evening Color=<rgba 0.0685 0.1946 0.5711 1>
     )
     1=&FNTimeOfDayColor(
      Night Color=<rgba 0.0227 0.0844 0.649 1>
      Morning Color=<rgba 0.0307 0.1144 0.8796 1>
      Day Color=<rgba 0.0307 0.1144 0.8796 1>
      Evening Color=<rgba 0.0307 0.1144 0.8796 1>
     )
     2=&FNTimeOfDayColor(
      Night Color=<rgba 0.0328 0.1224 0.9406 1>
      Morning Color=<rgba 0.0685 0.1946 0.5711 1>
      Day Color=<rgba 0.0685 0.1946 0.5711 1>
      Evening Color=<rgba 0.0685 0.1946 0.5711 1>
     )
     3=?
     Index=&SpawnGroup()
     Num=::NewEnumerator1
    )
    Lifetime=&MergeHelper_FloatSwitch(
     0=&RandomRangeFloat(Minimum=0.33, Maximum=0.57)
     1=&RandomRangeFloat(Minimum=0.15, Maximum=0.35)
     2=&RandomRangeFloat(Minimum=0.25, Maximum=0.43)
     3=&RandomRangeFloat(Minimum=0.2, Maximum=0.51)
     4=&RandomRangeFloat(Minimum=0, Maximum=1)
     Index=&SpawnGroup()
     Num=::NewEnumerator2
    )
    Mass Mode=::NewEnumerator3
    Mesh Scale Mode=::NewEnumerator3
    Mesh Uniform Scale=&MergeHelper_FloatSwitch(
     0=&RandomRangeFloat(Minimum=0.25, Maximum=0.5)
     1=&RandomRangeFloat(Minimum=0.25, Maximum=0.35)
     2=&RandomRangeFloat(Minimum=0.05, Maximum=0.1)
     3=&MakeFloatFromVector2d(Vector 2D=<5 10>, Channel=::NewEnumerator1)
     Index=&SpawnGroup()
     Num=::NewEnumerator2
    )
    Position Mode=::NewEnumerator1
    UsePositionOffset=false
    Write Material Random=true
   )
   &ShapeLocation(
    MaskSpawnGroup=false
    Override Seed=true
    Sphere Radius=&MergeHelper_FloatSwitch(0=57, 1=45, 2=55, 3=0, Index=&SpawnGroup(), Num=::NewEnumerator2)
    UseOwnerScale=true
   )
   &ApplyInitialForces(Apply Force to Position=true)
   &AddVelocity(
    Module.Use Falloff Distance=false
    Override Seed=false
    Velocity Mode=::NewEnumerator1
    Velocity Speed=&MergeHelper_FloatSwitch(
     0=&RandomRangeFloat(Minimum=555, Maximum=888)
     1=&RandomRangeFloat(Minimum=350, Maximum=555)
     2=&RandomRangeFloat(Minimum=350, Maximum=500)
     3=&RandomRangeFloat(Minimum=555, Maximum=888)
     Index=$Particles.VisibilityTag
     Num=::NewEnumerator2
    )
   )
   &InitialMeshOrientation(
    Mesh Orientation Mode=::NewEnumerator2
    Orientation Coordinate Space=::Local
    Orientation Vector=$Particles.Velocity
    Rotation=<0.125 0 0>
    Use Rotation Vector=true
   )
   &CameraOffset(Camera Offset Amount=&RandomRangeFloat(Minimum=3, Maximum=7))
   /SetVariables(Particles.MeshIndex=&SelectIntFromArray(
     Array Sampling Mode=::NewEnumerator1
     Direct Array Index=?
     Int Selection Array=DataInterface() {propertyValues {IntData [0,1,2,3]}}
     Random Seed=0
    )) {isSetParameters=true}
  ]
  particleUpdateScript [
   &ParticleState()
   /SetVariables(Particles.Update scale:Vector3f=&MergeHelper_VectorSwitch(
     0=&VectorFromFloat(Value=&RandomRangeFloat(
       Minimum=&FloatFromCurve(FloatCurve=DataInterface() {propertyValues {Curve {keys [<at 0 0>,<at 0.0314 0.9303>,<at 0.45 0>]}}}, Scale Curve=0.3)
       Maximum=&FloatFromCurve(FloatCurve=DataInterface() {propertyValues {Curve {keys [<at 0.6484 1>,<at 0.65 1>,<at 0.85 0>]}}}, Scale Curve=0.5)
      ))
     1=&VectorFromFloat(Value=&RandomRangeFloat(
       Minimum=&FloatFromCurve(
        FloatCurve=DataInterface() {propertyValues {Curve {
           defaultValue=1
           keys [{interpMode=RCIM_Cubic, tangentMode=RCTM_User, time=0, value=1}, {interpMode=RCIM_Cubic, tangentMode=RCTM_User, time=0.5, value=0}]
          }}}
        Scale Curve=0.5
       )
       Maximum=&FloatFromCurve(FloatCurve=DataInterface() {propertyValues {Curve {
           defaultValue=1
           keys [{interpMode=RCIM_Cubic, tangentMode=RCTM_User, time=0.3, value=1}, {interpMode=RCIM_Cubic, tangentMode=RCTM_User, time=0.7, value=0}]
          }}})
      ))
     2=&VectorFromFloat(Value=&RandomRangeFloat(
       Minimum=&FloatFromCurve(FloatCurve=DataInterface() {propertyValues {Curve {keys [<at 0.45 1>,<at 0.88 0>]}}}, Scale Curve=0.1)
       Maximum=&FloatFromCurve(FloatCurve=DataInterface() {propertyValues {Curve {keys [<at 0.5 1>,<at 0.95 0>]}}}, Scale Curve=0.2)
      ))
     3=&VectorFromFloat(Value=&FloatFromCurve(FloatCurve=DataInterface() {propertyValues {Curve {keys [<at 0.35 1>,<at 1 0>]}}}))
     4=&VectorFromFloat(Value=&RandomRangeFloat(
       Minimum=&FloatFromCurve(FloatCurve=DataInterface() {propertyValues {Curve {keys [<at 0.35 1>,<at 1 0>]}}})
       Maximum=&FloatFromCurve(FloatCurve=DataInterface() {propertyValues {Curve {keys [<at 0.5 1>,<at 1 0>]}}})
      ))
     5=&VectorFromFloat(Value=&RandomRangeFloat(
       Minimum=&FloatFromCurve(FloatCurve=DataInterface() {propertyValues {Curve {keys [<at 0.35 1>,<at 1 0>]}}})
       Maximum=&FloatFromCurve(FloatCurve=DataInterface() {propertyValues {Curve {keys [<at 0.5 1>,<at 1 0>]}}})
      ))
     Index=&SpawnGroup()
     Num=::NewEnumerator2
    )) {isSetParameters=true}
   &AerodynamicDrag(
    Aerodynamic Drag:NiagaraFloat=&RandomRangeFloat(Minimum=0, Maximum=0.05)
    Aerodynamic Rotational Drag:NiagaraFloat=&RandomRangeFloat(Minimum=20, Maximum=50)
    Drag Shape Curve=DataInterface() {propertyValues {Curve {keys [
        {interpMode=RCIM_Cubic, tangentMode=RCTM_User, time=0, value=0.1}
        {arriveTangent=2.0799, interpMode=RCIM_Cubic, leaveTangent=2.0799, tangentMode=RCTM_User, time=1, value=1}
       ]}}}
    Lift Contribution=0.5
    Lift Shape Curve=DataInterface() {propertyValues {Curve {keys [
        {interpMode=RCIM_Cubic, tangentMode=RCTM_User, time=0, value=0}
        {interpMode=RCIM_Cubic, tangentMode=RCTM_User, time=0.5, value=1}
        {tangentMode=RCTM_User, time=1, value=0}
       ]}}}
    Pivot Offset Maximum:Vector3f=<5 0.5 1>
    Pivot Offset Minimum:Vector3f=<-5 -0.5 -1>
    Rotation Strength:NiagaraFloat=&RandomRangeFloat(Minimum=0.7, Maximum=3)
    Use Linear Drag=true
    Use Rotational Drag=true
   )
   &GravityForce(Gravity=&MakeVector(
     X=0
     Y=0
     Z=&MergeHelper_FloatSwitch(
      0=&FloatFromCurve(
       FloatCurve=DataInterface() {propertyValues {Curve {
          defaultValue=1
          keys [
           {interpMode=RCIM_Cubic, tangentMode=RCTM_User, time=0, value=1.5}
           {arriveTangent=-1, interpMode=RCIM_Cubic, leaveTangent=-1, time=0.2274, value=1.0402}
           {arriveTangent=-0.2564, interpMode=RCIM_Cubic, leaveTangent=-0.2564, tangentMode=RCTM_User, time=0.5, value=1}
           {arriveTangent=-0.3012, interpMode=RCIM_Cubic, leaveTangent=-0.3012, tangentMode=RCTM_User, time=1, value=0.1}
          ]
         }}}
       Scale Curve=&RandomRangeFloat(Minimum=-850, Maximum=-1300)
      )
      1=&FloatFromCurve(
       FloatCurve=DataInterface() {propertyValues {Curve {
          defaultValue=1
          keys [
           {interpMode=RCIM_Cubic, tangentMode=RCTM_User, time=0, value=0}
           {interpMode=RCIM_Cubic, tangentMode=RCTM_User, time=0.35, value=0.5}
           {arriveTangent=0.9398, interpMode=RCIM_Cubic, leaveTangent=0.9398, time=0.75, value=0.3}
           {arriveTangent=-1.04, interpMode=RCIM_Cubic, leaveTangent=-1.04, time=0.8987, value=1.0157}
           <at 1.0192 0.02 RCIM_Cubic>
           <at 1.0938 1.0157 RCIM_Cubic>
          ]
         }}}
       Scale Curve=&RandomRangeFloat(Minimum=-666, Maximum=-1000)
      )
      2=&FloatFromCurve(
       FloatCurve=DataInterface() {propertyValues {Curve {
          defaultValue=1
          keys [
           {interpMode=RCIM_Cubic, tangentMode=RCTM_User, time=0, value=0}
           {interpMode=RCIM_Cubic, tangentMode=RCTM_User, time=0.25, value=1.5}
           {arriveTangent=-0.4932, interpMode=RCIM_Cubic, leaveTangent=-0.4932, time=0.5, value=0.1768}
           {arriveTangent=0.6253, interpMode=RCIM_Cubic, leaveTangent=0.6253, time=0.7599, value=1.2485}
           <at 1.121 0.5651 RCIM_Cubic>
          ]
         }}}
       Scale Curve=&RandomRangeFloat(Minimum=-250, Maximum=-500)
      )
      3=&FloatFromCurve(
       FloatCurve=DataInterface() {propertyValues {Curve {
          defaultValue=1
          keys [
           {interpMode=RCIM_Cubic, tangentMode=RCTM_User, time=0, value=1.5}
           {arriveTangent=-1, interpMode=RCIM_Cubic, leaveTangent=-1, time=0.2274, value=1.0402}
           {arriveTangent=-0.2564, interpMode=RCIM_Cubic, leaveTangent=-0.2564, tangentMode=RCTM_User, time=0.5, value=1}
           {arriveTangent=-0.3012, interpMode=RCIM_Cubic, leaveTangent=-0.3012, tangentMode=RCTM_User, time=1, value=0.1}
          ]
         }}}
       Scale Curve=&RandomRangeFloat(Minimum=-850, Maximum=-1300)
      )
      Index=$Particles.VisibilityTag
      Num=::NewEnumerator2
     )
    ))
   &Drag(
    Drag=&MergeHelper_FloatSwitch(
     0=&FloatFromCurve(
      FloatCurve=DataInterface() {propertyValues {Curve {
         defaultValue=1
         keys [{interpMode=RCIM_Cubic, tangentMode=RCTM_User, time=0, value=1}, {interpMode=RCIM_Cubic, tangentMode=RCTM_User, time=0.8024, value=0}]
        }}}
      Scale Curve=&RandomRangeFloat(Minimum=6, Maximum=8)
     )
     1=&FloatFromCurve(
      FloatCurve=DataInterface() {propertyValues {Curve {
         defaultValue=0
         keys [{interpMode=RCIM_Cubic, tangentMode=RCTM_User, time=0, value=0}, {interpMode=RCIM_Cubic, tangentMode=RCTM_User, time=1, value=1.3}]
        }}}
      Scale Curve=&RandomRangeFloat(Minimum=2.5, Maximum=5.5)
     )
     2=&FloatFromCurve(
      FloatCurve=DataInterface() {propertyValues {Curve {
         defaultValue=1
         keys [{interpMode=RCIM_Cubic, tangentMode=RCTM_User, time=0, value=0.6}, {interpMode=RCIM_Cubic, tangentMode=RCTM_User, time=1.5, value=0}]
        }}}
      Scale Curve=5
     )
     3=&FloatFromCurve(
      FloatCurve=DataInterface() {propertyValues {Curve {
         defaultValue=1
         keys [{interpMode=RCIM_Cubic, tangentMode=RCTM_User, time=0, value=1}, {interpMode=RCIM_Cubic, tangentMode=RCTM_User, time=1, value=0}]
        }}}
      Scale Curve=3
     )
     Index=$Particles.VisibilityTag
     Num=::NewEnumerator2
    )
    Use Linear Drag=true
    Use Rotational Drag=false
   )
   &VectorNoiseForce(Force Amount=300, Module.Use Random Noise Time Interval=true)
   &AddRotationalVelocity(Rotation Rate=&RandomRangeVector(Minimum=<0.02 0.02 0>, Maximum=<0.1 0.1 0.1>))
   &SolveForcesAndVelocity(Clamp Velocity=false, Limit Acceleration=false)
   &ScaleMeshSizeBySpeed(
    Min Scale Factor:Vector3f=<0 0 0>
    Sample Scale Factor By Curve=true
    Scale Factor Curve=DataInterface() {propertyValues {bExposeCurve=true, Curve {defaultValue=0, keys [{tangentMode=RCTM_User, time=0, value=1}, {tangentMode=RCTM_User, time=1, value=0}]}}}
   )
   &ScaleColor(
    Scale Alpha=0
    Scale RGB=&VectorFromFloat(Value=&FNTimeOfDayScalar(Night Scale=0.3, Morning Scale=0.333, Day Scale=0.666, Evening Scale=0.333))
    ScaleA=false
    ScaleRGB=true
   )
   &AlignSpriteToMeshOrientation(SetAlignment=true, SetFacing=true)
   &DynamicMaterialParameters(
    Index 0 Param 1=&RandomRangeFloat(
     Minimum=&FloatFromCurve(
      FloatCurve=DataInterface() {propertyValues {Curve {
         defaultValue=1
         keys [{interpMode=RCIM_Cubic, tangentMode=RCTM_User, time=0, value=1}, {interpMode=RCIM_Cubic, tangentMode=RCTM_User, time=1, value=0}]
        }}}
      Scale Curve=0.5
     )
     Maximum=&FloatFromCurve(FloatCurve=DataInterface() {propertyValues {Curve {
         defaultValue=1
         keys [{interpMode=RCIM_Cubic, tangentMode=RCTM_User, time=0, value=1}, {interpMode=RCIM_Cubic, tangentMode=RCTM_User, time=1, value=0}]
        }}})
    )
    Module.Index0Float0Write=true
    Module.Index0Float1Write=true
    Module.Index0Float2Write=true
    Module.Index0Float3Write=true
   )
  ]
  rendererClasses [&NiagaraMeshRendererProperties]
  renderers [
   {rendererClass=&NiagaraMeshRendererProperties, rendererIndex=0}
   {rendererClass=&NiagaraMeshRendererProperties, rendererIndex=1}
   {rendererClass=&NiagaraMeshRendererProperties, rendererIndex=2}
   {rendererClass=&NiagaraMeshRendererProperties, rendererIndex=3}
  ]
 }
 Balloon_explosion_Blue005 {
  emitterUpdateScript [
   &EmitterState()
   &SpawnBurst_Instantaneous(Spawn Count=4, Spawn Group=2, Spawn Time=0.053, Use Spawn Probability=false, UseLoopCountLimit=false) {module=SpawnBurst_Instantaneous002}
   &SpawnBurst_Instantaneous(Spawn Count=3, Spawn Group=1, Spawn Time=0.051, Use Spawn Probability=false, UseLoopCountLimit=false) {module=SpawnBurst_Instantaneous001}
   &SpawnBurst_Instantaneous(Spawn Count=3, Spawn Time=0.025, Use Spawn Probability=false, UseLoopCountLimit=false)
  ]
  particleSpawnScript [
   /SetVariables(Particles.VisibilityTag=&SpawnGroup()) {isSetParameters=true, module=SetVariables_#2}
   /SetVariables(Particles.ExecutionIndex:NiagaraFloat=&ReturnNormalizedExecIndex()) {isSetParameters=true, module=SetVariables_#3}
   &InitializeParticle(
    Color=&MergeHelper_ColorSwitch(0=<rgba 0.227 0.0513 0.0529 1>, 1=<rgba 1 1 1 1>, 2=<rgba 0 0 0 1>, 3=?, Index=$Particles.VisibilityTag, Num=::NewEnumerator1)
    Lifetime=&MergeHelper_FloatSwitch(
     0=&RandomRangeFloat(Minimum=0.21, Maximum=0.53)
     1=&RandomRangeFloat(Minimum=0.37, Maximum=0.4)
     2=&RandomRangeFloat(Minimum=0.25, Maximum=0.6)
     3=?
     Index=$Particles.VisibilityTag
     Num=::NewEnumerator1
    )
    Mass Max=1.5
    Mass Mode=::NewEnumerator3
    Mesh Scale Mode=::NewEnumerator3
    Mesh Uniform Scale=&MergeHelper_FloatSwitch(
     0=&RandomRangeFloat(Minimum=0.35, Maximum=0.5)
     1=&RandomRangeFloat(Minimum=0.2, Maximum=0.3)
     2=&RandomRangeFloat(Minimum=0.08, Maximum=0.15)
     3=&MakeFloatFromVector2d(Vector 2D=<5 10>, Channel=::NewEnumerator1)
     Index=$Particles.VisibilityTag
     Num=::NewEnumerator1
    )
    UsePositionOffset=false
    Write Material Random=true
   )
   &ShapeLocation(
    MaskSpawnGroup=false
    Override Seed=false
    Scale Mode:ENiagara_ScaleMode=::NewEnumerator1
    Sphere Radius=&MergeHelper_FloatSwitch(0=25, 1=15, 2=5, 3=?, Index=$Particles.VisibilityTag, Num=::NewEnumerator1)
    Sphere Surface Distribution=0.981
   )
   &AddVelocity(
    Cone Angle=90
    Cone Axis=<1 0 0.5>
    Distribution Along Cone Axis=0.3
    Inner Cone Angle=50
    Module.UseConeFalloff=true
    Module.UseVelDistribution=true
    Override Seed=false
    Speed Falloff From Cone Axis:NiagaraFloat=&RandomRangeFloat(Minimum=0.1, Maximum=0.7)
    Velocity Mode=::NewEnumerator2
    Velocity Speed=&MergeHelper_FloatSwitch(
     0=&RandomRangeFloat(Minimum=444, Maximum=888)
     1=&RandomRangeFloat(Minimum=400, Maximum=550)
     2=&RandomRangeFloat(Minimum=333, Maximum=600)
     3=?
     Index=$Particles.VisibilityTag
     Num=::NewEnumerator1
    )
   )
   &AddVelocity(Velocity=&RandomRangeVector(Minimum=<10 -10 -50>, Maximum=<50 10 0>)) {module=AddVelocity001}
   &InitialMeshOrientation(
    Mesh Orientation Mode=::NewEnumerator2
    Orientation Coordinate Space=::Local
    Orientation Vector=$Particles.Velocity
    Rotation=<0.125 0 0>
    Use Rotation Vector=true
   )
   /SetVariables(Particles.MeshIndex=&SelectIntFromArray(
     Array Sampling Mode=::NewEnumerator1
     Direct Array Index=?
     Int Selection Array=DataInterface() {propertyValues {IntData [0,1,2,3]}}
     Random Seed=0
    )) {isSetParameters=true, module=SetVariables_#4}
  ]
  particleUpdateScript [
   &ParticleState()
   /SetVariables(Particles.Update scale:Vector3f=&MergeHelper_VectorSwitch(
     0=&VectorFromFloat(Value=&RandomRangeFloat(
       Minimum=&FloatFromCurve(
        FloatCurve=DataInterface() {propertyValues {Curve {keys [{interpMode=RCIM_Cubic, tangentMode=RCTM_Break, time=0.17, value=1}, {interpMode=RCIM_Cubic, tangentMode=RCTM_Break, time=0.85, value=0}]}}}
        Scale Curve=0.1
       )
       Maximum=&FloatFromCurve(
        FloatCurve=DataInterface() {propertyValues {Curve {keys [{interpMode=RCIM_Cubic, tangentMode=RCTM_Break, time=0.25, value=1}, {interpMode=RCIM_Cubic, tangentMode=RCTM_Break, time=0.75, value=0}]}}}
        Scale Curve=0.5
       )
      ))
     1=&VectorFromFloat(Value=&RandomRangeFloat(
       Minimum=&FloatFromCurve(FloatCurve=DataInterface() {propertyValues {Curve {keys [<at 0.15 1>,<at 0.35 0>]}}}, Scale Curve=0.1)
       Maximum=&FloatFromCurve(FloatCurve=DataInterface() {propertyValues {Curve {keys [<at 0.5 1>,<at 0.85 0>]}}}, Scale Curve=0.3)
      ))
     2=&VectorFromFloat(Value=&RandomRangeFloat(
       Minimum=&FloatFromCurve(FloatCurve=DataInterface() {propertyValues {Curve {keys [<at 0.45 1>,<at 0.88 0>]}}}, Scale Curve=0.2)
       Maximum=&FloatFromCurve(FloatCurve=DataInterface() {propertyValues {Curve {keys [<at 0.5 1>,<at 0.95 0>]}}}, Scale Curve=0.35)
      ))
     3=&VectorFromFloat(Value=&RandomRangeFloat(
       Minimum=&FloatFromCurve(FloatCurve=DataInterface() {propertyValues {Curve {keys [<at 0.35 1>,<at 1 0>]}}})
       Maximum=&FloatFromCurve(FloatCurve=DataInterface() {propertyValues {Curve {keys [<at 0.5 1>,<at 1 0>]}}})
      ))
     4=&VectorFromFloat(Value=&RandomRangeFloat(
       Minimum=&FloatFromCurve(FloatCurve=DataInterface() {propertyValues {Curve {keys [<at 0.35 1>,<at 1 0>]}}})
       Maximum=&FloatFromCurve(FloatCurve=DataInterface() {propertyValues {Curve {keys [<at 0.5 1>,<at 1 0>]}}})
      ))
     5=&VectorFromFloat(Value=&RandomRangeFloat(
       Minimum=&FloatFromCurve(FloatCurve=DataInterface() {propertyValues {Curve {keys [<at 0.35 1>,<at 1 0>]}}})
       Maximum=&FloatFromCurve(FloatCurve=DataInterface() {propertyValues {Curve {keys [<at 0.5 1>,<at 1 0>]}}})
      ))
     Index=&SpawnGroup()
     Num=::NewEnumerator1
    )) {isSetParameters=true, module=SetVariables_#5}
   &ScaleMeshSize(Scale Factor=&VectorFromFloat(Value=&RandomRangeFloat(
      Minimum=&FloatFromCurve(FloatCurve=DataInterface() {propertyValues {Curve {keys [<at 0.3 1>,<at 1 0>]}}})
      Maximum=&FloatFromCurve(FloatCurve=DataInterface() {propertyValues {Curve {keys [<at 0.5 1>,<at 1 0>]}}})
     ))) {enabled=false}
   &AerodynamicDrag(
    Aerodynamic Drag:NiagaraFloat=&RandomRangeFloat(Minimum=0, Maximum=0.05)
    Aerodynamic Rotational Drag:NiagaraFloat=&RandomRangeFloat(Minimum=0, Maximum=25)
    Drag Shape Curve=DataInterface() {propertyValues {Curve {keys [
        {interpMode=RCIM_Cubic, tangentMode=RCTM_User, time=0, value=0.1}
        {arriveTangent=2.0799, interpMode=RCIM_Cubic, leaveTangent=2.0799, tangentMode=RCTM_User, time=1, value=1}
       ]}}}
    Lift Contribution=0.5
    Lift Shape Curve=DataInterface() {propertyValues {Curve {keys [
        {interpMode=RCIM_Cubic, tangentMode=RCTM_User, time=0, value=0}
        {interpMode=RCIM_Cubic, tangentMode=RCTM_User, time=0.5, value=1}
        {tangentMode=RCTM_User, time=1, value=0}
       ]}}}
    Pivot Offset Maximum:Vector3f=<5 0.5 1>
    Pivot Offset Minimum:Vector3f=<-5 -0.5 -1>
    Rotation Strength:NiagaraFloat=&RandomRangeFloat(Minimum=0.7, Maximum=1)
    Use Linear Drag=true
    Use Rotational Drag=true
   )
   &GravityForce(Gravity=&MakeVector(
     X=0
     Y=0
     Z=&MergeHelper_FloatSwitch(
      0=&FloatFromCurve(
       FloatCurve=DataInterface() {propertyValues {Curve {
          defaultValue=1
          keys [
           {interpMode=RCIM_Cubic, tangentMode=RCTM_User, time=0, value=1.3}
           {arriveTangent=-1.6, interpMode=RCIM_Cubic, leaveTangent=-1.6, time=0.15, value=1}
           {arriveTangent=1.0551, interpMode=RCIM_Cubic, leaveTangent=1.0551, time=0.5, value=0.5}
           {arriveTangent=-0.3012, interpMode=RCIM_Cubic, leaveTangent=-0.3012, tangentMode=RCTM_User, time=1.0978, value=2}
          ]
         }}}
       Scale Curve=&RandomRangeFloat(Minimum=-900, Maximum=-1300)
      )
      1=&FloatFromCurve(
       FloatCurve=DataInterface() {propertyValues {Curve {
          defaultValue=1
          keys [
           {interpMode=RCIM_Cubic, tangentMode=RCTM_User, time=0, value=0}
           {interpMode=RCIM_Cubic, tangentMode=RCTM_User, time=0.1234, value=0.8792}
           {arriveTangent=0.1761, interpMode=RCIM_Cubic, leaveTangent=0.1761, time=0.8212, value=0.0431}
           {arriveTangent=0.8774, interpMode=RCIM_Cubic, leaveTangent=0.8774, time=0.8987, value=1.0157}
           <at 1 0.2 RCIM_Cubic>
           <at 1.0938 1.0157 RCIM_Cubic>
          ]
         }}}
       Scale Curve=&RandomRangeFloat(Minimum=-500, Maximum=-1000)
      )
      2=&FloatFromCurve(
       FloatCurve=DataInterface() {propertyValues {Curve {
          defaultValue=1
          keys [
           {interpMode=RCIM_Cubic, tangentMode=RCTM_User, time=0, value=0}
           {interpMode=RCIM_Cubic, tangentMode=RCTM_User, time=0.25, value=1}
           {arriveTangent=0.4286, interpMode=RCIM_Cubic, leaveTangent=0.4286, time=0.3882, value=0.2004}
           {arriveTangent=0.4977, interpMode=RCIM_Cubic, leaveTangent=0.4977, time=0.6, value=1.15}
           <at 1.121 0.5651 RCIM_Cubic>
          ]
         }}}
       Scale Curve=&RandomRangeFloat(Minimum=-250, Maximum=-500)
      )
      3=?
      Index=$Particles.VisibilityTag
      Num=::NewEnumerator1
     )
    ))
   &Drag(
    Drag=&MergeHelper_FloatSwitch(
     0=&FloatFromCurve(
      FloatCurve=DataInterface() {propertyValues {Curve {
         defaultValue=0.3
         keys [
          {interpMode=RCIM_Cubic, tangentMode=RCTM_User, time=0, value=0}
          {interpMode=RCIM_Cubic, tangentMode=RCTM_User, time=0.072, value=0.9686}
          {tangentMode=RCTM_User, time=0.5537, value=0.0471}
         ]
        }}}
      Scale Curve=&RandomRangeFloat(Minimum=3, Maximum=7)
     )
     1=&FloatFromCurve(
      FloatCurve=DataInterface() {propertyValues {Curve {
         defaultValue=0
         keys [{interpMode=RCIM_Cubic, tangentMode=RCTM_User, time=0, value=0}, {interpMode=RCIM_Cubic, tangentMode=RCTM_User, time=1, value=1}]
        }}}
      Scale Curve=&RandomRangeFloat(Minimum=2, Maximum=4)
     )
     2=&FloatFromCurve(
      FloatCurve=DataInterface() {propertyValues {Curve {
         defaultValue=1
         keys [{interpMode=RCIM_Cubic, tangentMode=RCTM_User, time=0, value=0}, {interpMode=RCIM_Cubic, tangentMode=RCTM_User, time=1, value=1.5}]
        }}}
      Scale Curve=3.5
     )
     3=&FloatFromCurve(
      FloatCurve=DataInterface() {propertyValues {Curve {
         defaultValue=1
         keys [{interpMode=RCIM_Cubic, tangentMode=RCTM_User, time=0, value=0}, {interpMode=RCIM_Cubic, tangentMode=RCTM_User, time=1, value=1}]
        }}}
      Scale Curve=2.5
     )
     Index=$Particles.VisibilityTag
     Num=::NewEnumerator1
    )
    Use Linear Drag=true
    Use Rotational Drag=false
   )
   &VectorNoiseForce(Force Amount=300, Module.Use Random Noise Time Interval=true)
   &AddRotationalVelocity(Rotation Rate=<0.05 0.05 1>)
   &SolveForcesAndVelocity(Clamp Velocity=false, Limit Acceleration=false)
   &ScaleMeshSizeBySpeed(
    Sample Scale Factor By Curve=true
    Scale Factor Curve=DataInterface() {propertyValues {Curve {defaultValue=0, keys [{tangentMode=RCTM_User, time=0, value=1}, {tangentMode=RCTM_User, time=1, value=0}]}}}
    Speed Threshold=600
   )
   &ScaleColor(
    Scale Alpha=0
    Scale RGB=&VectorFromFloat(Value=&FNTimeOfDayScalar(Night Scale=0.3, Morning Scale=0.333, Day Scale=0.666, Evening Scale=0.333))
    ScaleA=false
    ScaleRGB=true
   )
   &AlignSpriteToMeshOrientation(SetAlignment=true, SetFacing=true)
   &DynamicMaterialParameters(
    Index 0 Param 1=&RandomRangeFloat(
     Minimum=&FloatFromCurve(
      FloatCurve=DataInterface() {propertyValues {Curve {
         defaultValue=1
         keys [{interpMode=RCIM_Cubic, tangentMode=RCTM_User, time=0, value=1}, {interpMode=RCIM_Cubic, tangentMode=RCTM_User, time=1, value=0}]
        }}}
      Scale Curve=0.5
     )
     Maximum=&FloatFromCurve(FloatCurve=DataInterface() {propertyValues {Curve {
         defaultValue=1
         keys [{interpMode=RCIM_Cubic, tangentMode=RCTM_User, time=0, value=1}, {interpMode=RCIM_Cubic, tangentMode=RCTM_User, time=1, value=0}]
        }}})
    )
    Module.Index0Float0Write=true
    Module.Index0Float1Write=true
    Module.Index0Float2Write=true
    Module.Index0Float3Write=true
   )
  ]
  rendererClasses [&NiagaraMeshRendererProperties]
  renderers [
   {rendererClass=&NiagaraMeshRendererProperties, rendererIndex=0}
   {rendererClass=&NiagaraMeshRendererProperties, rendererIndex=1}
   {rendererClass=&NiagaraMeshRendererProperties, rendererIndex=2}
  ]
 }
 Blue_Air_streaks {
  bEnabled=false
  emitterUpdateScript [
   &EmitterState()
   &SpawnBurst_Instantaneous(Spawn Count=15, Use Spawn Probability=false, UseLoopCountLimit=false)
  ]
  particleSpawnScript [
   /SetVariables(Particles.RandomFloat=&UniformRangedFloat(Minimum=0, Maximum=3)) {isSetParameters=true, module=SetVariables_#6}
   &InitializeParticle(
    Color Maximum=$User.Balloon Color
    Color Minimum=<rgba 0.2245 0.2891 0.8796 1>
    Color Mode=::NewEnumerator2
    Lifetime Max=0.35
    Lifetime Min=0.2
    Lifetime Mode=::NewEnumerator1
    Sprite Size Max=<60 200>
    Sprite Size Min=<50 100>
    Sprite Size Mode=::NewEnumerator2
    UsePositionOffset=false
    Write Material Random=true
   )
   &ShapeLocation(MaskSpawnGroup=false, Override Seed=false, Sphere Radius=25, UseOwnerScale=true)
   &AddVelocity(Module.Use Falloff Distance=false, Override Seed=false, Velocity Mode=::NewEnumerator1)
   &ApplyInitialForces()
   &SubUVAnimation(End Frame=3, SubUV Animation Mode:ENiagara_SubUVLookupMode=::NewEnumerator2)
  ]
  particleUpdateScript [
   &ParticleState()
   &SolveForcesAndVelocity(Clamp Velocity=false, Limit Acceleration=false)
   &ScaleSpriteSize(Scale Factor=&Vector2DFromCurve(
     Vector2Curve=DataInterface() {propertyValues {XCurve {keys [<at 0.0036 1.6897>,<at 0.9967 0.0582>]}, YCurve {keys [<at 0 1>,<at 1.002 3.7227>]}}}
     ScaleCurve=&Multiply_Vector2D_ByFloat(Float=0.4, Vector2D=<1 1>)
    ))
   &ScaleColor(
    Scale Alpha=&FloatFromCurve(
     FloatCurve=DataInterface() {propertyValues {Curve {keys [
         {interpMode=RCIM_Cubic, leaveTangent=-1.9294, tangentMode=RCTM_Break, time=0, value=1}
         {interpMode=RCIM_Cubic, tangentMode=RCTM_Break, time=1, value=0}
        ]}}}
     Scale Curve=&RandomRangeFloat(Minimum=0.1, Maximum=0.3)
    )
    ScaleA=true
    ScaleRGB=false
   )
   &DynamicMaterialParameters(
    Index 0 Param 2=1
    Index 0 Param 3=1
    Index 0 Param 4=&FloatFromCurve(FloatCurve=DataInterface() {propertyValues {Curve {keys [<at 0.3 1>,<at 1 0.25>]}}})
    Module.Param0WriteEnabled=true
    Module.Param1WriteEnabled=true
    Module.Param2WriteEnabled=true
    Module.Param3WriteEnabled=true
   )
  ]
  rendererClasses [&NiagaraSpriteRendererProperties]
  renderers [{rendererClass=&NiagaraSpriteRendererProperties}]
 }
 Blue_Air_streaks001 {
  bEnabled=false
  emitterUpdateScript [
   &EmitterState()
   &SpawnBurst_Instantaneous(Spawn Count=15, Spawn Time=0.01, Use Spawn Probability=false, UseLoopCountLimit=false)
  ]
  particleSpawnScript [
   /SetVariables(Particles.RandomFloat=&UniformRangedFloat(Minimum=0, Maximum=3)) {isSetParameters=true, module=SetVariables_#6}
   &InitializeParticle(
    Color Maximum=$User.Balloon Color
    Color Minimum=<rgba 0.2245 0.2891 0.8796 1>
    Color Mode=::NewEnumerator2
    Lifetime Max=0.35
    Lifetime Min=0.2
    Lifetime Mode=::NewEnumerator1
    Sprite Size Max=<60 135>
    Sprite Size Min=<50 55>
    Sprite Size Mode=::NewEnumerator2
    UsePositionOffset=false
    Write Material Random=true
   )
   &ShapeLocation(MaskSpawnGroup=false, Override Seed=false, Sphere Radius=25, UseOwnerScale=true)
   &AddVelocity(Module.Use Falloff Distance=false, Override Seed=false, Velocity Mode=::NewEnumerator1)
   &ApplyInitialForces()
   &SubUVAnimation(End Frame=3, SubUV Animation Mode:ENiagara_SubUVLookupMode=::NewEnumerator2)
  ]
  particleUpdateScript [
   &ParticleState()
   &SolveForcesAndVelocity(Clamp Velocity=false, Limit Acceleration=false)
   &ScaleSpriteSize(Scale Factor=&Vector2DFromCurve(
     Vector2Curve=DataInterface() {propertyValues {XCurve {keys [<at 0.0036 1.6897>,<at 0.9967 0.0582>]}, YCurve {keys [<at 0 1>,<at 1.002 3.7227>]}}}
     ScaleCurve=&Multiply_Vector2D_ByFloat(Float=0.4, Vector2D=<1 1>)
    ))
   &ScaleColor(
    Scale Alpha=&FloatFromCurve(
     FloatCurve=DataInterface() {propertyValues {Curve {keys [
         {interpMode=RCIM_Cubic, leaveTangent=-1.9294, tangentMode=RCTM_Break, time=0, value=1}
         {interpMode=RCIM_Cubic, tangentMode=RCTM_Break, time=1, value=0}
        ]}}}
     Scale Curve=&RandomRangeFloat(Minimum=0.1, Maximum=0.3)
    )
    ScaleA=true
    ScaleRGB=false
   )
   &DynamicMaterialParameters(
    Index 0 Param 2=1
    Index 0 Param 3=1
    Index 0 Param 4=&FloatFromCurve(FloatCurve=DataInterface() {propertyValues {Curve {keys [<at 0.3 1>,<at 1 0.25>]}}})
    Module.Param0WriteEnabled=true
    Module.Param1WriteEnabled=true
    Module.Param2WriteEnabled=true
    Module.Param3WriteEnabled=true
   )
  ]
  rendererClasses [&NiagaraSpriteRendererProperties]
  renderers [{rendererClass=&NiagaraSpriteRendererProperties}]
 }
 Blue_Air_streaks002 {
  emitterUpdateScript [
   &EmitterState()
   &SpawnBurst_Instantaneous(Spawn Count=15, Spawn Time=0.01, Use Spawn Probability=false, UseLoopCountLimit=false)
  ]
  particleSpawnScript [
   /SetVariables(Particles.RandomFloat=&UniformRangedFloat(Minimum=0, Maximum=3)) {isSetParameters=true, module=SetVariables_#6}
   &InitializeParticle(
    Color Maximum=<rgba 0.0685 0.1946 0.5711 1>
    Color Minimum=<rgba 0.2245 0.2891 0.8796 1>
    Color Mode=::NewEnumerator2
    Lifetime Max=0.35
    Lifetime Min=0.2
    Lifetime Mode=::NewEnumerator1
    Sprite Size Max=<60 135>
    Sprite Size Min=<50 55>
    Sprite Size Mode=::NewEnumerator2
    UsePositionOffset=false
    Write Material Random=true
   )
   &ShapeLocation(MaskSpawnGroup=false, Override Seed=false, Sphere Radius=25, UseOwnerScale=true)
   &AddVelocity(Module.Use Falloff Distance=false, Override Seed=false, Velocity Mode=::NewEnumerator1)
   &ApplyInitialForces()
   &SubUVAnimation(End Frame=3, SubUV Animation Mode:ENiagara_SubUVLookupMode=::NewEnumerator2)
  ]
  particleUpdateScript [
   &ParticleState()
   &SolveForcesAndVelocity(Clamp Velocity=false, Limit Acceleration=false)
   &ScaleSpriteSize(Scale Factor=&Vector2DFromCurve(
     Vector2Curve=DataInterface() {propertyValues {XCurve {keys [<at 0.0036 1.6897>,<at 0.9967 0.0582>]}, YCurve {keys [<at 0 1>,<at 1.002 3.7227>]}}}
     ScaleCurve=&Multiply_Vector2D_ByFloat(Float=0.4, Vector2D=<1 1>)
    ))
   &ScaleColor(
    Scale Alpha=&FloatFromCurve(
     FloatCurve=DataInterface() {propertyValues {Curve {keys [
         {interpMode=RCIM_Cubic, leaveTangent=-1.9294, tangentMode=RCTM_Break, time=0, value=1}
         {arriveTangent=-1.2019, interpMode=RCIM_Cubic, leaveTangent=-1.2019, time=0.494, value=0.265}
         {arriveTangent=-0.5237, interpMode=RCIM_Cubic, leaveTangent=-0.5237, time=0.7933, value=0.0466}
         {interpMode=RCIM_Cubic, tangentMode=RCTM_Break, time=1, value=0}
        ]}}}
     Scale Curve=&RandomRangeFloat(Minimum=0.1, Maximum=0.3)
    )
    ScaleA=true
    ScaleRGB=false
   )
   &DynamicMaterialParameters(
    Index 0 Param 2=1
    Index 0 Param 3=1
    Index 0 Param 4=&FloatFromCurve(FloatCurve=DataInterface() {propertyValues {Curve {keys [<at 0.3 1>,<at 1 0.25>]}}})
    Module.Index0Float0Write=true
    Module.Index0Float1Write=true
    Module.Index0Float2Write=true
    Module.Index0Float3Write=true
   )
  ]
  rendererClasses [&NiagaraSpriteRendererProperties]
  renderers [{rendererClass=&NiagaraSpriteRendererProperties}]
 }
 Impact {
  bEnabled=false
  emitterUpdateScript [
   &EmitterState(
    Enable Distance Culling=true
    Life Cycle Mode=::NewEnumerator1
    Loop Behavior=::NewEnumerator1
    Loop Duration=0.8261
    Max Distance Response=::KillImmediately
    MaxDistance=3500
    Module.bUseMaxDistance=true
    Module.bUseMinDistance=false
    Scalability Mode=::NewEnumerator1
    UseLoopDelay=false
   )
   &SpawnBurst_Instantaneous(Spawn Count=1, Use Spawn Probability=false)
   &SpawnBurst_Instantaneous(Spawn Time=0.8261, Use Spawn Probability=false, UseLoopCountLimit=false) {module=SpawnBurst_Instantaneous001}
  ]
  particleSpawnScript [
   &InitializeParticle(
    Color=$User.Balloon Color
    Lifetime=0.15
    Mesh Scale Mode=::NewEnumerator1
    Mesh Uniform Scale Max=1.25
    Mesh Uniform Scale Min=0.8
    UsePositionOffset=false
    Write Material Random=true
   ) {module=InitializeParticle001}
   &InitialMeshOrientation(
    Mesh Orientation Mode=::NewEnumerator2
    Rotation=&RandomRangeVector(Minimum=<-1 -1 -1>, Maximum=<1 1 1>)
    Use Rotation Vector=true
   )
   &CameraOffset(Camera Offset Amount=25)
  ]
  particleUpdateScript [
   &ParticleState()
   &ScaleColor(
    Color Value To Scale=$Particles.Color
    Scale Alpha=&FloatFromCurve(
     FloatCurve=DataInterface() {propertyValues {Curve {keys [<at 0 1>,<at 1 0>]}}}
     Scale Curve=&FNTimeOfDayScalar(Night Scale=0.5, Morning Scale=1, Evening Scale=0.5)
    )
    Scale RGB=<1.25 1.25 1.25>
    ScaleA=true
    ScaleRGB=true
   )
   &DynamicMaterialParameters(
    Index 0 Param 1=&FloatFromCurve(FloatCurve=DataInterface() {propertyValues {Curve {keys [<at 0 1 RCIM_Cubic>,<at 0.8 0 RCIM_Cubic>]}}})
    Index 0 Param 2=2
    Index 0 Param 3=1
    Index 0 Param 4=1
    Index 1 Param 1=75
    Index 1 Param 2=0.4
    Index 1 Param 3=0.6
    Index 1 Param 4=1
    Index 2 Param 1=&FloatFromCurve(FloatCurve=DataInterface() {propertyValues {Curve {keys [<at 0.46 0>,<at 1 1>]}}})
    Index 2 Param 2=45
    Module.Param0WriteEnabled=true
    Module.Param0WriteEnabled1=true
    Module.Param0WriteEnabled2=true
    Module.Param1WriteEnabled=true
    Module.Param1WriteEnabled1=true
    Module.Param1WriteEnabled2=true
    Module.Param2WriteEnabled=true
    Module.Param2WriteEnabled1=true
    Module.Param2WriteEnabled2=true
    Module.Param3WriteEnabled=true
    Module.Param3WriteEnabled1=true
    Module.Param3WriteEnabled2=true
    Write Parameter Index 1=true
    Write Parameter Index 2=true
   )
   &ScaleMeshSize(Scale Factor=&VectorFromFloat(Value=&FloatFromCurve(
      FloatCurve=DataInterface() {propertyValues {Curve {keys [{arriveTangent=3.7989, interpMode=RCIM_Cubic, leaveTangent=3.7989, tangentMode=RCTM_User, time=0, value=0.35}, <at 1 1.5>]}}}
      Scale Curve=&FN_FrontEndFloatModifier(Gameplay=1, Front End=0.4)
     )))
   &ScaleColor(
    Color Value To Scale=$Particles.Color
    Linear Color Curve=DataInterface() {propertyValues {
      AlphaCurve {keys [<at 0 1>,<at 0.6532 0.7>,<at 1 0>]}
      BlueCurve {keys [<at 0 5>,<at 0.1508 5>,<at 0.278 0>,<at 0.6975 0>,<at 0.992 0>,<at 1 1>]}
      GreenCurve {keys [<at 0 1.1843>,<at 0.1508 1.0899>,<at 0.278 1.8823>,<at 0.6975 0.3765>,<at 0.992 0.3765>,<at 1 1>]}
      RedCurve {keys [<at 0 0>,<at 0.1508 0>,<at 0.278 5>,<at 0.6975 1>,<at 0.992 1>,<at 1 1>]}
     }}
    Scale Mode=::NewEnumerator2
   ) {enabled=false, module=ScaleColor001}
  ]
  rendererClasses [&NiagaraMeshRendererProperties]
  renderers [{rendererClass=&NiagaraMeshRendererProperties, rendererIndex=0}]
 }
 Impact001 {
  bEnabled=false
  emitterUpdateScript [
   &EmitterState(
    Enable Distance Culling=true
    Life Cycle Mode=::NewEnumerator1
    Loop Behavior=::NewEnumerator1
    Loop Duration=0.8261
    Max Distance Response=::KillImmediately
    MaxDistance=3500
    Module.bUseMaxDistance=true
    Module.bUseMinDistance=false
    Scalability Mode=::NewEnumerator1
    UseLoopDelay=false
   )
   &SpawnBurst_Instantaneous(Spawn Count=1, Use Spawn Probability=false)
   &SpawnBurst_Instantaneous(Spawn Time=0.8261, Use Spawn Probability=false, UseLoopCountLimit=false) {module=SpawnBurst_Instantaneous001}
  ]
  particleSpawnScript [
   &InitializeParticle(
    Color=$User.Balloon Color
    Lifetime=0.15
    Mesh Scale Mode=::NewEnumerator1
    Mesh Uniform Scale Max=0.8
    Mesh Uniform Scale Min=0.6
    UsePositionOffset=false
    Write Material Random=true
   ) {module=InitializeParticle001}
   &InitialMeshOrientation(
    Mesh Orientation Mode=::NewEnumerator2
    Rotation=&RandomRangeVector(Minimum=<-1 -1 -1>, Maximum=<1 1 1>)
    Use Rotation Vector=true
   )
   &CameraOffset(Camera Offset Amount=25)
  ]
  particleUpdateScript [
   &ParticleState()
   &ScaleColor(
    Color Value To Scale=$Particles.Color
    Scale Alpha=&FloatFromCurve(
     FloatCurve=DataInterface() {propertyValues {Curve {keys [<at 0 1>,<at 1 0>]}}}
     Scale Curve=&FNTimeOfDayScalar(Night Scale=0.5, Morning Scale=1, Evening Scale=0.5)
    )
    Scale RGB=<1.25 1.25 1.25>
    ScaleA=true
    ScaleRGB=true
   )
   &DynamicMaterialParameters(
    Index 0 Param 1=&FloatFromCurve(FloatCurve=DataInterface() {propertyValues {Curve {keys [<at 0 1 RCIM_Cubic>,<at 0.8 0 RCIM_Cubic>]}}})
    Index 0 Param 2=2
    Index 0 Param 3=1
    Index 0 Param 4=1
    Index 1 Param 1=75
    Index 1 Param 2=0.4
    Index 1 Param 3=0.6
    Index 1 Param 4=1
    Index 2 Param 1=&FloatFromCurve(FloatCurve=DataInterface() {propertyValues {Curve {keys [<at 0.46 0>,<at 1 1>]}}})
    Index 2 Param 2=45
    Module.Param0WriteEnabled=true
    Module.Param0WriteEnabled1=true
    Module.Param0WriteEnabled2=true
    Module.Param1WriteEnabled=true
    Module.Param1WriteEnabled1=true
    Module.Param1WriteEnabled2=true
    Module.Param2WriteEnabled=true
    Module.Param2WriteEnabled1=true
    Module.Param2WriteEnabled2=true
    Module.Param3WriteEnabled=true
    Module.Param3WriteEnabled1=true
    Module.Param3WriteEnabled2=true
    Write Parameter Index 1=true
    Write Parameter Index 2=true
   )
   &ScaleMeshSize(Scale Factor=&VectorFromFloat(Value=&FloatFromCurve(
      FloatCurve=DataInterface() {propertyValues {Curve {keys [{arriveTangent=3.7989, interpMode=RCIM_Cubic, leaveTangent=3.7989, tangentMode=RCTM_User, time=0, value=0.35}, <at 1 1.5>]}}}
      Scale Curve=&FN_FrontEndFloatModifier(Gameplay=1, Front End=0.4)
     )))
   &ScaleColor(
    Color Value To Scale=$Particles.Color
    Linear Color Curve=DataInterface() {propertyValues {
      AlphaCurve {keys [<at 0 1>,<at 0.6532 0.7>,<at 1 0>]}
      BlueCurve {keys [<at 0 5>,<at 0.1508 5>,<at 0.278 0>,<at 0.6975 0>,<at 0.992 0>,<at 1 1>]}
      GreenCurve {keys [<at 0 1.1843>,<at 0.1508 1.0899>,<at 0.278 1.8823>,<at 0.6975 0.3765>,<at 0.992 0.3765>,<at 1 1>]}
      RedCurve {keys [<at 0 0>,<at 0.1508 0>,<at 0.278 5>,<at 0.6975 1>,<at 0.992 1>,<at 1 1>]}
     }}
    Scale Mode=::NewEnumerator2
   ) {enabled=false, module=ScaleColor001}
  ]
  rendererClasses [&NiagaraMeshRendererProperties]
  renderers [{rendererClass=&NiagaraMeshRendererProperties, rendererIndex=0}]
 }
 Impact002 {
  emitterUpdateScript [
   &EmitterState(
    Enable Distance Culling=true
    Life Cycle Mode=::NewEnumerator1
    Loop Behavior=::NewEnumerator1
    Loop Duration=0.8261
    Max Distance Response=::KillImmediately
    MaxDistance=3500
    Module.bUseMaxDistance=true
    Module.bUseMinDistance=false
    Scalability Mode=::NewEnumerator1
    UseLoopDelay=false
   )
   &SpawnBurst_Instantaneous(Spawn Count=1, Use Spawn Probability=false)
   &SpawnBurst_Instantaneous(Spawn Time=0.8261, Use Spawn Probability=false, UseLoopCountLimit=false) {module=SpawnBurst_Instantaneous001}
  ]
  particleSpawnScript [
   &InitializeParticle(
    Color=<rgba 0.0453 0.1289 0.3781 1>
    Lifetime=0.15
    Mesh Scale Mode=::NewEnumerator3
    Mesh Uniform Scale=0.8
    UsePositionOffset=false
    Write Material Random=true
   ) {module=InitializeParticle001}
   &InitialMeshOrientation(
    Mesh Orientation Mode=::NewEnumerator2
    Rotation=&RandomRangeVector(Minimum=<-1 -1 -1>, Maximum=<1 1 1>)
    Use Rotation Vector=true
   )
   &CameraOffset(Camera Offset Amount=25)
  ]
  particleUpdateScript [
   &ParticleState()
   &ScaleColor(
    Color Value To Scale=$Particles.Color
    Scale Alpha=&FloatFromCurve(
     FloatCurve=DataInterface() {propertyValues {Curve {keys [<at 0 1>,<at 1 0>]}}}
     Scale Curve=&FNTimeOfDayScalar(Night Scale=0.5, Morning Scale=1, Evening Scale=0.5)
    )
    Scale RGB=<1.25 1.25 1.25>
    ScaleA=true
    ScaleRGB=true
   )
   &DynamicMaterialParameters(
    Index 0 Param 1=&FloatFromCurve(FloatCurve=DataInterface() {propertyValues {Curve {keys [<at 0 1 RCIM_Cubic>,<at 0.8 0 RCIM_Cubic>]}}})
    Index 0 Param 2=2
    Index 0 Param 3=1
    Index 0 Param 4=1
    Index 1 Param 1=75
    Index 1 Param 2=0.4
    Index 1 Param 3=0.6
    Index 1 Param 4=1
    Index 2 Param 1=&FloatFromCurve(FloatCurve=DataInterface() {propertyValues {Curve {keys [<at 0.46 0>,<at 1 1>]}}})
    Index 2 Param 2=45
    Module.Param0WriteEnabled=true
    Module.Param0WriteEnabled1=true
    Module.Param0WriteEnabled2=true
    Module.Param1WriteEnabled=true
    Module.Param1WriteEnabled1=true
    Module.Param1WriteEnabled2=true
    Module.Param2WriteEnabled=true
    Module.Param2WriteEnabled1=true
    Module.Param2WriteEnabled2=true
    Module.Param3WriteEnabled=true
    Module.Param3WriteEnabled1=true
    Module.Param3WriteEnabled2=true
    Write Parameter Index 1=true
    Write Parameter Index 2=true
   )
   &ScaleMeshSize(Scale Factor=&VectorFromFloat(Value=&FloatFromCurve(
      FloatCurve=DataInterface() {propertyValues {Curve {keys [{arriveTangent=4.5426, time=0, value=0.35}, {arriveTangent=0.1157, time=0.7473, value=1.4892}]}}}
      Scale Curve=&FN_FrontEndFloatModifier(Gameplay=1.2, Front End=1)
     )))
  ]
  rendererClasses [&NiagaraMeshRendererProperties]
  renderers [{rendererClass=&NiagaraMeshRendererProperties, rendererIndex=0}]
 }
 Mesh_Partition_group_burst_particles {
  emitterUpdateScript [
   &EmitterState()
   &SpawnBurst_Instantaneous(
    Spawn Count=&RandomRangeInt(Minimum=5, Maximum=13)
    Spawn Time=0.02
    Use Spawn Probability=false
    UseLoopCountLimit=false
   )
  ]
  particleSpawnScript [
   &FN_SetSpawnGroup(Number of Spawn Groups=3)
   &PartitionParticles(
    Number Of Partitions=&SpawnGroup()
    Particles Per Partition=1
    Partition Mode:ENiagara_PartitionMode_v2=::NewEnumerator1
   )
   &InitializeParticle(
    Color=&SelectLinearColorFromArray(
     Array Sampling Mode=::NewEnumerator1
     Direct Array Index=&SpawnGroup()
     Direct Array Mode=::NewEnumerator1
     Color Selection Array=DataInterface() {propertyValues {ColorData [<rgba 0.0528 0.1501 0.4406 1>, <rgba 0.0943 0.268 0.7865 1>, <rgba 0.0307 0.1144 0.8796 1>]}}
     Random Seed=?
    )
    Lifetime=&MergeHelper_FloatSwitch(
     0=&RandomRangeFloat(Minimum=0.15, Maximum=0.25)
     1=&RandomRangeFloat(Minimum=0.12, Maximum=0.23)
     2=&RandomRangeFloat(Minimum=0.1, Maximum=0.22)
     3=&RandomRangeFloat(Minimum=0, Maximum=0)
     4=&RandomRangeFloat(Minimum=0, Maximum=0)
     5=&RandomRangeFloat(Minimum=0, Maximum=0)
     Index=$Particles.SpawnGroup
     Num=::NewEnumerator1
    )
    Mesh Scale Mode=::NewEnumerator3
    Mesh Uniform Scale=&MergeHelper_FloatSwitch(
     0=&RandomRangeFloat(Minimum=0.07, Maximum=0.1)
     1=&RandomRangeFloat(Minimum=0.05, Maximum=0.08)
     2=&RandomRangeFloat(Minimum=0.02, Maximum=0.06)
     3=&MakeFloatFromVector2d(Vector 2D=<5 10>, Channel=::NewEnumerator1)
     Index=$Particles.VisibilityTag
     Num=::NewEnumerator1
    )
    UsePositionOffset=false
    Write Material Random=true
   )
   &ShapeLocation(MaskSpawnGroup=false, Override Seed=false, Sphere Radius=45, UseOwnerScale=true) {module=ShapeLocation003}
   &AddVelocity(
    Module.Use Falloff Distance=false
    Origin Offset=<1 0 0>
    Override Seed=false
    Velocity Mode=::NewEnumerator1
    Velocity Speed=&Multiply_Float(
     A=&RandomRangeFloat(Minimum=500, Maximum=750)
     B=&FloatFromCurve(FloatCurve=DataInterface() {propertyValues {Curve {keys [
          {arriveTangent=-3.3408, interpMode=RCIM_Cubic, leaveTangent=-3.3408, tangentMode=RCTM_User, time=0, value=1}
          {arriveTangent=-0.6538, tangentMode=RCTM_User, time=1, value=0}
         ]}}})
    )
   ) {module=AddVelocity001}
   &ApplyInitialForces()
   &InitialMeshOrientation(Rotation=<0.125 0 0>, Use Rotation Vector=true)
   /SetVariables(Particles.MeshIndex=&SelectIntFromArray(
     Array Sampling Mode=::NewEnumerator1
     Direct Array Index=?
     Int Selection Array=DataInterface() {propertyValues {IntData [0,1,1]}}
     Random Seed=0
    )) {isSetParameters=true}
  ]
  particleUpdateScript [
   &ParticleState()
   &ScaleMeshSize(Scale Factor=&VectorFromFloat(Value=&RandomRangeFloat(
      Minimum=&FloatFromCurve(FloatCurve=DataInterface() {propertyValues {Curve {
          defaultValue=1
          keys [{interpMode=RCIM_Cubic, tangentMode=RCTM_User, time=0, value=1}, {interpMode=RCIM_Cubic, tangentMode=RCTM_User, time=0.3, value=0}]
         }}})
      Maximum=&FloatFromCurve(FloatCurve=DataInterface() {propertyValues {Curve {
          defaultValue=1
          keys [{interpMode=RCIM_Cubic, tangentMode=RCTM_User, time=0, value=1}, {interpMode=RCIM_Cubic, tangentMode=RCTM_User, time=1, value=0}]
         }}})
     )))
   &GravityForce(Gravity=&MergeHelper_VectorSwitch(
     0=&MakeVector(
      X=0
      Y=0
      Z=&FloatFromCurve(
       FloatCurve=DataInterface() {propertyValues {Curve {
          defaultValue=1
          keys [
           {interpMode=RCIM_Cubic, tangentMode=RCTM_User, time=0, value=1}
           {arriveTangent=0.4021, interpMode=RCIM_Cubic, leaveTangent=0.4021, time=0.2, value=0.5}
           {arriveTangent=0.5556, interpMode=RCIM_Cubic, leaveTangent=0.5556, time=0.5091, value=1.2047}
           {arriveTangent=0.4939, interpMode=RCIM_Cubic, leaveTangent=0.4939, tangentMode=RCTM_User, time=2, value=1.5}
          ]
         }}}
       Scale Curve=&RandomRangeFloat(Minimum=-450, Maximum=-1200)
      )
     )
     1=&MakeVector(
      X=0
      Y=0
      Z=&FloatFromCurve(
       FloatCurve=DataInterface() {propertyValues {Curve {
          defaultValue=1
          keys [
           {interpMode=RCIM_Cubic, tangentMode=RCTM_User, time=0, value=1}
           {interpMode=RCIM_Cubic, tangentMode=RCTM_User, time=0.1456, value=0.5}
           {arriveTangent=0.9048, interpMode=RCIM_Cubic, leaveTangent=0.9048, time=0.2674, value=0.9696}
           {arriveTangent=-0.7577, interpMode=RCIM_Cubic, leaveTangent=-0.7577, time=0.7367, value=1.0348}
           {arriveTangent=-0.0536, interpMode=RCIM_Cubic, leaveTangent=-0.0536, time=1.0192, value=0.4}
           <at 1.0938 1.0157 RCIM_Cubic>
          ]
         }}}
       Scale Curve=&RandomRangeFloat(Minimum=-500, Maximum=-1000)
      )
     )
     2=&MakeVector(
      X=0
      Y=0
      Z=&FloatFromCurve(
       FloatCurve=DataInterface() {propertyValues {Curve {
          defaultValue=1
          keys [
           {interpMode=RCIM_Cubic, tangentMode=RCTM_User, time=0, value=0}
           {interpMode=RCIM_Cubic, tangentMode=RCTM_User, time=0.15, value=1.3}
           {arriveTangent=-0.6, interpMode=RCIM_Cubic, leaveTangent=-0.6, time=0.45, value=0.35}
           {arriveTangent=-1.86, interpMode=RCIM_Cubic, leaveTangent=-1.86, tangentMode=RCTM_User, time=0.65, value=1}
           <at 1.121 0.5651 RCIM_Cubic>
          ]
         }}}
       Scale Curve=&RandomRangeFloat(Minimum=-600, Maximum=-980)
      )
     )
     3=&MakeVector(
      X=0
      Y=0
      Z=&MergeHelper_FloatSwitch(
       0=&FloatFromCurve(
        FloatCurve=DataInterface() {propertyValues {Curve {
           defaultValue=1
           keys [
            {interpMode=RCIM_Cubic, tangentMode=RCTM_User, time=0, value=0}
            {arriveTangent=0.2683, interpMode=RCIM_Cubic, leaveTangent=0.2683, time=0.7, value=0.7652}
            {arriveTangent=0.143, interpMode=RCIM_Cubic, leaveTangent=0.143, time=1.5434, value=0.414}
            {arriveTangent=-0.3012, interpMode=RCIM_Cubic, leaveTangent=-0.3012, tangentMode=RCTM_User, time=2.0131, value=0.9529}
           ]
          }}}
        Scale Curve=&RandomRangeFloat(Minimum=-450, Maximum=-850)
       )
       1=&FloatFromCurve(
        FloatCurve=DataInterface() {propertyValues {Curve {
           defaultValue=1
           keys [
            {interpMode=RCIM_Cubic, tangentMode=RCTM_User, time=0, value=0}
            {interpMode=RCIM_Cubic, tangentMode=RCTM_User, time=0.7, value=0.9843}
            {arriveTangent=0.1579, interpMode=RCIM_Cubic, leaveTangent=0.1579, time=0.8212, value=0.0431}
            {arriveTangent=-0.1169, interpMode=RCIM_Cubic, leaveTangent=-0.1169, time=0.8987, value=1.0157}
            <at 1.0192 0.02 RCIM_Cubic>
            <at 1.0938 1.0157 RCIM_Cubic>
           ]
          }}}
        Scale Curve=&RandomRangeFloat(Minimum=-400, Maximum=-600)
       )
       2=&FloatFromCurve(
        FloatCurve=DataInterface() {propertyValues {Curve {
           defaultValue=1
           keys [
            {interpMode=RCIM_Cubic, tangentMode=RCTM_User, time=0, value=0}
            {interpMode=RCIM_Cubic, tangentMode=RCTM_User, time=0.4, value=0.7}
            {arriveTangent=1.524, interpMode=RCIM_Cubic, leaveTangent=1.524, time=0.573, value=0.1768}
            {arriveTangent=0.7086, interpMode=RCIM_Cubic, leaveTangent=0.7086, time=0.7599, value=1.2485}
            <at 1.121 0.5651 RCIM_Cubic>
           ]
          }}}
        Scale Curve=-150
       )
       3=?
       Index=$Particles.VisibilityTag
       Num=::NewEnumerator1
      )
     )
     4=&MakeVector(
      X=0
      Y=0
      Z=&MergeHelper_FloatSwitch(
       0=&FloatFromCurve(
        FloatCurve=DataInterface() {propertyValues {Curve {
           defaultValue=1
           keys [
            {interpMode=RCIM_Cubic, tangentMode=RCTM_User, time=0, value=0}
            {arriveTangent=0.2683, interpMode=RCIM_Cubic, leaveTangent=0.2683, time=0.7, value=0.7652}
            {arriveTangent=0.143, interpMode=RCIM_Cubic, leaveTangent=0.143, time=1.5434, value=0.414}
            {arriveTangent=-0.3012, interpMode=RCIM_Cubic, leaveTangent=-0.3012, tangentMode=RCTM_User, time=2.0131, value=0.9529}
           ]
          }}}
        Scale Curve=&RandomRangeFloat(Minimum=-450, Maximum=-850)
       )
       1=&FloatFromCurve(
        FloatCurve=DataInterface() {propertyValues {Curve {
           defaultValue=1
           keys [
            {interpMode=RCIM_Cubic, tangentMode=RCTM_User, time=0, value=0}
            {interpMode=RCIM_Cubic, tangentMode=RCTM_User, time=0.7, value=0.9843}
            {arriveTangent=0.1579, interpMode=RCIM_Cubic, leaveTangent=0.1579, time=0.8212, value=0.0431}
            {arriveTangent=-0.1169, interpMode=RCIM_Cubic, leaveTangent=-0.1169, time=0.8987, value=1.0157}
            <at 1.0192 0.02 RCIM_Cubic>
            <at 1.0938 1.0157 RCIM_Cubic>
           ]
          }}}
        Scale Curve=&RandomRangeFloat(Minimum=-400, Maximum=-600)
       )
       2=&FloatFromCurve(
        FloatCurve=DataInterface() {propertyValues {Curve {
           defaultValue=1
           keys [
            {interpMode=RCIM_Cubic, tangentMode=RCTM_User, time=0, value=0}
            {interpMode=RCIM_Cubic, tangentMode=RCTM_User, time=0.4, value=0.7}
            {arriveTangent=1.524, interpMode=RCIM_Cubic, leaveTangent=1.524, time=0.573, value=0.1768}
            {arriveTangent=0.7086, interpMode=RCIM_Cubic, leaveTangent=0.7086, time=0.7599, value=1.2485}
            <at 1.121 0.5651 RCIM_Cubic>
           ]
          }}}
        Scale Curve=-150
       )
       3=?
       Index=$Particles.VisibilityTag
       Num=::NewEnumerator1
      )
     )
     5=&MakeVector(
      X=0
      Y=0
      Z=&MergeHelper_FloatSwitch(
       0=&FloatFromCurve(
        FloatCurve=DataInterface() {propertyValues {Curve {
           defaultValue=1
           keys [
            {interpMode=RCIM_Cubic, tangentMode=RCTM_User, time=0, value=0}
            {arriveTangent=0.2683, interpMode=RCIM_Cubic, leaveTangent=0.2683, time=0.7, value=0.7652}
            {arriveTangent=0.143, interpMode=RCIM_Cubic, leaveTangent=0.143, time=1.5434, value=0.414}
            {arriveTangent=-0.3012, interpMode=RCIM_Cubic, leaveTangent=-0.3012, tangentMode=RCTM_User, time=2.0131, value=0.9529}
           ]
          }}}
        Scale Curve=&RandomRangeFloat(Minimum=-450, Maximum=-850)
       )
       1=&FloatFromCurve(
        FloatCurve=DataInterface() {propertyValues {Curve {
           defaultValue=1
           keys [
            {interpMode=RCIM_Cubic, tangentMode=RCTM_User, time=0, value=0}
            {interpMode=RCIM_Cubic, tangentMode=RCTM_User, time=0.7, value=0.9843}
            {arriveTangent=0.1579, interpMode=RCIM_Cubic, leaveTangent=0.1579, time=0.8212, value=0.0431}
            {arriveTangent=-0.1169, interpMode=RCIM_Cubic, leaveTangent=-0.1169, time=0.8987, value=1.0157}
            <at 1.0192 0.02 RCIM_Cubic>
            <at 1.0938 1.0157 RCIM_Cubic>
           ]
          }}}
        Scale Curve=&RandomRangeFloat(Minimum=-400, Maximum=-600)
       )
       2=&FloatFromCurve(
        FloatCurve=DataInterface() {propertyValues {Curve {
           defaultValue=1
           keys [
            {interpMode=RCIM_Cubic, tangentMode=RCTM_User, time=0, value=0}
            {interpMode=RCIM_Cubic, tangentMode=RCTM_User, time=0.4, value=0.7}
            {arriveTangent=1.524, interpMode=RCIM_Cubic, leaveTangent=1.524, time=0.573, value=0.1768}
            {arriveTangent=0.7086, interpMode=RCIM_Cubic, leaveTangent=0.7086, time=0.7599, value=1.2485}
            <at 1.121 0.5651 RCIM_Cubic>
           ]
          }}}
        Scale Curve=-150
       )
       3=?
       Index=$Particles.VisibilityTag
       Num=::NewEnumerator1
      )
     )
     Index=&SpawnGroup()
     Num=::NewEnumerator1
    ))
   &Drag(
    Drag=&MergeHelper_FloatSwitch(
     0=&FloatFromCurve(
      FloatCurve=DataInterface() {propertyValues {Curve {
         defaultValue=1
         keys [{interpMode=RCIM_Cubic, tangentMode=RCTM_User, time=0, value=1}, {interpMode=RCIM_Cubic, tangentMode=RCTM_User, time=0.8024, value=0}]
        }}}
      Scale Curve=&RandomRangeFloat(Minimum=5, Maximum=7)
     )
     1=&FloatFromCurve(
      FloatCurve=DataInterface() {propertyValues {Curve {
         defaultValue=1
         keys [
          {interpMode=RCIM_Cubic, tangentMode=RCTM_User, time=0, value=0}
          {interpMode=RCIM_Cubic, tangentMode=RCTM_User, time=0.5, value=1}
          <at 1.0039 0.3 RCIM_Cubic>
         ]
        }}}
      Scale Curve=&RandomRangeFloat(Minimum=1, Maximum=3)
     )
     2=&FloatFromCurve(
      FloatCurve=DataInterface() {propertyValues {Curve {
         defaultValue=1
         keys [
          {interpMode=RCIM_Cubic, tangentMode=RCTM_User, time=0, value=0}
          {interpMode=RCIM_Cubic, tangentMode=RCTM_User, time=0.4, value=0.7}
          {arriveTangent=1.524, interpMode=RCIM_Cubic, leaveTangent=1.524, time=0.573, value=0.1768}
          {arriveTangent=0.7086, interpMode=RCIM_Cubic, leaveTangent=0.7086, time=0.7599, value=1.2485}
          <at 1.121 0.5651 RCIM_Cubic>
         ]
        }}}
      Scale Curve=-150
     )
     3=&FloatFromCurve(
      FloatCurve=DataInterface() {propertyValues {Curve {
         defaultValue=1
         keys [{interpMode=RCIM_Cubic, tangentMode=RCTM_User, time=0, value=1}, {interpMode=RCIM_Cubic, tangentMode=RCTM_User, time=1, value=0}]
        }}}
      Scale Curve=&RandomRangeFloat(Minimum=8, Maximum=10)
     )
     4=&FloatFromCurve(
      FloatCurve=DataInterface() {propertyValues {Curve {
         defaultValue=1
         keys [{interpMode=RCIM_Cubic, tangentMode=RCTM_User, time=0, value=1}, {interpMode=RCIM_Cubic, tangentMode=RCTM_User, time=1, value=0}]
        }}}
      Scale Curve=&RandomRangeFloat(Minimum=8, Maximum=10)
     )
     5=&FloatFromCurve(
      FloatCurve=DataInterface() {propertyValues {Curve {
         defaultValue=1
         keys [{interpMode=RCIM_Cubic, tangentMode=RCTM_User, time=0, value=1}, {interpMode=RCIM_Cubic, tangentMode=RCTM_User, time=1, value=0}]
        }}}
      Scale Curve=&RandomRangeFloat(Minimum=8, Maximum=10)
     )
     Index=&SpawnGroup()
     Num=::NewEnumerator1
    )
    Use Linear Drag=true
    Use Rotational Drag=true
   )
   &AerodynamicDrag(
    Aerodynamic Drag:NiagaraFloat=&RandomRangeFloat(Minimum=0, Maximum=0.05)
    Aerodynamic Rotational Drag:NiagaraFloat=&RandomRangeFloat(Minimum=15, Maximum=75)
    Drag Shape Curve=DataInterface() {propertyValues {Curve {keys [
        {interpMode=RCIM_Cubic, tangentMode=RCTM_User, time=0, value=0.1}
        {arriveTangent=2.0799, interpMode=RCIM_Cubic, leaveTangent=2.0799, tangentMode=RCTM_User, time=1, value=1}
       ]}}}
    Lift Contribution=0.5
    Lift Shape Curve=DataInterface() {propertyValues {Curve {keys [
        {interpMode=RCIM_Cubic, tangentMode=RCTM_User, time=0, value=0}
        {interpMode=RCIM_Cubic, tangentMode=RCTM_User, time=0.5, value=1}
        {tangentMode=RCTM_User, time=1, value=0}
       ]}}}
    Pivot Offset Maximum:Vector3f=<5 0.5 1>
    Pivot Offset Minimum:Vector3f=<-5 -0.5 -1>
    Rotation Strength:NiagaraFloat=&RandomRangeFloat(Minimum=1, Maximum=3)
    Use Linear Drag=true
    Use Rotational Drag=true
   )
   &AddRotationalVelocity(Rotation Rate=<0.5 0.5 1>)
   &VectorNoiseForce(Force Amount=300, Module.Use Random Noise Time Interval=true)
   &SolveForcesAndVelocity(Clamp Velocity=false, Limit Acceleration=false)
   &ScaleMeshSizeBySpeed(
    Sample Scale Factor By Curve=true
    Scale Factor Curve=DataInterface() {propertyValues {Curve {
       defaultValue=1
       keys [{interpMode=RCIM_Cubic, tangentMode=RCTM_User, time=0, value=1}, {interpMode=RCIM_Cubic, tangentMode=RCTM_User, time=0.0356, value=0.0188}]
      }}}
    Speed Threshold:NiagaraFloat=&RandomRangeFloat(Minimum=500, Maximum=850)
   )
   &DynamicMaterialParameters(
    Index 0 Param 1=&RandomRangeFloat(
     Minimum=&FloatFromCurve(
      FloatCurve=DataInterface() {propertyValues {Curve {
         defaultValue=1
         keys [{interpMode=RCIM_Cubic, tangentMode=RCTM_User, time=0, value=1}, {interpMode=RCIM_Cubic, tangentMode=RCTM_User, time=0.85, value=0}]
        }}}
      Scale Curve=0.5
     )
     Maximum=&FloatFromCurve(FloatCurve=DataInterface() {propertyValues {Curve {
         defaultValue=1
         keys [{interpMode=RCIM_Cubic, tangentMode=RCTM_User, time=0, value=1}, {interpMode=RCIM_Cubic, tangentMode=RCTM_User, time=1, value=0}]
        }}})
    )
    Module.Index0Float0Write=true
    Module.Index0Float1Write=true
    Module.Index0Float2Write=true
    Module.Index0Float3Write=true
   )
   &AlignSpriteToMeshOrientation(SetAlignment=true, SetFacing=true)
  ]
  rendererClasses [&NiagaraMeshRendererProperties]
  renderers [{rendererClass=&NiagaraMeshRendererProperties, rendererIndex=0}]
 }
 Mesh_Partition_group_burst_particles001 {
  bEnabled=false
  emitterUpdateScript [
   &EmitterState()
   &SpawnBurst_Instantaneous(
    Spawn Count=&RandomRangeInt(Minimum=15, Maximum=25)
    Spawn Time=0.02
    Use Spawn Probability=false
    UseLoopCountLimit=false
   )
   &SpawnRate(Use Spawn Probability=false) {enabled=false}
  ]
  particleSpawnScript [
   &FN_SetSpawnGroup(Number of Spawn Groups=5)
   &PartitionParticles(Number Of Partitions=&SpawnGroup(), Partition Mode:ENiagara_PartitionMode_v2=::NewEnumerator1)
   &InitializeParticle(
    Color=&SelectLinearColorFromArray(
     Array Sampling Mode=::NewEnumerator1
     Direct Array Index=&SpawnGroup()
     Direct Array Mode=::NewEnumerator1
     Color Selection Array=DataInterface() {propertyValues {ColorData [
        <rgba 0.0528 0.1501 0.4406 1>
        <rgba 0.0943 0.268 0.7865 1>
        <rgba 0.0307 0.1144 0.8796 1>
        <rgba 1 0.5403 0 1>
        <rgba 0 0.971 1 1>
        <rgba 1 0.1558 0 1>
       ]}}
     Random Seed=?
    )
    Lifetime=&MergeHelper_FloatSwitch(
     0=&RandomRangeFloat(Minimum=0.25, Maximum=0.55)
     1=&RandomRangeFloat(Minimum=0.2, Maximum=0.45)
     2=&RandomRangeFloat(Minimum=0.33, Maximum=0.6)
     3=&RandomRangeFloat(Minimum=0, Maximum=0)
     4=&RandomRangeFloat(Minimum=0, Maximum=0)
     5=&RandomRangeFloat(Minimum=0, Maximum=0)
     Index=$Particles.SpawnGroup
     Num=::NewEnumerator3
    )
    Mesh Scale Mode=::NewEnumerator3
    Mesh Uniform Scale=&MergeHelper_FloatSwitch(
     0=&RandomRangeFloat(Minimum=0.05, Maximum=0.25)
     1=&RandomRangeFloat(Minimum=0.1, Maximum=0.3)
     2=&RandomRangeFloat(Minimum=0.05, Maximum=0.15)
     3=&MakeFloatFromVector2d(Vector 2D=<5 10>, Channel=::NewEnumerator1)
     Index=$Particles.VisibilityTag
     Num=::NewEnumerator2
    )
    Sprite Size Mode=::NewEnumerator3
    Uniform Sprite Size=&MergeHelper_FloatSwitch(
     0=&RandomRangeFloat(Minimum=0.02, Maximum=0.05)
     1=0
     2=0
     3=0
     4=0
     Index=$Particles.SpawnGroup
     Num=::NewEnumerator3
    )
    UsePositionOffset=false
    Write Material Random=true
   )
   &ShapeLocation(
    MaskSpawnGroup=false
    Offset=&MergeHelper_VectorSwitch(
     0=<100 0 0>
     1=<300 0 0>
     2=<600 0 0>
     3=<900 0 0>
     4=<1200 0 0>
     5=<1500 0 0>
     Index=&SpawnGroup()
     Num=::NewEnumerator4
    )
    Offset Mode=::NewEnumerator0
    Override Seed=false
    Shape Primitive=::NewEnumerator1
    Surface Only=false
    UseOwnerScale=true
   ) {enabled=false}
   &AddVelocity(Velocity=&MergeHelper_VectorSwitch(
     0=&RandomRangeVector(Minimum=<-300 -300 -300>, Maximum=<300 300 300>)
     1=&RandomRangeVector(Minimum=<-130 -130 -130>, Maximum=<130 130 130>)
     2=&RandomRangeVector(Minimum=<-200 -200 -200>, Maximum=<200 -200 200>)
     3=&RandomRangeVector(Minimum=<0 0 0>, Maximum=<1 1 1>)
     4=&RandomRangeVector(Minimum=<0 0 0>, Maximum=<1 1 1>)
     5=&RandomRangeVector(Minimum=<0 0 0>, Maximum=<1 1 1>)
     Index=&SpawnGroup()
     Num=::NewEnumerator4
    )) {enabled=false}
   &AddVelocity(Velocity=&MergeHelper_VectorSwitch(
     0=&RandomRangeVector(Minimum=<250 -250 -300>, Maximum=<350 250 250>)
     1=&RandomRangeVector(Minimum=<-300 -250 -100>, Maximum=<150 -450 300>)
     2=&RandomRangeVector(Minimum=<-222 -200 -135>, Maximum=<-400 200 60>)
     3=&RandomRangeVector(Minimum=<-222 -200 -50>, Maximum=<-400 200 50>)
     4=&RandomRangeVector(Minimum=<-50 -50 -50>, Maximum=<50 50 50>)
     5=&RandomRangeVector(Minimum=<-600 -600 -600>, Maximum=<600 600 600>)
     Index=&SpawnGroup()
     Num=::NewEnumerator4
    )) {module=AddVelocity001}
   &ShapeLocation(
    MaskSpawnGroup=false
    Offset=&SelectVectorFromArray(
     Direct Array Index=$Particles.SpawnGroup
     Vector Selection Array=DataInterface() {propertyValues {FloatData [<0 0 0>, <0 -5 0>, <5 0 0>, <0 50 0>, <0 0 0>, <-30 0 0>]}}
    )
    Offset Mode=::NewEnumerator0
    Override Seed=false
    Sphere Radius=1
    UseOwnerScale=true
   ) {module=ShapeLocation001}
   &InitialMeshOrientation(Rotation=<0.125 0 0>, Use Rotation Vector=true)
  ]
  particleUpdateScript [
   &ParticleState()
   &GravityForce(Gravity=&MergeHelper_VectorSwitch(
     0=&MakeVector(
      X=0
      Y=0
      Z=&FloatFromCurve(
       FloatCurve=DataInterface() {propertyValues {Curve {
          defaultValue=1
          keys [
           {interpMode=RCIM_Cubic, tangentMode=RCTM_User, time=0, value=0}
           {arriveTangent=0.8281, interpMode=RCIM_Cubic, leaveTangent=0.8281, time=0.3, value=0.7652}
           {arriveTangent=0.1096, interpMode=RCIM_Cubic, leaveTangent=0.1096, time=0.5, value=0.414}
           {arriveTangent=-0.3012, interpMode=RCIM_Cubic, leaveTangent=-0.3012, tangentMode=RCTM_User, time=2.0131, value=0.9529}
          ]
         }}}
       Scale Curve=&RandomRangeFloat(Minimum=-450, Maximum=-850)
      )
     )
     1=&MakeVector(
      X=0
      Y=0
      Z=&FloatFromCurve(
       FloatCurve=DataInterface() {propertyValues {Curve {
          defaultValue=1
          keys [
           {interpMode=RCIM_Cubic, tangentMode=RCTM_User, time=0, value=0}
           {interpMode=RCIM_Cubic, tangentMode=RCTM_User, time=0.7, value=0.9843}
           {arriveTangent=0.1579, interpMode=RCIM_Cubic, leaveTangent=0.1579, time=0.8212, value=0.0431}
           {arriveTangent=-0.1169, interpMode=RCIM_Cubic, leaveTangent=-0.1169, time=0.8987, value=1.0157}
           <at 1.0192 0.02 RCIM_Cubic>
           <at 1.0938 1.0157 RCIM_Cubic>
          ]
         }}}
       Scale Curve=&RandomRangeFloat(Minimum=-400, Maximum=-600)
      )
     )
     2=&MakeVector(
      X=0
      Y=0
      Z=&FloatFromCurve(
       FloatCurve=DataInterface() {propertyValues {Curve {
          defaultValue=1
          keys [
           {interpMode=RCIM_Cubic, tangentMode=RCTM_User, time=0, value=0}
           {interpMode=RCIM_Cubic, tangentMode=RCTM_User, time=0.4, value=0.7}
           {arriveTangent=1.524, interpMode=RCIM_Cubic, leaveTangent=1.524, time=0.573, value=0.1768}
           {arriveTangent=0.7086, interpMode=RCIM_Cubic, leaveTangent=0.7086, time=0.7599, value=1.2485}
           <at 1.121 0.5651 RCIM_Cubic>
          ]
         }}}
       Scale Curve=-150
      )
     )
     3=&MakeVector(
      X=0
      Y=0
      Z=&FloatFromCurve(
       FloatCurve=DataInterface() {propertyValues {Curve {
          defaultValue=1
          keys [
           {interpMode=RCIM_Cubic, tangentMode=RCTM_User, time=0, value=0}
           {arriveTangent=0.2683, interpMode=RCIM_Cubic, leaveTangent=0.2683, time=0.7, value=0.7652}
           {arriveTangent=0.143, interpMode=RCIM_Cubic, leaveTangent=0.143, time=1.5434, value=0.414}
           {arriveTangent=-0.3012, interpMode=RCIM_Cubic, leaveTangent=-0.3012, tangentMode=RCTM_User, time=2.0131, value=0.9529}
          ]
         }}}
       Scale Curve=&RandomRangeFloat(Minimum=-450, Maximum=-850)
      )
     )
     4=&MakeVector(
      X=0
      Y=0
      Z=&MergeHelper_FloatSwitch(
       0=&FloatFromCurve(
        FloatCurve=DataInterface() {propertyValues {Curve {
           defaultValue=1
           keys [
            {interpMode=RCIM_Cubic, tangentMode=RCTM_User, time=0, value=0}
            {arriveTangent=0.2683, interpMode=RCIM_Cubic, leaveTangent=0.2683, time=0.7, value=0.7652}
            {arriveTangent=0.143, interpMode=RCIM_Cubic, leaveTangent=0.143, time=1.5434, value=0.414}
            {arriveTangent=-0.3012, interpMode=RCIM_Cubic, leaveTangent=-0.3012, tangentMode=RCTM_User, time=2.0131, value=0.9529}
           ]
          }}}
        Scale Curve=&RandomRangeFloat(Minimum=-450, Maximum=-850)
       )
       1=&FloatFromCurve(
        FloatCurve=DataInterface() {propertyValues {Curve {
           defaultValue=1
           keys [
            {interpMode=RCIM_Cubic, tangentMode=RCTM_User, time=0, value=0}
            {interpMode=RCIM_Cubic, tangentMode=RCTM_User, time=0.7, value=0.9843}
            {arriveTangent=0.1579, interpMode=RCIM_Cubic, leaveTangent=0.1579, time=0.8212, value=0.0431}
            {arriveTangent=-0.1169, interpMode=RCIM_Cubic, leaveTangent=-0.1169, time=0.8987, value=1.0157}
            <at 1.0192 0.02 RCIM_Cubic>
            <at 1.0938 1.0157 RCIM_Cubic>
           ]
          }}}
        Scale Curve=&RandomRangeFloat(Minimum=-400, Maximum=-600)
       )
       2=&FloatFromCurve(
        FloatCurve=DataInterface() {propertyValues {Curve {
           defaultValue=1
           keys [
            {interpMode=RCIM_Cubic, tangentMode=RCTM_User, time=0, value=0}
            {interpMode=RCIM_Cubic, tangentMode=RCTM_User, time=0.4, value=0.7}
            {arriveTangent=1.524, interpMode=RCIM_Cubic, leaveTangent=1.524, time=0.573, value=0.1768}
            {arriveTangent=0.7086, interpMode=RCIM_Cubic, leaveTangent=0.7086, time=0.7599, value=1.2485}
            <at 1.121 0.5651 RCIM_Cubic>
           ]
          }}}
        Scale Curve=-150
       )
       3=?
       Index=$Particles.VisibilityTag
       Num=::NewEnumerator1
      )
     )
     5=&MakeVector(
      X=0
      Y=0
      Z=&MergeHelper_FloatSwitch(
       0=&FloatFromCurve(
        FloatCurve=DataInterface() {propertyValues {Curve {
           defaultValue=1
           keys [
            {interpMode=RCIM_Cubic, tangentMode=RCTM_User, time=0, value=0}
            {arriveTangent=0.2683, interpMode=RCIM_Cubic, leaveTangent=0.2683, time=0.7, value=0.7652}
            {arriveTangent=0.143, interpMode=RCIM_Cubic, leaveTangent=0.143, time=1.5434, value=0.414}
            {arriveTangent=-0.3012, interpMode=RCIM_Cubic, leaveTangent=-0.3012, tangentMode=RCTM_User, time=2.0131, value=0.9529}
           ]
          }}}
        Scale Curve=&RandomRangeFloat(Minimum=-450, Maximum=-850)
       )
       1=&FloatFromCurve(
        FloatCurve=DataInterface() {propertyValues {Curve {
           defaultValue=1
           keys [
            {interpMode=RCIM_Cubic, tangentMode=RCTM_User, time=0, value=0}
            {interpMode=RCIM_Cubic, tangentMode=RCTM_User, time=0.7, value=0.9843}
            {arriveTangent=0.1579, interpMode=RCIM_Cubic, leaveTangent=0.1579, time=0.8212, value=0.0431}
            {arriveTangent=-0.1169, interpMode=RCIM_Cubic, leaveTangent=-0.1169, time=0.8987, value=1.0157}
            <at 1.0192 0.02 RCIM_Cubic>
            <at 1.0938 1.0157 RCIM_Cubic>
           ]
          }}}
        Scale Curve=&RandomRangeFloat(Minimum=-400, Maximum=-600)
       )
       2=&FloatFromCurve(
        FloatCurve=DataInterface() {propertyValues {Curve {
           defaultValue=1
           keys [
            {interpMode=RCIM_Cubic, tangentMode=RCTM_User, time=0, value=0}
            {interpMode=RCIM_Cubic, tangentMode=RCTM_User, time=0.4, value=0.7}
            {arriveTangent=1.524, interpMode=RCIM_Cubic, leaveTangent=1.524, time=0.573, value=0.1768}
            {arriveTangent=0.7086, interpMode=RCIM_Cubic, leaveTangent=0.7086, time=0.7599, value=1.2485}
            <at 1.121 0.5651 RCIM_Cubic>
           ]
          }}}
        Scale Curve=-150
       )
       3=?
       Index=$Particles.VisibilityTag
       Num=::NewEnumerator1
      )
     )
     Index=&SpawnGroup()
     Num=::NewEnumerator4
    ))
   &Drag(
    Drag=&MergeHelper_FloatSwitch(
     0=&FloatFromCurve(
      FloatCurve=DataInterface() {propertyValues {Curve {
         defaultValue=1
         keys [{interpMode=RCIM_Cubic, tangentMode=RCTM_User, time=0, value=1}, {interpMode=RCIM_Cubic, tangentMode=RCTM_User, time=1, value=0}]
        }}}
      Scale Curve=&RandomRangeFloat(Minimum=0.1, Maximum=0.5)
     )
     1=&FloatFromCurve(
      FloatCurve=DataInterface() {propertyValues {Curve {
         defaultValue=1
         keys [{interpMode=RCIM_Cubic, tangentMode=RCTM_User, time=0, value=1}, {interpMode=RCIM_Cubic, tangentMode=RCTM_User, time=1, value=0}]
        }}}
      Scale Curve=&RandomRangeFloat(Minimum=0.02, Maximum=0.05)
     )
     2=&FloatFromCurve(
      FloatCurve=DataInterface() {propertyValues {Curve {
         defaultValue=1
         keys [{interpMode=RCIM_Cubic, tangentMode=RCTM_User, time=0, value=1}, {interpMode=RCIM_Cubic, tangentMode=RCTM_User, time=1, value=0}]
        }}}
      Scale Curve=&RandomRangeFloat(Minimum=0, Maximum=2)
     )
     3=&FloatFromCurve(
      FloatCurve=DataInterface() {propertyValues {Curve {
         defaultValue=1
         keys [{interpMode=RCIM_Cubic, tangentMode=RCTM_User, time=0, value=1}, {interpMode=RCIM_Cubic, tangentMode=RCTM_User, time=1, value=0}]
        }}}
      Scale Curve=&RandomRangeFloat(Minimum=8, Maximum=10)
     )
     4=&FloatFromCurve(
      FloatCurve=DataInterface() {propertyValues {Curve {
         defaultValue=1
         keys [{interpMode=RCIM_Cubic, tangentMode=RCTM_User, time=0, value=1}, {interpMode=RCIM_Cubic, tangentMode=RCTM_User, time=1, value=0}]
        }}}
      Scale Curve=&RandomRangeFloat(Minimum=8, Maximum=10)
     )
     5=&FloatFromCurve(
      FloatCurve=DataInterface() {propertyValues {Curve {
         defaultValue=1
         keys [{interpMode=RCIM_Cubic, tangentMode=RCTM_User, time=0, value=1}, {interpMode=RCIM_Cubic, tangentMode=RCTM_User, time=1, value=0}]
        }}}
      Scale Curve=&RandomRangeFloat(Minimum=8, Maximum=10)
     )
     Index=&SpawnGroup()
     Num=::NewEnumerator4
    )
    Use Linear Drag=true
    Use Rotational Drag=true
   )
   &AerodynamicDrag(
    Aerodynamic Drag:NiagaraFloat=&RandomRangeFloat(Minimum=0, Maximum=0.05)
    Aerodynamic Rotational Drag:NiagaraFloat=&RandomRangeFloat(Minimum=35, Maximum=100)
    Drag Shape Curve=DataInterface() {propertyValues {Curve {keys [
        {interpMode=RCIM_Cubic, tangentMode=RCTM_User, time=0, value=0.1}
        {arriveTangent=2.0799, interpMode=RCIM_Cubic, leaveTangent=2.0799, tangentMode=RCTM_User, time=1, value=1}
       ]}}}
    Lift Contribution=0.5
    Lift Shape Curve=DataInterface() {propertyValues {Curve {keys [
        {interpMode=RCIM_Cubic, tangentMode=RCTM_User, time=0, value=0}
        {interpMode=RCIM_Cubic, tangentMode=RCTM_User, time=0.5, value=1}
        {tangentMode=RCTM_User, time=1, value=0}
       ]}}}
    Pivot Offset Maximum:Vector3f=<5 0.5 1>
    Pivot Offset Minimum:Vector3f=<-5 -0.5 -1>
    Rotation Strength:NiagaraFloat=&RandomRangeFloat(Minimum=1, Maximum=3)
    Use Linear Drag=true
    Use Rotational Drag=true
   )
   &AddRotationalVelocity(Rotation Rate=<0.2 0.3 0.5>)
   &SolveForcesAndVelocity(Clamp Velocity=false, Limit Acceleration=false)
   &ScaleColor(
    Scale Alpha=&MergeHelper_FloatSwitch(
     0=&FloatFromCurve(FloatCurve=DataInterface() {propertyValues {Curve {keys [<at 0 1>,<at 1 0>]}}})
     1=&FloatFromCurve(FloatCurve=DataInterface() {propertyValues {Curve {keys [<at 0 1>,<at 1 0>]}}})
     2=&FloatFromCurve(FloatCurve=DataInterface() {propertyValues {Curve {keys [<at 0 1>,<at 1 0>]}}})
     3=&FloatFromCurve(FloatCurve=DataInterface() {propertyValues {Curve {keys [<at 0 1>,<at 1 0>]}}})
     4=&FloatFromCurve(FloatCurve=DataInterface() {propertyValues {Curve {keys [<at 0 1>,<at 1 0>]}}})
     5=&FloatFromCurve(FloatCurve=DataInterface() {propertyValues {Curve {keys [<at 0 1>,<at 1 0>]}}})
     Index=&SpawnGroup()
     Num=::NewEnumerator4
    )
    ScaleA=true
    ScaleRGB=true
   )
   /SetVariables(Particles.Update scale:Vector3f=&MergeHelper_VectorSwitch(
     0=&VectorFromFloat(Value=&RandomRangeFloat(
       Minimum=&FloatFromCurve(FloatCurve=DataInterface() {propertyValues {Curve {keys [<at 0.15 1>,<at 0.6 0>]}}}, Scale Curve=0.3)
       Maximum=&FloatFromCurve(FloatCurve=DataInterface() {propertyValues {Curve {keys [<at 0.5 1>,<at 0.85 0>]}}}, Scale Curve=0.5)
      ))
     1=&VectorFromFloat(Value=&RandomRangeFloat(
       Minimum=&FloatFromCurve(FloatCurve=DataInterface() {propertyValues {Curve {keys [<at 0.25 1>,<at 0.55 0>]}}}, Scale Curve=0.55)
       Maximum=&FloatFromCurve(FloatCurve=DataInterface() {propertyValues {Curve {keys [<at 0.5 1>,<at 0.85 0>]}}}, Scale Curve=0.75)
      ))
     2=&VectorFromFloat(Value=&RandomRangeFloat(
       Minimum=&FloatFromCurve(FloatCurve=DataInterface() {propertyValues {Curve {keys [<at 0.45 1>,<at 0.88 0>]}}}, Scale Curve=0.6)
       Maximum=&FloatFromCurve(FloatCurve=DataInterface() {propertyValues {Curve {keys [<at 0.5 1>,<at 0.95 0>]}}}, Scale Curve=0.95)
      ))
     3=&VectorFromFloat(Value=&RandomRangeFloat(
       Minimum=&FloatFromCurve(FloatCurve=DataInterface() {propertyValues {Curve {keys [<at 0.35 1>,<at 1 0>]}}})
       Maximum=&FloatFromCurve(FloatCurve=DataInterface() {propertyValues {Curve {keys [<at 0.5 1>,<at 1 0>]}}})
      ))
     4=&VectorFromFloat(Value=&RandomRangeFloat(
       Minimum=&FloatFromCurve(FloatCurve=DataInterface() {propertyValues {Curve {keys [<at 0.35 1>,<at 1 0>]}}})
       Maximum=&FloatFromCurve(FloatCurve=DataInterface() {propertyValues {Curve {keys [<at 0.5 1>,<at 1 0>]}}})
      ))
     5=&VectorFromFloat(Value=&RandomRangeFloat(
       Minimum=&FloatFromCurve(FloatCurve=DataInterface() {propertyValues {Curve {keys [<at 0.35 1>,<at 1 0>]}}})
       Maximum=&FloatFromCurve(FloatCurve=DataInterface() {propertyValues {Curve {keys [<at 0.5 1>,<at 1 0>]}}})
      ))
     Index=&SpawnGroup()
     Num=::NewEnumerator1
    )) {isSetParameters=true}
   &DynamicMaterialParameters(
    Index 0 Param 1=&RandomRangeFloat(
     Minimum=&FloatFromCurve(
      FloatCurve=DataInterface() {propertyValues {Curve {
         defaultValue=1
         keys [{interpMode=RCIM_Cubic, tangentMode=RCTM_User, time=0, value=1}, {interpMode=RCIM_Cubic, tangentMode=RCTM_User, time=1, value=0}]
        }}}
      Scale Curve=0.5
     )
     Maximum=&FloatFromCurve(FloatCurve=DataInterface() {propertyValues {Curve {
         defaultValue=1
         keys [{interpMode=RCIM_Cubic, tangentMode=RCTM_User, time=0, value=1}, {interpMode=RCIM_Cubic, tangentMode=RCTM_User, time=1, value=0}]
        }}})
    )
    Module.Index0Float0Write=true
    Module.Index0Float1Write=true
    Module.Index0Float2Write=true
    Module.Index0Float3Write=true
   )
   &AlignSpriteToMeshOrientation(SetAlignment=true, SetFacing=true)
  ]
  rendererClasses [&NiagaraSpriteRendererProperties, &NiagaraMeshRendererProperties]
  renderers [
   {rendererClass=&NiagaraSpriteRendererProperties}
   {rendererClass=&NiagaraMeshRendererProperties, rendererIndex=1}
  ]
 }
 Partition_group_burst_particles {
  bEnabled=false
  emitterUpdateScript [
   &EmitterState()
   &SpawnRate(Use Spawn Probability=false) {enabled=false}
   &SpawnBurst_Instantaneous(Spawn Count=50, Use Spawn Probability=false, UseLoopCountLimit=false)
  ]
  particleSpawnScript [
   &FN_SetSpawnGroup(Number of Spawn Groups=6)
   &PartitionParticles(
    Number Of Partitions=&SpawnGroup()
    Particles Per Partition=10
    Partition Mode:ENiagara_PartitionMode_v2=::NewEnumerator1
   )
   &InitializeParticle(
    Color=&SelectLinearColorFromArray(
     Array Sampling Mode=::NewEnumerator1
     Direct Array Index=&SpawnGroup()
     Direct Array Mode=::NewEnumerator1
     Color Selection Array=DataInterface() {propertyValues {ColorData [<rgba 1 0 0.9045 1>, <rgba 0.136 0.0713 1 1>, <rgba 1 0 0.0444 1>, <rgba 1 0.5403 0 1>, <rgba 0 0.971 1 1>, <rgba 1 0.1558 0 1>]}}
     Random Seed=?
    )
    Lifetime=&MergeHelper_FloatSwitch(
     0=1
     1=0.5
     2=&RandomRangeFloat(Minimum=0.5, Maximum=1)
     3=2
     4=0.7
     5=1.3
     Index=$Particles.SpawnGroup
     Num=::NewEnumerator4
    )
    Sprite Size=&MergeHelper_Vector2Switch(0=<1 1>, 1=<1 1>, 2=<1 1>, 3=<1 1>, 4=<1 1>, 5=<1 1>, Index=$Particles.SpawnGroup, Num=::NewEnumerator4)
    Sprite Size Mode=::NewEnumerator4
    UsePositionOffset=false
    Write Material Random=true
   )
   &ShapeLocation(
    MaskSpawnGroup=false
    Offset=&MergeHelper_VectorSwitch(
     0=<100 0 0>
     1=<300 0 0>
     2=<600 0 0>
     3=<900 0 0>
     4=<1200 0 0>
     5=<1500 0 0>
     Index=&SpawnGroup()
     Num=::NewEnumerator4
    )
    Offset Mode=::NewEnumerator0
    Override Seed=false
    Shape Primitive=::NewEnumerator1
    Surface Only=false
    UseOwnerScale=true
   ) {enabled=false}
   &AddVelocity(Velocity=&MergeHelper_VectorSwitch(
     0=&RandomRangeVector(Minimum=<-300 -300 -300>, Maximum=<300 300 300>)
     1=&RandomRangeVector(Minimum=<-130 -130 -130>, Maximum=<130 130 130>)
     2=&RandomRangeVector(Minimum=<-200 -200 -200>, Maximum=<200 -200 200>)
     3=&RandomRangeVector(Minimum=<0 0 0>, Maximum=<1 1 1>)
     4=&RandomRangeVector(Minimum=<0 0 0>, Maximum=<1 1 1>)
     5=&RandomRangeVector(Minimum=<0 0 0>, Maximum=<1 1 1>)
     Index=&SpawnGroup()
     Num=::NewEnumerator4
    )) {enabled=false}
   &AddVelocity(Velocity=&MergeHelper_VectorSwitch(
     0=&RandomRangeVector(Minimum=<-50 -50 -50>, Maximum=<50 50 50>)
     1=&RandomRangeVector(Minimum=<-50 -50 -50>, Maximum=<50 50 50>)
     2=&RandomRangeVector(Minimum=<-50 -50 -50>, Maximum=<50 50 50>)
     3=&RandomRangeVector(Minimum=<-50 -50 -50>, Maximum=<50 50 50>)
     4=&RandomRangeVector(Minimum=<-50 -50 -50>, Maximum=<50 50 50>)
     5=&RandomRangeVector(Minimum=<-50 -50 -50>, Maximum=<50 50 50>)
     Index=&SpawnGroup()
     Num=::NewEnumerator4
    )) {module=AddVelocity001}
   &ShapeLocation(
    MaskSpawnGroup=false
    Offset=&SelectVectorFromArray(
     Direct Array Index=$Particles.SpawnGroup
     Vector Selection Array=DataInterface() {propertyValues {FloatData [<-150 0 150>, <0 1 1>, <0 0 0>, <0 0 0>, <0 0 0>, <0 0 0>]}}
    )
    Offset Mode=::NewEnumerator0
    Override Seed=false
    Sphere Radius=1
    UseOwnerScale=true
   ) {module=ShapeLocation001}
  ]
  particleUpdateScript [
   &ParticleState()
   &GravityForce(Gravity=&MergeHelper_VectorSwitch(
     0=&MakeVector(
      X=0
      Y=0
      Z=&MergeHelper_FloatSwitch(
       0=&FloatFromCurve(
        FloatCurve=DataInterface() {propertyValues {Curve {
           defaultValue=1
           keys [
            {interpMode=RCIM_Cubic, tangentMode=RCTM_User, time=0, value=0}
            {arriveTangent=0.2683, interpMode=RCIM_Cubic, leaveTangent=0.2683, time=0.7, value=0.7652}
            {arriveTangent=0.143, interpMode=RCIM_Cubic, leaveTangent=0.143, time=1.5434, value=0.414}
            {arriveTangent=-0.3012, interpMode=RCIM_Cubic, leaveTangent=-0.3012, tangentMode=RCTM_User, time=2.0131, value=0.9529}
           ]
          }}}
        Scale Curve=&RandomRangeFloat(Minimum=-450, Maximum=-850)
       )
       1=&FloatFromCurve(
        FloatCurve=DataInterface() {propertyValues {Curve {
           defaultValue=1
           keys [
            {interpMode=RCIM_Cubic, tangentMode=RCTM_User, time=0, value=0}
            {interpMode=RCIM_Cubic, tangentMode=RCTM_User, time=0.7, value=0.9843}
            {arriveTangent=0.1579, interpMode=RCIM_Cubic, leaveTangent=0.1579, time=0.8212, value=0.0431}
            {arriveTangent=-0.1169, interpMode=RCIM_Cubic, leaveTangent=-0.1169, time=0.8987, value=1.0157}
            <at 1.0192 0.02 RCIM_Cubic>
            <at 1.0938 1.0157 RCIM_Cubic>
           ]
          }}}
        Scale Curve=&RandomRangeFloat(Minimum=-400, Maximum=-600)
       )
       2=&FloatFromCurve(
        FloatCurve=DataInterface() {propertyValues {Curve {
           defaultValue=1
           keys [
            {interpMode=RCIM_Cubic, tangentMode=RCTM_User, time=0, value=0}
            {interpMode=RCIM_Cubic, tangentMode=RCTM_User, time=0.4, value=0.7}
            {arriveTangent=1.524, interpMode=RCIM_Cubic, leaveTangent=1.524, time=0.573, value=0.1768}
            {arriveTangent=0.7086, interpMode=RCIM_Cubic, leaveTangent=0.7086, time=0.7599, value=1.2485}
            <at 1.121 0.5651 RCIM_Cubic>
           ]
          }}}
        Scale Curve=-150
       )
       3=?
       Index=$Particles.VisibilityTag
       Num=::NewEnumerator1
      )
     )
     1=&MakeVector(
      X=0
      Y=0
      Z=&MergeHelper_FloatSwitch(
       0=&FloatFromCurve(
        FloatCurve=DataInterface() {propertyValues {Curve {
           defaultValue=1
           keys [
            {interpMode=RCIM_Cubic, tangentMode=RCTM_User, time=0, value=0}
            {arriveTangent=0.2683, interpMode=RCIM_Cubic, leaveTangent=0.2683, time=0.7, value=0.7652}
            {arriveTangent=0.143, interpMode=RCIM_Cubic, leaveTangent=0.143, time=1.5434, value=0.414}
            {arriveTangent=-0.3012, interpMode=RCIM_Cubic, leaveTangent=-0.3012, tangentMode=RCTM_User, time=2.0131, value=0.9529}
           ]
          }}}
        Scale Curve=&RandomRangeFloat(Minimum=-450, Maximum=-850)
       )
       1=&FloatFromCurve(
        FloatCurve=DataInterface() {propertyValues {Curve {
           defaultValue=1
           keys [
            {interpMode=RCIM_Cubic, tangentMode=RCTM_User, time=0, value=0}
            {interpMode=RCIM_Cubic, tangentMode=RCTM_User, time=0.7, value=0.9843}
            {arriveTangent=0.1579, interpMode=RCIM_Cubic, leaveTangent=0.1579, time=0.8212, value=0.0431}
            {arriveTangent=-0.1169, interpMode=RCIM_Cubic, leaveTangent=-0.1169, time=0.8987, value=1.0157}
            <at 1.0192 0.02 RCIM_Cubic>
            <at 1.0938 1.0157 RCIM_Cubic>
           ]
          }}}
        Scale Curve=&RandomRangeFloat(Minimum=-400, Maximum=-600)
       )
       2=&FloatFromCurve(
        FloatCurve=DataInterface() {propertyValues {Curve {
           defaultValue=1
           keys [
            {interpMode=RCIM_Cubic, tangentMode=RCTM_User, time=0, value=0}
            {interpMode=RCIM_Cubic, tangentMode=RCTM_User, time=0.4, value=0.7}
            {arriveTangent=1.524, interpMode=RCIM_Cubic, leaveTangent=1.524, time=0.573, value=0.1768}
            {arriveTangent=0.7086, interpMode=RCIM_Cubic, leaveTangent=0.7086, time=0.7599, value=1.2485}
            <at 1.121 0.5651 RCIM_Cubic>
           ]
          }}}
        Scale Curve=-150
       )
       3=?
       Index=$Particles.VisibilityTag
       Num=::NewEnumerator1
      )
     )
     2=&MakeVector(
      X=0
      Y=0
      Z=&MergeHelper_FloatSwitch(
       0=&FloatFromCurve(
        FloatCurve=DataInterface() {propertyValues {Curve {
           defaultValue=1
           keys [
            {interpMode=RCIM_Cubic, tangentMode=RCTM_User, time=0, value=0}
            {arriveTangent=0.2683, interpMode=RCIM_Cubic, leaveTangent=0.2683, time=0.7, value=0.7652}
            {arriveTangent=0.143, interpMode=RCIM_Cubic, leaveTangent=0.143, time=1.5434, value=0.414}
            {arriveTangent=-0.3012, interpMode=RCIM_Cubic, leaveTangent=-0.3012, tangentMode=RCTM_User, time=2.0131, value=0.9529}
           ]
          }}}
        Scale Curve=&RandomRangeFloat(Minimum=-450, Maximum=-850)
       )
       1=&FloatFromCurve(
        FloatCurve=DataInterface() {propertyValues {Curve {
           defaultValue=1
           keys [
            {interpMode=RCIM_Cubic, tangentMode=RCTM_User, time=0, value=0}
            {interpMode=RCIM_Cubic, tangentMode=RCTM_User, time=0.7, value=0.9843}
            {arriveTangent=0.1579, interpMode=RCIM_Cubic, leaveTangent=0.1579, time=0.8212, value=0.0431}
            {arriveTangent=-0.1169, interpMode=RCIM_Cubic, leaveTangent=-0.1169, time=0.8987, value=1.0157}
            <at 1.0192 0.02 RCIM_Cubic>
            <at 1.0938 1.0157 RCIM_Cubic>
           ]
          }}}
        Scale Curve=&RandomRangeFloat(Minimum=-400, Maximum=-600)
       )
       2=&FloatFromCurve(
        FloatCurve=DataInterface() {propertyValues {Curve {
           defaultValue=1
           keys [
            {interpMode=RCIM_Cubic, tangentMode=RCTM_User, time=0, value=0}
            {interpMode=RCIM_Cubic, tangentMode=RCTM_User, time=0.4, value=0.7}
            {arriveTangent=1.524, interpMode=RCIM_Cubic, leaveTangent=1.524, time=0.573, value=0.1768}
            {arriveTangent=0.7086, interpMode=RCIM_Cubic, leaveTangent=0.7086, time=0.7599, value=1.2485}
            <at 1.121 0.5651 RCIM_Cubic>
           ]
          }}}
        Scale Curve=-150
       )
       3=?
       Index=$Particles.VisibilityTag
       Num=::NewEnumerator1
      )
     )
     3=&MakeVector(
      X=0
      Y=0
      Z=&MergeHelper_FloatSwitch(
       0=&FloatFromCurve(
        FloatCurve=DataInterface() {propertyValues {Curve {
           defaultValue=1
           keys [
            {interpMode=RCIM_Cubic, tangentMode=RCTM_User, time=0, value=0}
            {arriveTangent=0.2683, interpMode=RCIM_Cubic, leaveTangent=0.2683, time=0.7, value=0.7652}
            {arriveTangent=0.143, interpMode=RCIM_Cubic, leaveTangent=0.143, time=1.5434, value=0.414}
            {arriveTangent=-0.3012, interpMode=RCIM_Cubic, leaveTangent=-0.3012, tangentMode=RCTM_User, time=2.0131, value=0.9529}
           ]
          }}}
        Scale Curve=&RandomRangeFloat(Minimum=-450, Maximum=-850)
       )
       1=&FloatFromCurve(
        FloatCurve=DataInterface() {propertyValues {Curve {
           defaultValue=1
           keys [
            {interpMode=RCIM_Cubic, tangentMode=RCTM_User, time=0, value=0}
            {interpMode=RCIM_Cubic, tangentMode=RCTM_User, time=0.7, value=0.9843}
            {arriveTangent=0.1579, interpMode=RCIM_Cubic, leaveTangent=0.1579, time=0.8212, value=0.0431}
            {arriveTangent=-0.1169, interpMode=RCIM_Cubic, leaveTangent=-0.1169, time=0.8987, value=1.0157}
            <at 1.0192 0.02 RCIM_Cubic>
            <at 1.0938 1.0157 RCIM_Cubic>
           ]
          }}}
        Scale Curve=&RandomRangeFloat(Minimum=-400, Maximum=-600)
       )
       2=&FloatFromCurve(
        FloatCurve=DataInterface() {propertyValues {Curve {
           defaultValue=1
           keys [
            {interpMode=RCIM_Cubic, tangentMode=RCTM_User, time=0, value=0}
            {interpMode=RCIM_Cubic, tangentMode=RCTM_User, time=0.4, value=0.7}
            {arriveTangent=1.524, interpMode=RCIM_Cubic, leaveTangent=1.524, time=0.573, value=0.1768}
            {arriveTangent=0.7086, interpMode=RCIM_Cubic, leaveTangent=0.7086, time=0.7599, value=1.2485}
            <at 1.121 0.5651 RCIM_Cubic>
           ]
          }}}
        Scale Curve=-150
       )
       3=?
       Index=$Particles.VisibilityTag
       Num=::NewEnumerator1
      )
     )
     4=&MakeVector(
      X=0
      Y=0
      Z=&MergeHelper_FloatSwitch(
       0=&FloatFromCurve(
        FloatCurve=DataInterface() {propertyValues {Curve {
           defaultValue=1
           keys [
            {interpMode=RCIM_Cubic, tangentMode=RCTM_User, time=0, value=0}
            {arriveTangent=0.2683, interpMode=RCIM_Cubic, leaveTangent=0.2683, time=0.7, value=0.7652}
            {arriveTangent=0.143, interpMode=RCIM_Cubic, leaveTangent=0.143, time=1.5434, value=0.414}
            {arriveTangent=-0.3012, interpMode=RCIM_Cubic, leaveTangent=-0.3012, tangentMode=RCTM_User, time=2.0131, value=0.9529}
           ]
          }}}
        Scale Curve=&RandomRangeFloat(Minimum=-450, Maximum=-850)
       )
       1=&FloatFromCurve(
        FloatCurve=DataInterface() {propertyValues {Curve {
           defaultValue=1
           keys [
            {interpMode=RCIM_Cubic, tangentMode=RCTM_User, time=0, value=0}
            {interpMode=RCIM_Cubic, tangentMode=RCTM_User, time=0.7, value=0.9843}
            {arriveTangent=0.1579, interpMode=RCIM_Cubic, leaveTangent=0.1579, time=0.8212, value=0.0431}
            {arriveTangent=-0.1169, interpMode=RCIM_Cubic, leaveTangent=-0.1169, time=0.8987, value=1.0157}
            <at 1.0192 0.02 RCIM_Cubic>
            <at 1.0938 1.0157 RCIM_Cubic>
           ]
          }}}
        Scale Curve=&RandomRangeFloat(Minimum=-400, Maximum=-600)
       )
       2=&FloatFromCurve(
        FloatCurve=DataInterface() {propertyValues {Curve {
           defaultValue=1
           keys [
            {interpMode=RCIM_Cubic, tangentMode=RCTM_User, time=0, value=0}
            {interpMode=RCIM_Cubic, tangentMode=RCTM_User, time=0.4, value=0.7}
            {arriveTangent=1.524, interpMode=RCIM_Cubic, leaveTangent=1.524, time=0.573, value=0.1768}
            {arriveTangent=0.7086, interpMode=RCIM_Cubic, leaveTangent=0.7086, time=0.7599, value=1.2485}
            <at 1.121 0.5651 RCIM_Cubic>
           ]
          }}}
        Scale Curve=-150
       )
       3=?
       Index=$Particles.VisibilityTag
       Num=::NewEnumerator1
      )
     )
     5=&MakeVector(
      X=0
      Y=0
      Z=&MergeHelper_FloatSwitch(
       0=&FloatFromCurve(
        FloatCurve=DataInterface() {propertyValues {Curve {
           defaultValue=1
           keys [
            {interpMode=RCIM_Cubic, tangentMode=RCTM_User, time=0, value=0}
            {arriveTangent=0.2683, interpMode=RCIM_Cubic, leaveTangent=0.2683, time=0.7, value=0.7652}
            {arriveTangent=0.143, interpMode=RCIM_Cubic, leaveTangent=0.143, time=1.5434, value=0.414}
            {arriveTangent=-0.3012, interpMode=RCIM_Cubic, leaveTangent=-0.3012, tangentMode=RCTM_User, time=2.0131, value=0.9529}
           ]
          }}}
        Scale Curve=&RandomRangeFloat(Minimum=-450, Maximum=-850)
       )
       1=&FloatFromCurve(
        FloatCurve=DataInterface() {propertyValues {Curve {
           defaultValue=1
           keys [
            {interpMode=RCIM_Cubic, tangentMode=RCTM_User, time=0, value=0}
            {interpMode=RCIM_Cubic, tangentMode=RCTM_User, time=0.7, value=0.9843}
            {arriveTangent=0.1579, interpMode=RCIM_Cubic, leaveTangent=0.1579, time=0.8212, value=0.0431}
            {arriveTangent=-0.1169, interpMode=RCIM_Cubic, leaveTangent=-0.1169, time=0.8987, value=1.0157}
            <at 1.0192 0.02 RCIM_Cubic>
            <at 1.0938 1.0157 RCIM_Cubic>
           ]
          }}}
        Scale Curve=&RandomRangeFloat(Minimum=-400, Maximum=-600)
       )
       2=&FloatFromCurve(
        FloatCurve=DataInterface() {propertyValues {Curve {
           defaultValue=1
           keys [
            {interpMode=RCIM_Cubic, tangentMode=RCTM_User, time=0, value=0}
            {interpMode=RCIM_Cubic, tangentMode=RCTM_User, time=0.4, value=0.7}
            {arriveTangent=1.524, interpMode=RCIM_Cubic, leaveTangent=1.524, time=0.573, value=0.1768}
            {arriveTangent=0.7086, interpMode=RCIM_Cubic, leaveTangent=0.7086, time=0.7599, value=1.2485}
            <at 1.121 0.5651 RCIM_Cubic>
           ]
          }}}
        Scale Curve=-150
       )
       3=?
       Index=$Particles.VisibilityTag
       Num=::NewEnumerator1
      )
     )
     Index=&SpawnGroup()
     Num=::NewEnumerator4
    ))
   &Drag(
    Drag=&MergeHelper_FloatSwitch(
     0=&FloatFromCurve(
      FloatCurve=DataInterface() {propertyValues {Curve {
         defaultValue=1
         keys [{interpMode=RCIM_Cubic, tangentMode=RCTM_User, time=0, value=1}, {interpMode=RCIM_Cubic, tangentMode=RCTM_User, time=1, value=0}]
        }}}
      Scale Curve=&RandomRangeFloat(Minimum=8, Maximum=10)
     )
     1=&FloatFromCurve(
      FloatCurve=DataInterface() {propertyValues {Curve {
         defaultValue=1
         keys [{interpMode=RCIM_Cubic, tangentMode=RCTM_User, time=0, value=1}, {interpMode=RCIM_Cubic, tangentMode=RCTM_User, time=1, value=0}]
        }}}
      Scale Curve=&RandomRangeFloat(Minimum=8, Maximum=10)
     )
     2=&FloatFromCurve(
      FloatCurve=DataInterface() {propertyValues {Curve {
         defaultValue=1
         keys [{interpMode=RCIM_Cubic, tangentMode=RCTM_User, time=0, value=1}, {interpMode=RCIM_Cubic, tangentMode=RCTM_User, time=1, value=0}]
        }}}
      Scale Curve=&RandomRangeFloat(Minimum=8, Maximum=10)
     )
     3=&FloatFromCurve(
      FloatCurve=DataInterface() {propertyValues {Curve {
         defaultValue=1
         keys [{interpMode=RCIM_Cubic, tangentMode=RCTM_User, time=0, value=1}, {interpMode=RCIM_Cubic, tangentMode=RCTM_User, time=1, value=0}]
        }}}
      Scale Curve=&RandomRangeFloat(Minimum=8, Maximum=10)
     )
     4=&FloatFromCurve(
      FloatCurve=DataInterface() {propertyValues {Curve {
         defaultValue=1
         keys [{interpMode=RCIM_Cubic, tangentMode=RCTM_User, time=0, value=1}, {interpMode=RCIM_Cubic, tangentMode=RCTM_User, time=1, value=0}]
        }}}
      Scale Curve=&RandomRangeFloat(Minimum=8, Maximum=10)
     )
     5=&FloatFromCurve(
      FloatCurve=DataInterface() {propertyValues {Curve {
         defaultValue=1
         keys [{interpMode=RCIM_Cubic, tangentMode=RCTM_User, time=0, value=1}, {interpMode=RCIM_Cubic, tangentMode=RCTM_User, time=1, value=0}]
        }}}
      Scale Curve=&RandomRangeFloat(Minimum=8, Maximum=10)
     )
     Index=&SpawnGroup()
     Num=::NewEnumerator4
    )
    Use Linear Drag=true
    Use Rotational Drag=true
   )
   &SolveForcesAndVelocity(Clamp Velocity=false, Limit Acceleration=false)
   &ScaleColor(
    Scale Alpha=&MergeHelper_FloatSwitch(
     0=&FloatFromCurve(FloatCurve=DataInterface() {propertyValues {Curve {keys [<at 0 1>,<at 1 0>]}}})
     1=&FloatFromCurve(FloatCurve=DataInterface() {propertyValues {Curve {keys [<at 0 1>,<at 1 0>]}}})
     2=&FloatFromCurve(FloatCurve=DataInterface() {propertyValues {Curve {keys [<at 0 1>,<at 1 0>]}}})
     3=&FloatFromCurve(FloatCurve=DataInterface() {propertyValues {Curve {keys [<at 0 1>,<at 1 0>]}}})
     4=&FloatFromCurve(FloatCurve=DataInterface() {propertyValues {Curve {keys [<at 0 1>,<at 1 0>]}}})
     5=&FloatFromCurve(FloatCurve=DataInterface() {propertyValues {Curve {keys [<at 0 1>,<at 1 0>]}}})
     Index=&SpawnGroup()
     Num=::NewEnumerator4
    )
    ScaleA=true
    ScaleRGB=true
   )
  ]
  rendererClasses [&NiagaraSpriteRendererProperties]
  renderers [{rendererClass=&NiagaraSpriteRendererProperties}]
 }
 Pop_particles {
  bEnabled=false
  emitterUpdateScript [
   &EmitterState()
   &SpawnBurst_Instantaneous(Spawn Count=&RandomRangeInt(Minimum=25, Maximum=35), Use Spawn Probability=false)
  ]
  particleSpawnScript [
   &InitializeParticle(
    Color=$User.Balloon Color
    Lifetime Max=0.4
    Lifetime Min=0.125
    Lifetime Mode=::NewEnumerator1
    Sprite Size Max=<2 2>
    Sprite Size Min=<1 2>
    Sprite Size Mode=::NewEnumerator2
    UsePositionOffset=false
    Write Material Random=true
   )
   &ShapeLocation(MaskSpawnGroup=false, Override Seed=false, Sphere Radius=30, UseOwnerScale=true)
   &AddVelocity(
    Module.Use Falloff Distance=false
    Override Seed=false
    Velocity Mode=::NewEnumerator1
    Velocity Speed=&RandomRangeFloat(Minimum=150, Maximum=200)
   )
  ]
  particleUpdateScript [
   &ParticleState()
   &AccelerationForce(Acceleration=<0 0 -150>)
   &SolveForcesAndVelocity(Clamp Velocity=false, Limit Acceleration=false)
   &ScaleColor(
    Scale Alpha=&FloatFromCurve(FloatCurve=DataInterface() {propertyValues {Curve {keys [{arriveTangent=11.9433, interpMode=RCIM_Cubic, time=0, value=0}, <at 1 1 RCIM_Cubic>]}}})
    Scale RGB=&VectorFromFloat(Value=25)
    ScaleA=true
    ScaleRGB=true
   )
   &ScaleSpriteSize(Scale Factor=&Multiply_Vector2D_ByFloat(Float=3, Vector2D=&Lerp_Vector2(A=<1 1>, B=<0 0.5>, Alpha=$Particles.NormalizedAge)))
  ]
  rendererClasses [&NiagaraSpriteRendererProperties]
  renderers [{rendererClass=&NiagaraSpriteRendererProperties}]
 }
 Pop_particles001 {
  bEnabled=false
  emitterUpdateScript [
   &EmitterState()
   &SpawnBurst_Instantaneous(Spawn Count=&RandomRangeInt(Minimum=25, Maximum=35), Use Spawn Probability=false)
  ]
  particleSpawnScript [
   &InitializeParticle(
    Color=$User.Balloon Color
    Lifetime Max=0.4
    Lifetime Min=0.125
    Lifetime Mode=::NewEnumerator1
    Sprite Size Max=<2 2>
    Sprite Size Min=<1 2>
    Sprite Size Mode=::NewEnumerator2
    UsePositionOffset=false
    Write Material Random=true
   )
   &ShapeLocation(MaskSpawnGroup=false, Override Seed=false, Sphere Radius=30, UseOwnerScale=true)
   &AddVelocity(
    Module.Use Falloff Distance=false
    Override Seed=false
    Velocity Mode=::NewEnumerator1
    Velocity Speed=&RandomRangeFloat(Minimum=150, Maximum=200)
   )
  ]
  particleUpdateScript [
   &ParticleState()
   &AccelerationForce(Acceleration=<0 0 -150>)
   &SolveForcesAndVelocity(Clamp Velocity=false, Limit Acceleration=false)
   &ScaleColor(
    Scale Alpha=&FloatFromCurve(FloatCurve=DataInterface() {propertyValues {Curve {keys [{arriveTangent=11.9433, interpMode=RCIM_Cubic, time=0, value=0}, <at 1 1 RCIM_Cubic>]}}})
    Scale RGB=&VectorFromFloat(Value=25)
    ScaleA=true
    ScaleRGB=true
   )
   &ScaleSpriteSize(Scale Factor=&Multiply_Vector2D_ByFloat(Float=3, Vector2D=&Lerp_Vector2(A=<1 1>, B=<0 0.5>, Alpha=$Particles.NormalizedAge)))
  ]
  rendererClasses [&NiagaraSpriteRendererProperties]
  renderers [{rendererClass=&NiagaraSpriteRendererProperties}]
 }
 Pop_particles002 {
  emitterUpdateScript [
   &EmitterState()
   &SpawnBurst_Instantaneous(Spawn Count=&RandomRangeInt(Minimum=25, Maximum=38), Use Spawn Probability=false)
  ]
  particleSpawnScript [
   &InitializeParticle(
    Color=<rgba 0.0307 0.1144 0.8796 1>
    Lifetime Max=0.4
    Lifetime Min=0.125
    Lifetime Mode=::NewEnumerator1
    Sprite Size Max=<2 2>
    Sprite Size Min=<1 2>
    Sprite Size Mode=::NewEnumerator2
    UsePositionOffset=false
    Write Material Random=true
   )
   &ShapeLocation(MaskSpawnGroup=false, Override Seed=false, Sphere Radius=30, UseOwnerScale=true)
   &AddVelocity(
    Module.Use Falloff Distance=false
    Override Seed=false
    Velocity Mode=::NewEnumerator1
    Velocity Speed=&RandomRangeFloat(Minimum=150, Maximum=300)
   )
  ]
  particleUpdateScript [
   &ParticleState()
   &AccelerationForce(Acceleration=<0 0 -150>)
   &SolveForcesAndVelocity(Clamp Velocity=false, Limit Acceleration=false)
   &ScaleColor(
    Scale Alpha=&FloatFromCurve(FloatCurve=DataInterface() {propertyValues {Curve {keys [{arriveTangent=11.9433, interpMode=RCIM_Cubic, time=0, value=0}, <at 1 1 RCIM_Cubic>]}}})
    Scale RGB=&VectorFromFloat(Value=25)
    ScaleA=true
    ScaleRGB=true
   )
   &ScaleSpriteSize(Scale Factor=&Multiply_Vector2D_ByFloat(Float=3, Vector2D=&Lerp_Vector2(A=<1 1>, B=<0 0.5>, Alpha=$Particles.NormalizedAge)))
  ]
  rendererClasses [&NiagaraSpriteRendererProperties]
  renderers [{rendererClass=&NiagaraSpriteRendererProperties}]
 }
 SnapRing {
  bEnabled=false
  emitterUpdateScript [
   &EmitterState()
   &SpawnBurst_Instantaneous(Spawn Count=1, Use Spawn Probability=false)
  ]
  particleSpawnScript [
   /Niagara/Modules/Spawn/Initialization/InitializeParticle(
    Color=<rgba 3 2.5089 1.95 2>
    Lifetime=0.15
    Mesh Scale=&VectorFromFloat(Value=1.25)
    Module.Write Color=true
    Module.Write Lifetime=true
    Module.Write Mass=true
    Module.Write Material Random=true
    Module.Write Position=false
    Module.Write Scale=true
    Module.Write SpriteRotation=false
    Module.Write SpriteSize=true
    Position=&SimulationPosition()
    Sprite Size=&Vector2DFromFloat(Value=400)
   ) {enabled=false}
   /Niagara/Modules/Spawn/Initialization/V2/InitializeParticle(
    Color=<rgba 0.3124 0.4115 0.8796 1>
    Lifetime=0.15
    Sprite Size Mode=::NewEnumerator3
    Uniform Sprite Size=400
    UsePositionOffset=false
    Write Material Random=true
   ) {module=InitializeParticle001}
  ]
  particleUpdateScript [
   &ParticleState()
   &DynamicMaterialParameters(
    Index 0 Param 1=&FloatFromCurve(FloatCurve=DataInterface() {propertyValues {Curve {keys [<at 0 0 RCIM_Cubic>,<at 1 1 RCIM_Cubic>]}}})
    Index 0 Param 2=50
    Index 0 Param 3=1
    Index 0 Param 4=1
    Module.Param0WriteEnabled=true
    Module.Param1WriteEnabled=true
    Module.Param2WriteEnabled=true
    Module.Param3WriteEnabled=true
   )
   &ScaleSpriteSize(Scale Factor=&Vector2DFromFloat(Value=&FloatFromCurve(FloatCurve=DataInterface() {propertyValues {Curve {keys [<at 0 0>,<at 1 1>]}}})))
   &ScaleColor(
    Scale Alpha=&FloatFromCurve(FloatCurve=DataInterface() {propertyValues {Curve {keys [<at 0.2 1 RCIM_Cubic>,<at 1 0>]}}})
    ScaleA=true
    ScaleRGB=true
   )
  ]
  rendererClasses [&NiagaraSpriteRendererProperties]
  renderers [{rendererClass=&NiagaraSpriteRendererProperties}]
 }
 SnapRing001 {
  bEnabled=false
  emitterUpdateScript [
   &EmitterState()
   &SpawnBurst_Instantaneous(Spawn Count=1, Use Spawn Probability=false)
  ]
  particleSpawnScript [
   /Niagara/Modules/Spawn/Initialization/InitializeParticle(
    Color=<rgba 3 2.5089 1.95 2>
    Lifetime=0.15
    Mesh Scale=&VectorFromFloat(Value=1.25)
    Module.Write Color=true
    Module.Write Lifetime=true
    Module.Write Mass=true
    Module.Write Material Random=true
    Module.Write Position=false
    Module.Write Scale=true
    Module.Write SpriteRotation=false
    Module.Write SpriteSize=true
    Position=&SimulationPosition()
    Sprite Size=&Vector2DFromFloat(Value=400)
   ) {enabled=false}
   /Niagara/Modules/Spawn/Initialization/V2/InitializeParticle(
    Color=<rgba 0.3124 0.4115 0.8796 1>
    Lifetime=0.15
    Sprite Size Mode=::NewEnumerator3
    Uniform Sprite Size=400
    UsePositionOffset=false
    Write Material Random=true
   ) {module=InitializeParticle001}
  ]
  particleUpdateScript [
   &ParticleState()
   &DynamicMaterialParameters(
    Index 0 Param 1=&FloatFromCurve(FloatCurve=DataInterface() {propertyValues {Curve {keys [<at 0 0 RCIM_Cubic>,<at 1 1 RCIM_Cubic>]}}})
    Index 0 Param 2=50
    Index 0 Param 3=1
    Index 0 Param 4=1
    Module.Param0WriteEnabled=true
    Module.Param1WriteEnabled=true
    Module.Param2WriteEnabled=true
    Module.Param3WriteEnabled=true
   )
   &ScaleSpriteSize(Scale Factor=&Vector2DFromFloat(Value=&FloatFromCurve(FloatCurve=DataInterface() {propertyValues {Curve {keys [<at 0 0>,<at 1 1>]}}})))
   &ScaleColor(
    Scale Alpha=&FloatFromCurve(FloatCurve=DataInterface() {propertyValues {Curve {keys [<at 0.2 1 RCIM_Cubic>,<at 1 0>]}}})
    ScaleA=true
    ScaleRGB=true
   )
  ]
  rendererClasses [&NiagaraSpriteRendererProperties]
  renderers [{rendererClass=&NiagaraSpriteRendererProperties}]
 }
 SnapRing002 {
  emitterUpdateScript [
   &EmitterState()
   &SpawnBurst_Instantaneous(Spawn Count=1, Use Spawn Probability=false)
  ]
  particleSpawnScript [
   /Niagara/Modules/Spawn/Initialization/InitializeParticle(
    Color=<rgba 3 2.5089 1.95 2>
    Lifetime=0.15
    Mesh Scale=&VectorFromFloat(Value=1.25)
    Module.Write Color=true
    Module.Write Lifetime=true
    Module.Write Mass=true
    Module.Write Material Random=true
    Module.Write Position=false
    Module.Write Scale=true
    Module.Write SpriteRotation=false
    Module.Write SpriteSize=true
    Position=&SimulationPosition()
    Sprite Size=&Vector2DFromFloat(Value=400)
   ) {enabled=false}
   /Niagara/Modules/Spawn/Initialization/V2/InitializeParticle(
    Color=<rgba 0.3124 0.4115 0.8796 1>
    Lifetime=0.15
    Sprite Size Mode=::NewEnumerator3
    Uniform Sprite Size=400
    UsePositionOffset=false
    Write Material Random=true
   ) {module=InitializeParticle001}
  ]
  particleUpdateScript [
   &ParticleState()
   &DynamicMaterialParameters(
    Index 0 Param 1=&FloatFromCurve(FloatCurve=DataInterface() {propertyValues {Curve {keys [<at 0 0 RCIM_Cubic>,<at 1 1 RCIM_Cubic>]}}})
    Index 0 Param 2=50
    Index 0 Param 3=1
    Index 0 Param 4=1
    Module.Param0WriteEnabled=true
    Module.Param1WriteEnabled=true
    Module.Param2WriteEnabled=true
    Module.Param3WriteEnabled=true
   )
   &ScaleSpriteSize(Scale Factor=&Vector2DFromFloat(Value=&FloatFromCurve(FloatCurve=DataInterface() {propertyValues {Curve {keys [<at 0 0>,<at 1 1>]}}})))
   &ScaleColor(
    Scale Alpha=&FloatFromCurve(FloatCurve=DataInterface() {propertyValues {Curve {keys [<at 0.2 1 RCIM_Cubic>,<at 1 0>]}}})
    ScaleA=true
    ScaleRGB=true
   )
  ]
  rendererClasses [&NiagaraSpriteRendererProperties]
  renderers [{rendererClass=&NiagaraSpriteRendererProperties}]
 }
}
summary {
 systemName=NS_Emote_ShyTurkey_BalloonPop
 emitters [
  {bEnabled=false, emitterName=Blue_Air_streaks, rendererClasses [&NiagaraSpriteRendererProperties]}
  {bEnabled=false, emitterName=Pop_particles, rendererClasses [&NiagaraSpriteRendererProperties]}
  {bEnabled=false, emitterName=SnapRing001, rendererClasses [&NiagaraSpriteRendererProperties]}
  {bEnabled=false, emitterName=Balloon_explosion_Blue, rendererClasses [&NiagaraMeshRendererProperties]}
  {bEnabled=false, emitterName=Impact, rendererClasses [&NiagaraMeshRendererProperties]}
  {bEnabled=false, emitterName=Balloon_Confetti, rendererClasses [&NiagaraSpriteRendererProperties]}
  {bEnabled=false, emitterName=Balloon_Confetti001, rendererClasses [&NiagaraSpriteRendererProperties]}
  {bEnabled=false, emitterName=Balloon_explosion_Blue001, rendererClasses [&NiagaraMeshRendererProperties]}
  {bEnabled=false, emitterName=Blue_Air_streaks001, rendererClasses [&NiagaraSpriteRendererProperties]}
  {bEnabled=false, emitterName=Pop_particles001, rendererClasses [&NiagaraSpriteRendererProperties]}
  {bEnabled=false, emitterName=SnapRing, rendererClasses [&NiagaraSpriteRendererProperties]}
  {bEnabled=false, emitterName=Balloon_explosion_Blue002, rendererClasses [&NiagaraMeshRendererProperties]}
  {
   bEnabled=false
   emitterName=Balloon_Confetti002
   rendererClasses [&NiagaraSpriteRendererProperties, &NiagaraMeshRendererProperties]
  }
  {bEnabled=false, emitterName=Balloon_explosion_Blue003, rendererClasses [&NiagaraMeshRendererProperties]}
  {bEnabled=false, emitterName=Partition_group_burst_particles, rendererClasses [&NiagaraSpriteRendererProperties]}
  {bEnabled=false, emitterName=Impact001, rendererClasses [&NiagaraMeshRendererProperties]}
  {
   bEnabled=false
   emitterName=Mesh_Partition_group_burst_particles001
   rendererClasses [&NiagaraSpriteRendererProperties, &NiagaraMeshRendererProperties]
  }
  {emitterName=Blue_Air_streaks002, rendererClasses [&NiagaraSpriteRendererProperties]}
  {emitterName=Pop_particles002, rendererClasses [&NiagaraSpriteRendererProperties]}
  {emitterName=SnapRing002, rendererClasses [&NiagaraSpriteRendererProperties]}
  {emitterName=Balloon_explosion_Blue004, rendererClasses [&NiagaraMeshRendererProperties]}
  {emitterName=Balloon_Confetti003, rendererClasses [&NiagaraMeshRendererProperties]}
  {emitterName=Balloon_explosion_Blue005, rendererClasses [&NiagaraMeshRendererProperties]}
  {emitterName=Impact002, rendererClasses [&NiagaraMeshRendererProperties]}
  {emitterName=Mesh_Partition_group_burst_particles, rendererClasses [&NiagaraMeshRendererProperties]}
 ]
 userVariables [
  User.bIsFrontEnd {type=NiagaraBool}
  User.bIsFrontEndPreview {type=NiagaraBool}
 ]
}
